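# Animation Blueprint for the Fortnite player pawn on menu screens, posing the character in standing idles or glider previews.
# Standing uses weapon-set poses or lobby idles chosen by gender, body type and index; gliding selects state machines by glider type.
# The result passes through a linked override layer, montage slots and a weapon additive into a cached pose.
# Curve-driven virtual-bone copies, limb IK, prop hiding and root offset finish it before inertialization.
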AnimBlueprint /Game/Abilities/Player/Pawns/MenuScreen_Athena
target_skeleton=&Fortnite_M_Avg_Player_Skeleton
graphs [
 AnimGraph {nodes [
   Apply Additive(Additive="Sequence Player (additive) 'DefaultFaceIdle_Frontend_Add'"() {sequence=/Game/Animation/Game/MainPlayer/Facial/DefaultFaceIdle_Frontend_Add}) {alphaBoolBlend {blendOption=Linear}}
   n1 = GetCurveValue(CurveName=LeftHandIK)
   Modify Curve(CurveValues_0=SelectFloat(A=0, B=1, bPickA="Equal (Float)"(A=@n1, B=1)), Alpha=1) {curveNames [BothHandFK], curveValues [0]}
   n2 = Use cached pose 'Final Animation Pose'()
   n3 = Break Fort Anim Input Pelvis Adjustment!(FortAnimInput_PelvisAdjustment:"Fort Anim Input Pelvis Adjustment Structure (by ref)"=Get PelvisAdjustmentInput() -> (PelvisAdjustmentInput:Fort Anim Input Pelvis Adjustment Structure)) -> (
    PawnMesh:Skeletal Mesh Component Object Reference
    DistanceToFeet:float=~
    LegLength:float=~
    DesiredLegLengthTreshold:float=~
    DotProductBetweenUpVectorsTreshold:float=~
    PelvisInterpSpeed:float=~
    EmotePelvisOffsetInterpSpeed:float=~
    EmotePelvisOffsetAlpha:float=~
    CurrentEmotePelvisOffset:Vector=~
   ) {is=title K2Node_BreakStruct}
   n4 = Copy Bone - Source Bone: VB VB spine_05_weapon_r_ik_hand_gun - Target Bone: ik_hand_gun!(
    ComponentPose:Component Space Pose Link Structure="Transform (Modify) Bone - Bone: pelvis"(
     ComponentPose=Local To Component(LocalPose=@n2)
     Alpha=@n3.EmotePelvisOffsetAlpha
     Translation=@n3.CurrentEmotePelvisOffset
     Rotation="0,0,0"
     Scale="1,1,1"
    ) {lODThreshold=1, translationMode=BMM_Additive, alphaBoolBlend {blendOption=Linear}, boneToModify {boneName=pelvis}}
    AlphaCurveName:Name=WeaponRInChestSpace
    bCopyTranslation:bool=true
    bCopyRotation:bool=true
    bCopyScale:bool=true
   ) -> (Pose:Component Space Pose Link Structure) {
    is=title AnimGraphNode_CopyBone
    alphaInputType=Curve
    alphaBoolBlend {blendOption=Linear}
    sourceBone {boneName=VB VB spine_05_weapon_r_ik_hand_gun}
    targetBone {boneName=ik_hand_gun}
   }
   n5 = Copy Bone - Source Bone: VB VB spine_05_hand_attach_r_ik_hand_gun - Target Bone: ik_hand_gun!(
    ComponentPose:Component Space Pose Link Structure=@n4
    AlphaCurveName:Name=HandAttachRInChestSpace
    bCopyTranslation:bool=true
    bCopyRotation:bool=true
    bCopyScale:bool=true
   ) -> (Pose:Component Space Pose Link Structure) {
    is=title AnimGraphNode_CopyBone
    alphaInputType=Curve
    alphaBoolBlend {blendOption=Linear}
    sourceBone {boneName=VB VB spine_05_hand_attach_r_ik_hand_gun}
    targetBone {boneName=ik_hand_gun}
   }
   n6 = Copy Bone - Source Bone: VB VB spine_05_weapon_r_ik_hand_l - Target Bone: ik_hand_l!(
    ComponentPose:Component Space Pose Link Structure=@n5
    AlphaCurveName:Name=WeaponRInChestSpace
    bCopyTranslation:bool=true
    bCopyRotation:bool=true
    bCopyScale:bool=true
   ) -> (Pose:Component Space Pose Link Structure) {
    is=title AnimGraphNode_CopyBone
    alphaInputType=Curve
    alphaBoolBlend {blendOption=Linear}
    sourceBone {boneName=VB VB spine_05_weapon_r_ik_hand_l}
    targetBone {boneName=ik_hand_l}
   }
   n7 = Copy Bone - Source Bone: VB VB spine_05_hand_attach_r_ik_hand_l - Target Bone: ik_hand_l!(
    ComponentPose:Component Space Pose Link Structure=@n6
    AlphaCurveName:Name=HandAttachRInChestSpace
    bCopyTranslation:bool=true
    bCopyRotation:bool=true
    bCopyScale:bool=true
   ) -> (Pose:Component Space Pose Link Structure) {
    is=title AnimGraphNode_CopyBone
    alphaInputType=Curve
    alphaBoolBlend {blendOption=Linear}
    sourceBone {boneName=VB VB spine_05_hand_attach_r_ik_hand_l}
    targetBone {boneName=ik_hand_l}
   }
   n8 = Break Fort Anim Input Hand IK!(FortAnimInput_HandIK:"Fort Anim Input Hand IK Structure (by ref)"=Get HandIKInput() -> (HandIKInput:Fort Anim Input Hand IK Structure)) -> (
    HandIKRetargetingWeight:float=~
    HandsInRootSpaceAlpha:float=~
    LeftHandIKAlpha:float=~
    RightHandIKAlpha:float=~
    RightHandFKRotationAlpha:float=~
   ) {is=title K2Node_BreakStruct}
   n9 = Leg IK!(
    ComponentPose:Component Space Pose Link Structure=Copy Bone - Source Bone: VB Root_hand_r - Target Bone: ik_hand_r!(
     ComponentPose:Component Space Pose Link Structure=Copy Bone - Source Bone: VB Root_hand_l - Target Bone: ik_hand_l!(
      ComponentPose:Component Space Pose Link Structure=@n7
      Alpha:float=@n8.HandsInRootSpaceAlpha
      bCopyTranslation:bool=true
      bCopyRotation:bool=true
      bCopyScale:bool=true
     ) -> (Pose:Component Space Pose Link Structure) {is=title AnimGraphNode_CopyBone, alphaBoolBlend {blendOption=Linear}, sourceBone {boneName=VB Root_hand_l}, targetBone {boneName=ik_hand_l}}
     Alpha:float=@n8.HandsInRootSpaceAlpha
     bCopyTranslation:bool=true
     bCopyRotation:bool=true
     bCopyScale:bool=true
    ) -> (Pose:Component Space Pose Link Structure) {is=title AnimGraphNode_CopyBone, alphaBoolBlend {blendOption=Linear}, sourceBone {boneName=VB Root_hand_r}, targetBone {boneName=ik_hand_r}}
    Alpha:float=Get LegIKAlpha() -> (LegIKAlpha:float)
   ) -> (Pose:Component Space Pose Link Structure) {
    is=title AnimGraphNode_LegIK
    alphaBoolBlend {blendOption=Linear}
    legsDefinition [
     {
      bEnableKneeTwistCorrection=false
      bEnableRotationLimit=false
      footBoneForwardAxis=Y
      hingeRotationAxis=Z
      minRotationAngle=15
      numBonesInLimb=2
      twistOffsetCurveName=None
      fKFootBone {boneName=foot_l}
      iKFootBone {boneName=ik_foot_l}
     }
     {
      bEnableKneeTwistCorrection=false
      bEnableRotationLimit=false
      footBoneForwardAxis=Y
      hingeRotationAxis=Z
      minRotationAngle=15
      numBonesInLimb=2
      twistOffsetCurveName=None
      fKFootBone {boneName=foot_r}
      iKFootBone {boneName=ik_foot_r}
     }
    ]
   }
   n10 = Leg IK(ComponentPose=@n9, Alpha=Get LeftHandIKAlpha() -> (LeftHandIKAlpha:float)) {
    alphaBoolBlend {blendOption=Linear}
    legsDefinition [{
      bEnableKneeTwistCorrection=false
      bEnableRotationLimit=false
      footBoneForwardAxis=Y
      hingeRotationAxis=Z
      minRotationAngle=15
      numBonesInLimb=2
      twistOffsetCurveName=None
      fKFootBone {boneName=hand_l}
      iKFootBone {boneName=ik_hand_l}
     }]
   }
   n11 = Leg IK(ComponentPose=@n10, Alpha=Get RightHandIKAlpha() -> (RightHandIKAlpha:float)) {
    alphaBoolBlend {blendOption=Linear}
    legsDefinition [{
      bEnableKneeTwistCorrection=false
      bEnableRotationLimit=false
      footBoneForwardAxis=Y
      hingeRotationAxis=Z
      minRotationAngle=15
      numBonesInLimb=2
      twistOffsetCurveName=None
      fKFootBone {boneName=hand_r}
      iKFootBone {boneName=ik_hand_r}
     }]
   }
   n12 = Copy Bone - Source Bone: VB spine_05_weapon_r - Target Bone: weapon_r!(
    ComponentPose:Component Space Pose Link Structure=@n11
    AlphaCurveName:Name=WeaponRInChestSpace
    bCopyTranslation:bool=true
    bCopyRotation:bool=true
    bCopyScale:bool=true
   ) -> (Pose:Component Space Pose Link Structure) {
    is=title AnimGraphNode_CopyBone
    alphaInputType=Curve
    alphaBoolBlend {blendOption=Linear}
    sourceBone {boneName=VB spine_05_weapon_r}
    targetBone {boneName=weapon_r}
   }
   n13 = Copy Bone - Source Bone: VB spine_05_hand_attach_r - Target Bone: hand_attach_r!(
    ComponentPose:Component Space Pose Link Structure=@n12
    AlphaCurveName:Name=HandAttachRInChestSpace
    bCopyTranslation:bool=true
    bCopyRotation:bool=true
    bCopyScale:bool=true
   ) -> (Pose:Component Space Pose Link Structure) {
    is=title AnimGraphNode_CopyBone
    alphaInputType=Curve
    alphaBoolBlend {blendOption=Linear}
    sourceBone {boneName=VB spine_05_hand_attach_r}
    targetBone {boneName=hand_attach_r}
   }
   n14 = Copy Bone - Source Bone: VB ik_foot_root_weapon_l - Target Bone: weapon_l!(
    ComponentPose:Component Space Pose Link Structure=@n13
    AlphaCurveName:Name=WeaponLinFloorSpace
    bCopyTranslation:bool=true
    bCopyRotation:bool=true
    bCopyScale:bool=true
   ) -> (Pose:Component Space Pose Link Structure) {
    is=title AnimGraphNode_CopyBone
    alphaInputType=Curve
    alphaBoolBlend {blendOption=Linear}
    sourceBone {boneName=VB ik_foot_root_weapon_l}
    targetBone {boneName=weapon_l}
   }
   n15 = Copy Bone - Source Bone: VB ik_foot_root_weapon_r - Target Bone: weapon_r!(
    ComponentPose:Component Space Pose Link Structure=@n14
    AlphaCurveName:Name=WeaponRinFloorSpace
    bCopyTranslation:bool=true
    bCopyRotation:bool=true
    bCopyScale:bool=true
   ) -> (Pose:Component Space Pose Link Structure) {
    is=title AnimGraphNode_CopyBone
    alphaInputType=Curve
    alphaBoolBlend {blendOption=Linear}
    sourceBone {boneName=VB ik_foot_root_weapon_r}
    targetBone {boneName=weapon_r}
   }
   n16 = Copy Bone - Source Bone: VB ik_foot_root_hand_attach_l - Target Bone: hand_attach_l!(
    ComponentPose:Component Space Pose Link Structure=@n15
    AlphaCurveName:Name=HandAttachLinFloorSpace
    bCopyTranslation:bool=true
    bCopyRotation:bool=true
    bCopyScale:bool=true
   ) -> (Pose:Component Space Pose Link Structure) {
    is=title AnimGraphNode_CopyBone
    alphaInputType=Curve
    alphaBoolBlend {blendOption=Linear}
    sourceBone {boneName=VB ik_foot_root_hand_attach_l}
    targetBone {boneName=hand_attach_l}
   }
   n17 = Copy Bone - Source Bone: VB ik_foot_root_hand_attach_r - Target Bone: hand_attach_r!(
    ComponentPose:Component Space Pose Link Structure=@n16
    AlphaCurveName:Name=HandAttachRinFloorSpace
    bCopyTranslation:bool=true
    bCopyRotation:bool=true
    bCopyScale:bool=true
   ) -> (Pose:Component Space Pose Link Structure) {
    is=title AnimGraphNode_CopyBone
    alphaInputType=Curve
    alphaBoolBlend {blendOption=Linear}
    sourceBone {boneName=VB ik_foot_root_hand_attach_r}
    targetBone {boneName=hand_attach_r}
   }
   n18 = Get HidePropBones() -> (HidePropBones:bool)
   n19 = "Transform (Modify) Bone - Bone: weapon_l"!(ComponentPose=@n17, bAlphaBoolEnabled:bool=@n18, Translation="0,0,-1000", Rotation="0,0,0", Scale="1,1,1") -> (Pose) {
    alphaInputType=Bool
    translationMode=BMM_Additive
    alphaBoolBlend {blendOption=Linear}
    boneToModify {boneName=weapon_l}
    translation {z=-1000}
   }
   n20 = "Transform (Modify) Bone - Bone: weapon_r"!(ComponentPose=@n19, bAlphaBoolEnabled:bool=@n18, Translation="0,0,-1000", Rotation="0,0,0", Scale="1,1,1") -> (Pose) {
    alphaInputType=Bool
    translationMode=BMM_Additive
    alphaBoolBlend {blendOption=Linear}
    boneToModify {boneName=weapon_r}
    translation {z=-1000}
   }
   n21 = "Transform (Modify) Bone - Bone: Root"(
    ComponentPose=@n20
    Alpha=1
    Translation=Get OffsetTranslate() -> (OffsetTranslate:Vector)
    Rotation=Get OffsetRotate() -> (OffsetRotate:Rotator)
    Scale="1,1,1"
   ) {rotationMode=BMM_Additive, translationMode=BMM_Additive, alphaBoolBlend {blendOption=Linear}, boneToModify {boneName=Root}}
   Output Pose(Result=Inertialization!(Source:Pose Link Structure=Component To Local(ComponentPose=@n21)) -> (Pose:Pose Link Structure) {is=title AnimGraphNode_Inertialization}) {layerGroup=None}
   AnimGraphNode_StateMachine_1 = Glider_SurfStyle() {is=title AnimGraphNode_StateMachine}
   AnimGraphNode_StateMachine_3 = Glider_or_Unmbrella() {is=title AnimGraphNode_StateMachine}
   AnimGraphNode_StateMachine_73 = NoGlider() {is=title AnimGraphNode_StateMachine}
   n22 = "Blend Poses (EGliderType)"!(
    BlendPose_0:Pose Link Structure=@AnimGraphNode_StateMachine_3
    BlendPose_1:Pose Link Structure=@AnimGraphNode_StateMachine_1
    BlendPose_2:Pose Link Structure=@AnimGraphNode_StateMachine_73
    BlendTime_0:float=0
    BlendTime_1:float=0
    BlendTime_2:float=0.1
    ActiveEnumValue:Byte=Break Fort Anim Input Player Glider Anim Asset!(FortAnimInput_PlayerGliderAnimAsset:"Fort Anim Input Player Glider Anim Asset Structure (by ref)"=Get CurrentGliderAnimSet() -> (CurrentGliderAnimSet:Fort Anim Input Player Glider Anim Asset Structure)) -> (
     Into_Base_MaleMedium_BS:Blend Space Object Reference
     PlayerGliderType:EGliderType Enum
     bUseLeanPlayRateAdjust:bool=~
     bUseForwardAs_Y_BS:bool=~
     bUseDeployCurveForIntoPlayRateAdjust:bool=~
     LeanForwardValue:float=~
     IdleValue:float=~
     LeanBackValue:float=~
     IdleInterpToSpeed:float=~
     LeanForwardInterpToSpeed:float=~
     LeanBackInterpToSpeed:float=~
    ) {is=title K2Node_BreakStruct}.PlayerGliderType
   ) -> (Pose:Pose Link Structure) {
    is=title AnimGraphNode_BlendListByEnum
    blendType=Linear
    transitionType=Inertialization
    blendPose [{linkId=-1, sourceLinkId=-1}, {linkId=-1, sourceLinkId=-1}, {linkId=-1, sourceLinkId=-1}]
    blendTime [0.1,0.1,0.1]
   }
   AnimGraphNode_StateMachine_0 = NoGlider_Lean() {is=title AnimGraphNode_StateMachine}
   Save cached pose 'Gliding Pose'(Pose=Apply Additive!(Base=@n22, Additive=@AnimGraphNode_StateMachine_0, AlphaCurveName:Name=ParaGlideLeanAlpha) -> (Pose) {alphaInputType=Curve, alphaBoolBlend {blendOption=Linear}}) {cacheName=Gliding Pose}
   n23 = Break Fort Anim Input Standing Pawn Anim Asset!(FortAnimInput_StandingPawnAnimAsset:"Fort Anim Input Standing Pawn Anim Asset Structure (by ref)"=Get CurrentAnimSet() -> (CurrentAnimSet:Fort Anim Input Standing Pawn Anim Asset Structure)) -> (TargetingPose:Anim Sequence Object Reference, NonTargetingPose:Anim Sequence Object Reference, IdleNoise:Anim Sequence Object Reference) {is=title K2Node_BreakStruct}
   n24 = Sequence Player!(Sequence:Anim Sequence Base Object Reference=@n23.NonTargetingPose) -> (Pose)
   n25 = Sequence Player(Sequence=@n23.IdleNoise)
   AnimGraphNode_StateMachine_2 = RandomLobbyIdle() {is=title AnimGraphNode_StateMachine}
   Save cached pose 'Standing Idle'(Pose=Blend Poses by bool(
     BlendPose_0=Apply Additive!(Base=@n24, Additive=@n25, Alpha:float) -> (Pose) {alphaBoolBlend {blendOption=Linear}}
     BlendPose_1=@AnimGraphNode_StateMachine_2
     bActiveValue=Get HasWeaponEquipped() -> (HasWeaponEquipped:bool)
    ) {blendType=Linear, transitionType=Inertialization}) {cacheName=Standing Idle}
   n26 = Use cached pose 'Standing Idle'()
   n27 = Get bUseSimplePose() -> (bUseSimplePose:bool)
   n28 = Get HasWeaponEquipped()
   Save cached pose 'Standing Pose'(Pose=Blend Poses by bool(
     BlendPose_0=Use cached pose 'Standing Idle'()
     BlendPose_1=Slot 'CustomFrontEndIdle'(Source=@n26) {slotName=CustomFrontEndIdle}
     bActiveValue=OR Boolean(A=@n27, B=@n28)
    ) {blendType=Linear, transitionType=Inertialization}) {cacheName=Standing Pose}
   n29 = Break Fort Anim Input Weapon Additive Anim Asset!(FortAnimInput_WeaponAdditiveAnimAsset:"Fort Anim Input Weapon Additive Anim Asset Structure (by ref)"=Get CurrentWeaponAdditiveAnimSet() -> (CurrentWeaponAdditiveAnimSet:Fort Anim Input Weapon Additive Anim Asset Structure)) -> (
    AdditivePoseOffset:Anim Sequence Object Reference
    TargetingAdditivePoseOffset:Anim Sequence Object Reference
    NonTargetedAdditivePoseOffset:Anim Sequence Object Reference
    RelaxedAdditivePoseOffset:Anim Sequence Object Reference
    RelaxedWhileSprintingAdditiveAnim:Anim Sequence Object Reference
    RelaxedWhileCrouchSprintingAdditiveAnim:Anim Sequence Object Reference
   ) {is=title K2Node_BreakStruct}
   n30 = PlayerAnimOverrideLayerInterface - FullBodyOverride!(PassThroughFullBodyPose:Pose Link Structure=Blend Poses by bool(
     BlendPose_0=Use cached pose 'Gliding Pose'()
     BlendPose_1=Use cached pose 'Standing Pose'()
     bActiveValue=Get bIsSkydiving() -> (bIsSkydiving:bool)
    ) {blendType=Linear, transitionType=Inertialization}) -> (Pose:Pose Link Structure) {
    is=title AnimGraphNode_LinkedAnimLayer
    tag=FullBodyOverride
    interface=/Game/Characters/Player/Male/Male_Avg_Base/AnimLayerInterface/PlayerAnimOverrideLayerInterface.PlayerAnimOverrideLayerInterface_C
    layer=FullBodyOverride
   }
   n31 = Slot 'AdditiveCorrective'(Source=Slot 'FullBody'(Source=@n30) {slotName=FullBody}) {slotName=AdditiveCorrective}
   Save cached pose 'Final Animation Pose'(Pose=Apply Additive(Base=Slot 'UpperBody'(Source=@n31) {slotName=UpperBody}, Additive=Sequence Player(Sequence=@n29.RelaxedAdditivePoseOffset)) {alphaBoolBlend {blendOption=Linear}}) {cacheName=Final Animation Pose}
  ]}
 NoGlider_Lean {
  entry=Lean_Zero
  path=AnimGraph.AnimGraphNode_StateMachine_0.NoGlider_Lean
  ids=id
  nodes [AnimStateNode_10 = Lean_Zero() {is=title AnimStateNode}]
 }
 Lean_Zero {
  path=AnimGraph.AnimGraphNode_StateMachine_0.NoGlider_Lean.AnimStateNode_10.Lean_Zero
  nodes [
   n1 = Break Fort Anim Input Player Glider Anim Asset!(FortAnimInput_PlayerGliderAnimAsset=Get CurrentGliderAnimSet()) -> (
    Default_Base_BS:Blend Space Object Reference
    Default_BodyAdditive_MaleMedium_BS:Blend Space Object Reference
    Default_BodyAdditive_MaleLarge_BS:Blend Space Object Reference
    Default_BodyAdditive_FemaleSmall_BS:Blend Space Object Reference
    Default_BodyAdditive_FemaleMedium_BS:Blend Space Object Reference
    Default_BodyAdditive_FemaleLarge_BS:Blend Space Object Reference
    Default_TurnAdditive_MaleMedium_BS:Blend Space Object Reference
    Default_TurnAdditive_MaleLarge_BS:Blend Space Object Reference
    Default_TurnAdditive_FemaleSmall_BS:Blend Space Object Reference
    Default_TurnAdditive_FemaleMedium_BS:Blend Space Object Reference
    Default_TurnAdditive_FemaleLarge_BS:Blend Space Object Reference
    Into_Base_BS:Blend Space Object Reference
    Into_Base_MaleMedium_BS
    Into_BodyAdditive_MaleMedium_BS:Blend Space Object Reference
    Into_BodyAdditive_MaleLarge_BS:Blend Space Object Reference
    Into_BodyAdditive_FemaleSmall_BS:Blend Space Object Reference
    Into_BodyAdditive_FemaleMedium_BS:Blend Space Object Reference
    Into_BodyAdditive_FemaleLarge_BS:Blend Space Object Reference
    Into_TurnAdditive_MaleMedium_BS:Blend Space Object Reference
    Into_TurnAdditive_MaleLarge_BS:Blend Space Object Reference
    Into_TurnAdditive_FemaleSmall_BS:Blend Space Object Reference
    Into_TurnAdditive_FemaleMedium_BS:Blend Space Object Reference
    Into_TurnAdditive_FemaleLarge_BS:Blend Space Object Reference
    Lean_MaleMedium_BS:Blend Space Object Reference
    Lean_MaleLarge_BS:Blend Space Object Reference
    Lean_FemaleSmall_BS:Blend Space Object Reference
    Lean_FemaleMedium_BS:Blend Space Object Reference
    Lean_FemaleLarge_BS:Blend Space Object Reference
    LeanAdditive_Center_MaleMedium_Pose:Anim Sequence Object Reference
    LeanAdditive_Center_MaleLarge_Pose:Anim Sequence Object Reference
    LeanAdditive_Center_FemaleSmall_Pose:Anim Sequence Object Reference
    LeanAdditive_Center_FemaleMedium_Pose:Anim Sequence Object Reference
    LeanAdditive_Center_FemaleLarge_Pose:Anim Sequence Object Reference
    LeanAdditive_Into_BS:Blend Space Object Reference
    LeanAdditive_ForwardInto_Anim:Anim Sequence Object Reference
    LeanAdditive_ForwardInto_FromDeploy_Anim:Anim Sequence Object Reference
    LeanAdditive_BackInto_Anim:Anim Sequence Object Reference
    LeanAdditive_LeftInto_Anim:Anim Sequence Object Reference
    LeanAdditive_RightInto_Anim:Anim Sequence Object Reference
    LeanAdditive_ForwardOut_Anim:Anim Sequence Object Reference
    LeanAdditive_BackOut_Anim:Anim Sequence Object Reference
    LeanAdditive_LeftOut_Anim:Anim Sequence Object Reference
    LeanAdditive_RightOut_Anim:Anim Sequence Object Reference
    ToGlide_BS:Blend Space Object Reference
    ToGlide_Lean_BS:Blend Space Object Reference
    ToDive_BS:Blend Space Object Reference
    ToDive_Lean_BS:Blend Space Object Reference
    Dive_WeaponR_Additive_BS:Blend Space Object Reference
    Glide_WeaponR_Additive_Anim:Anim Sequence Object Reference
    GenericAdditive_Male_BS:Blend Space Object Reference
    GenericAdditive_Female_BS:Blend Space Object Reference
    RootModPitchMin:float=~
    RootModPitchMax:float=~
    RootModYOffsetMin:float=~
    RootModYOffsetMax:float=~
    PelvisModVertical:Pelvis Mod Vertical Input Spring Structure
    PelvisModLateral:Pelvis Mod Lateral Input Spring Structure
    PelvisModBankAngle:Pelvis Mod Bank Angle Input Spring Structure
    ThighModLegAngle:Thigh Mod Leg Angle Input Structure
    ThighModLegPitch:Thigh Mod Leg Pitch Spring Structure
    ThighModLegBank:Thigh Mod Leg Bank Spring Structure
    ThighModLeftLegDrag:Thigh Mod Left Leg Drag Structure
    PlayerGliderType
    bEnableSpringMods:bool=~
    bAllowPlayerDeployRootMod:bool=~
    bUseSurfStyle:bool=~
    bUseLeanPlayRateAdjust=~
    bUseForwardAs_Y_BS=~
    bUseDeployCurveForIntoPlayRateAdjust=~
    LeanForwardValue=~
    IdleValue=~
    LeanBackValue=~
    IdleInterpToSpeed=~
    LeanForwardInterpToSpeed=~
    LeanBackInterpToSpeed=~
   )
   n2 = "Blend Poses (EFortPlayerAnimBodyType)"!(
    BlendPose_0:Pose Link Structure=Sequence Player(Sequence=@n1.LeanAdditive_Center_MaleMedium_Pose)
    BlendPose_1:Pose Link Structure=Sequence Player(Sequence=@n1.LeanAdditive_Center_MaleLarge_Pose)
    BlendTime_0:float=0.1
    BlendTime_1:float=0.1
    ActiveEnumValue:Byte=Get AnimBodyType() -> (AnimBodyType:EFortPlayerAnimBodyType Enum=Small)
   ) -> (Pose:Pose Link Structure) {
    is=title AnimGraphNode_BlendListByEnum
    blendType=Linear
    blendPose [{linkId=-1, sourceLinkId=-1}, {linkId=-1, sourceLinkId=-1}]
    blendTime [0.1,0.1]
   }
   n3 = "Blend Poses (EFortPlayerAnimBodyType)"!(
    BlendPose_0=Sequence Player(Sequence=@n1.LeanAdditive_Center_FemaleMedium_Pose)
    BlendPose_1=Sequence Player(Sequence=@n1.LeanAdditive_Center_FemaleSmall_Pose)
    BlendPose_2:Pose Link Structure=Sequence Player(Sequence=@n1.LeanAdditive_Center_FemaleLarge_Pose)
    BlendTime_0=0.1
    BlendTime_1=0.1
    BlendTime_2:float=0.1
    ActiveEnumValue=Get AnimBodyType() -> (AnimBodyType=Small)
   ) -> (Pose) {
    blendType=Linear
    blendPose [{linkId=-1, sourceLinkId=-1}, {linkId=-1, sourceLinkId=-1}, {linkId=-1, sourceLinkId=-1}]
    blendTime [0.1,0.1,0.1]
   }
   n4 = "Blend Poses (EFortCustomGender)"!(
    BlendPose_0:Pose Link Structure=@n2
    BlendPose_1:Pose Link Structure=@n3
    BlendTime_0:float=0
    BlendTime_1:float=0
    ActiveEnumValue:Byte=Get Gender() -> (Gender:EFortDisplayGender Enum=Unknown)
   ) -> (Pose:Pose Link Structure) {
    is=title AnimGraphNode_BlendListByEnum
    blendType=Linear
    blendPose [{linkId=-1, sourceLinkId=-1}, {linkId=-1, sourceLinkId=-1}]
    blendTime [0.1,0.1]
   }
   Output Animation Pose(Result=@n4)
  ]
 }
 Glider_SurfStyle {
  entry=ParaGlide_Into_Deploy
  path=AnimGraph.AnimGraphNode_StateMachine_1.Glider_SurfStyle
  ids=id
  nodes [
   AnimStateNode_0 = ParaGlide_SurfStyle() {is=title AnimStateNode}
   AnimStateNode_1 = ParaGlide_Into_Deploy() {is=title AnimStateNode}
   AnimStateTransitionNode_1 = AnimStateTransitionNode() {bAutomaticRuleBasedOnSequencePlayerInState=true, blendMode=Linear, crossfadeDuration=0.1}
  ]
  links [{
    dst=ParaGlide_SurfStyle
    rule=AnimGraph.AnimGraphNode_StateMachine_1.Glider_SurfStyle.AnimStateTransitionNode_1.AnimationTransitionGraph_1
    src=ParaGlide_Into_Deploy
    via=AnimStateTransitionNode_1
   }]
 }
 ParaGlide_SurfStyle {
  path=AnimGraph.AnimGraphNode_StateMachine_1.Glider_SurfStyle.AnimStateNode_0.ParaGlide_SurfStyle
  nodes [
   n1 = Break Fort Anim Input Player Glider Anim Asset(FortAnimInput_PlayerGliderAnimAsset=Get CurrentGliderAnimSet()) -> (
    RootModPitchMin=~
    RootModPitchMax=~
    RootModYOffsetMin=~
    RootModYOffsetMax=~
    bEnableSpringMods=~
    bAllowPlayerDeployRootMod=~
    bUseSurfStyle=~
    bUseLeanPlayRateAdjust=~
    bUseForwardAs_Y_BS=~
    bUseDeployCurveForIntoPlayRateAdjust=~
    LeanForwardValue=~
    IdleValue=~
    LeanBackValue=~
    IdleInterpToSpeed=~
    LeanForwardInterpToSpeed=~
    LeanBackInterpToSpeed=~
   )
   n2 = Break Fort Anim Input Skydiving!(FortAnimInput_Skydiving:"Fort Anim Input Skydiving Structure (by ref)"=Get Skydiving() -> (Skydiving:Fort Anim Input Skydiving Structure)) -> (
    ParaGilderRootOffset:Vector=~
    ParaGliderRootRotationOverride:Rotator=~
    bUseParaGlideRootModifier:bool=~
    bIsSkydivingFromLaunchPad:bool=~
    bIsSkydivingFromBus:bool=~
    bIsInVortex:bool=~
    bIsDiving:bool=~
    bIsParachuteOpen:bool=~
    bIsSkydiving:bool=~
    bIsParachuteLeaning:bool=~
    bIsSkydiveLeaning:bool=~
    bIsLeaning:bool=~
    bIsSkydiveDiveMode:bool=~
    bPlaySkydiveDrift:bool=~
    bSkydiveDriftDelayActive:bool=~
    bSkydiveDriftAnimAllowed:bool=~
    bIsGliderRight:bool=~
    bIsGliderCenter:bool=~
    bIsGliderLeft:bool=~
    bIsGliderForward:bool=~
    bIsGliderBack:bool=~
    bHasDeployedAndIsStandingUpright:bool=~
    LocalAccelForward:float=~
    LocalAccelRight:float=~
    LocalVelocityRight:float=~
    SkydiveAimPitch:float=~
    SkydiveAimYaw:float=~
    DeployChuteAnimRate:float=~
    SkydiveDriftAnimRate:float=~
    SkydiveDriftAnimRateCurrent:float=~
    SkydiveFidgetAnimRate:float=~
    SkydiveFidgetAnimRateCurrent:float=~
    SkydiveAdditiveAlpha:float=~
    SkydiveDriftDelay:float=~
    SkydiveDriftAnim:int
    SkydiveDriftAnimMax:int
    LaunchpadAnim:int
    LocalAccelDir:ESkydivingDirection Enum
    DirectionLast:ESkydivingDirection Enum
    PoseAdditive:Anim Sequence Object Reference
   ) {is=title K2Node_BreakStruct}
   n3 = Apply Additive(
    Base=Apply Additive(
     Base=Blendspace Player!(X=@n2.SkydiveAimYaw, Y=@n2.LocalAccelForward, BlendSpace:Blend Space Object Reference=@n1) -> (Pose)
     Additive=Blendspace Player(X=@n2.SkydiveAimYaw, Y=@n2.LocalAccelForward, BlendSpace=@n1.Default_BodyAdditive_FemaleMedium_BS) {bResetPlayTimeWhenBlendSpaceChanges=false, groupName=SurfGlide, groupRole=AlwaysFollower, method=SyncGroup}
    ) {alphaBoolBlend {blendOption=Linear}, alphaScaleBiasClamp {outRange {max=0, min=1}}}
    Additive=Blendspace Player(X=@n2.SkydiveAimYaw, Y=@n2.LocalAccelForward, BlendSpace=@n1.Default_TurnAdditive_FemaleMedium_BS)
   ) {alphaBoolBlend {blendOption=Linear}, alphaScaleBiasClamp {outRange {max=0, min=1}}}
   n4 = "Sequence Player (additive) 'TemplateSurfer_TurnLeanLow_Pose'"() {sequence=/Game/Animation/Game/MainPlayer/Skydive/ParaGlide/_TemplateSurfer/Medium/Female/TemplateSurfer_TurnLeanLow_Pose}
   AnimGraphNode_StateMachine_0 = TriggerAnimState() {is=title AnimGraphNode_StateMachine}
   Output Animation Pose(Result=Apply Additive!(
     Base=Apply Additive(Base=@n3, Additive=@n4, bAlphaBoolEnabled=false) {alphaInputType=Bool, alphaBoolBlend {blendInTime=0.125, blendOption=Linear, blendOutTime=0.125}}
     Additive=@AnimGraphNode_StateMachine_0
     bAlphaBoolEnabled:bool=false
    ) -> (Pose) {alphaInputType=Bool, alphaBoolBlend {blendInTime=0.1, blendOption=Linear, blendOutTime=0.1}})
  ]
 }
 TriggerAnimState {
  entry=TriggerAnim
  path=AnimGraph.AnimGraphNode_StateMachine_1.Glider_SurfStyle.AnimStateNode_0.ParaGlide_SurfStyle.AnimGraphNode_StateMachine_0.TriggerAnimState
  ids=id
  nodes [AnimStateNode_0 = TriggerAnim() {is=title AnimStateNode}]
 }
 TriggerAnim {
  path=AnimGraph.AnimGraphNode_StateMachine_1.Glider_SurfStyle.AnimStateNode_0.ParaGlide_SurfStyle.AnimGraphNode_StateMachine_0.TriggerAnimState.AnimStateNode_0.TriggerAnim
  nodes [
   n1 = Get CurrentGliderAnimSet()
   n2 = Break Fort Anim Input Skydiving(FortAnimInput_Skydiving=Get Skydiving()) -> (
    ParaGilderRootOffset=~
    ParaGliderRootRotationOverride=~
    bUseParaGlideRootModifier=~
    bIsSkydivingFromLaunchPad=~
    bIsSkydivingFromBus=~
    bIsInVortex=~
    bIsDiving=~
    bIsParachuteOpen=~
    bIsSkydiving=~
    bIsParachuteLeaning=~
    bIsSkydiveLeaning=~
    bIsLeaning=~
    bIsSkydiveDiveMode=~
    bPlaySkydiveDrift=~
    bSkydiveDriftDelayActive=~
    bSkydiveDriftAnimAllowed=~
    bIsGliderRight=~
    bIsGliderCenter=~
    bIsGliderLeft=~
    bIsGliderForward=~
    bIsGliderBack=~
    bHasDeployedAndIsStandingUpright=~
    LocalAccelForward=~
    LocalAccelRight=~
    LocalVelocityRight=~
    SkydiveAimPitch=~
    SkydiveAimYaw=~
    DeployChuteAnimRate=~
    SkydiveDriftAnimRate=~
    SkydiveDriftAnimRateCurrent=~
    SkydiveFidgetAnimRate=~
    SkydiveFidgetAnimRateCurrent=~
    SkydiveAdditiveAlpha=~
    SkydiveDriftDelay=~
   )
   Output Animation Pose(Result=Blendspace Player(
     X=@n2.SkydiveAimYaw
     Y=@n2.LocalAccelForward
     BlendSpace=Break Fort Anim Input Player Glider Anim Asset(FortAnimInput_PlayerGliderAnimAsset=@n1) -> (
      RootModPitchMin=~
      RootModPitchMax=~
      RootModYOffsetMin=~
      RootModYOffsetMax=~
      bEnableSpringMods=~
      bAllowPlayerDeployRootMod=~
      bUseSurfStyle=~
      bUseLeanPlayRateAdjust=~
      bUseForwardAs_Y_BS=~
      bUseDeployCurveForIntoPlayRateAdjust=~
      LeanForwardValue=~
      IdleValue=~
      LeanBackValue=~
      IdleInterpToSpeed=~
      LeanForwardInterpToSpeed=~
      LeanBackInterpToSpeed=~
     ).GenericAdditive_Female_BS
    ))
  ]
 }
 ParaGlide_Into_Deploy {
  path=AnimGraph.AnimGraphNode_StateMachine_1.Glider_SurfStyle.AnimStateNode_1.ParaGlide_Into_Deploy
  nodes [
   n1 = Break Fort Anim Input Player Glider Anim Asset!(FortAnimInput_PlayerGliderAnimAsset=Get CurrentGliderAnimSet()) -> (
    Into_Base_MaleMedium_BS
    Into_BodyAdditive_MaleMedium_BS
    Into_BodyAdditive_FemaleMedium_BS
    PelvisModVertical
    PelvisModLateral
    PelvisModBankAngle
    ThighModLegAngle
    ThighModLegPitch
    ThighModLegBank
    ThighModLeftLegDrag
    PlayerGliderType
    bEnableSpringMods=~
    bUseLeanPlayRateAdjust=~
    bUseForwardAs_Y_BS=~
    bUseDeployCurveForIntoPlayRateAdjust=~
    LeanForwardValue=~
    IdleValue=~
    LeanBackValue=~
    IdleInterpToSpeed=~
    LeanForwardInterpToSpeed=~
    LeanBackInterpToSpeed=~
   )
   n2 = Break Fort Anim Input Skydiving(FortAnimInput_Skydiving=Get Skydiving()) -> (
    ParaGilderRootOffset=~
    ParaGliderRootRotationOverride=~
    bUseParaGlideRootModifier=~
    bIsSkydivingFromLaunchPad=~
    bIsSkydivingFromBus=~
    bIsInVortex=~
    bIsDiving=~
    bIsParachuteOpen=~
    bIsSkydiving=~
    bIsParachuteLeaning=~
    bIsSkydiveLeaning=~
    bIsLeaning=~
    bIsSkydiveDiveMode=~
    bPlaySkydiveDrift=~
    bSkydiveDriftDelayActive=~
    bSkydiveDriftAnimAllowed=~
    bIsGliderRight=~
    bIsGliderCenter=~
    bIsGliderLeft=~
    bIsGliderForward=~
    bIsGliderBack=~
    bHasDeployedAndIsStandingUpright=~
    LocalAccelForward=~
    LocalAccelRight=~
    LocalVelocityRight=~
    SkydiveAimPitch=~
    SkydiveAimYaw=~
    DeployChuteAnimRate=~
    SkydiveDriftAnimRate=~
    SkydiveDriftAnimRateCurrent=~
    SkydiveFidgetAnimRate=~
    SkydiveFidgetAnimRateCurrent=~
    SkydiveAdditiveAlpha=~
    SkydiveDriftDelay=~
   )
   n3 = "Blend Poses (EFortCustomGender)"(
    BlendPose_0=Blendspace Player(X=@n2.SkydiveAimYaw, Y=@n2.SkydiveAimPitch, BlendSpace=@n1.Into_BodyAdditive_MaleMedium_BS) {bIgnoreForRelevancyTest=true, bLoop=false}
    BlendPose_1=Blendspace Player(X=@n2.SkydiveAimYaw, Y=@n2.SkydiveAimPitch, BlendSpace=@n1.Into_BodyAdditive_FemaleMedium_BS) {bIgnoreForRelevancyTest=true, bLoop=false}
    BlendTime_0=0
    BlendTime_1=0
    ActiveEnumValue=Get Gender() -> (Gender=Unknown)
   ) {blendType=Linear, blendPose [{linkId=-1, sourceLinkId=-1}, {linkId=-1, sourceLinkId=-1}], blendTime [0.1,0.1]}
   n4 = Get CurrentGliderAnimSet()
   n5 = Apply Additive!(
    Base=Blendspace Player(
     X=@n2.SkydiveAimYaw
     Y=@n2.SkydiveAimPitch
     BlendSpace=Break Fort Anim Input Player Glider Anim Asset!(FortAnimInput_PlayerGliderAnimAsset=@n4) -> (
      Into_Base_BS:Blend Space Object Reference
      Into_Base_MaleMedium_BS
      PelvisModVertical
      PelvisModLateral
      PelvisModBankAngle
      ThighModLegAngle
      ThighModLegPitch
      ThighModLegBank
      ThighModLeftLegDrag
      PlayerGliderType
      bEnableSpringMods=~
      bUseLeanPlayRateAdjust=~
      bUseForwardAs_Y_BS=~
      bUseDeployCurveForIntoPlayRateAdjust=~
      LeanForwardValue=~
      IdleValue=~
      LeanBackValue=~
      IdleInterpToSpeed=~
      LeanForwardInterpToSpeed=~
      LeanBackInterpToSpeed=~
     )
    ) {bLoop=false}
    Additive=@n3
    AlphaCurveName:Name=ParaGlideLeanAlpha
   ) -> (Pose) {alphaInputType=Curve, alphaBoolBlend {blendOption=Linear}, alphaScaleBiasClamp {bMapRange=true, inRange {max=0, min=1}}}
   n6 = Blendspace Player(
    X=@n2.SkydiveAimYaw
    Y=@n2.LocalAccelForward
    BlendSpace=Break Fort Anim Input Player Glider Anim Asset!(FortAnimInput_PlayerGliderAnimAsset=Get CurrentGliderAnimSet()) -> (
     Default_BodyAdditive_FemaleMedium_BS:Blend Space Object Reference
     Into_Base_MaleMedium_BS
     PelvisModVertical
     PelvisModLateral
     PelvisModBankAngle
     ThighModLegAngle
     ThighModLegPitch
     ThighModLegBank
     ThighModLeftLegDrag
     PlayerGliderType
     bEnableSpringMods=~
     bUseLeanPlayRateAdjust=~
     bUseForwardAs_Y_BS=~
     bUseDeployCurveForIntoPlayRateAdjust=~
     LeanForwardValue=~
     IdleValue=~
     LeanBackValue=~
     IdleInterpToSpeed=~
     LeanForwardInterpToSpeed=~
     LeanBackInterpToSpeed=~
    )
   ) {bIgnoreForRelevancyTest=true, bResetPlayTimeWhenBlendSpaceChanges=false, groupName=SurfGlide, groupRole=AlwaysLeader, method=SyncGroup}
   Output Animation Pose(Result=Apply Additive(Base=@n5, Additive=@n6, AlphaCurveName=ParaGlideLeanAlpha) {alphaInputType=Curve, alphaBoolBlend {blendOption=Linear}, alphaScaleBiasClamp {inRange {max=0, min=1}}})
  ]
 }
 AnimationTransitionGraph_1 {
  path=AnimGraph.AnimGraphNode_StateMachine_1.Glider_SurfStyle.AnimStateTransitionNode_1.AnimationTransitionGraph_1
  nodes [Result(bCanEnterTransition=OR Boolean(A=Get HasBeenSelected() -> (HasBeenSelected:bool), B=false))]
 }
 RandomLobbyIdle {
  entry=LobbyIdles
  path=AnimGraph.AnimGraphNode_StateMachine_2.RandomLobbyIdle
  ids=id
  nodes [AnimStateNode_0 = LobbyIdles() {is=title AnimStateNode}]
 }
 LobbyIdles {
  path=AnimGraph.AnimGraphNode_StateMachine_2.RandomLobbyIdle.AnimStateNode_0.LobbyIdles
  nodes [
   n1 = Get StartTime() -> (StartTime:double)
   n2 = Get PlayRate() -> (PlayRate:double)
   n3 = Blend Poses by int!(
    BlendPose_0:Pose Link Structure=Sequence Player 'Female_Large_Idle_02'!(PlayRate:float=@n2, StartPosition:float=@n1) -> (Pose) {sequence=/Game/Animation/Game/MainPlayer/Menu/BR/Female_Large_Idle_02}
    BlendPose_1:Pose Link Structure=Sequence Player 'Female_Large_Idle_01'!(PlayRate:float=@n2, StartPosition:float=@n1) -> (Pose) {sequence=/Game/Animation/Game/MainPlayer/Menu/BR/Female_Large_Idle_01}
    BlendTime_0:float=0.1
    BlendTime_1:float=0.1
    ActiveChildIndex:int=Get PoseInt() -> (PoseInt:int=0)
   ) -> (Pose:Pose Link Structure) {
    is=title AnimGraphNode_BlendListByInt
    blendType=Linear
    blendPose [{linkId=-1, sourceLinkId=-1}, {linkId=-1, sourceLinkId=-1}]
    blendTime [0.1,0.1]
   }
   n4 = Blend Poses by int(
    BlendPose_0=Sequence Player 'Male_Large_Idle_02'!(PlayRate:float=@n2, StartPosition:float=@n1) -> (Pose) {sequence=/Game/Animation/Game/MainPlayer/Menu/BR/Male_Large_Idle_02}
    BlendPose_1=Sequence Player 'Male_Large_Idle_01'!(PlayRate:float=@n2, StartPosition:float=@n1) -> (Pose) {sequence=/Game/Animation/Game/MainPlayer/Menu/BR/Male_Large_Idle_01}
    BlendTime_0=0.1
    BlendTime_1=0.1
    ActiveChildIndex=Get PoseInt() -> (PoseInt=0)
   ) {blendType=Linear, blendPose [{linkId=-1, sourceLinkId=-1}, {linkId=-1, sourceLinkId=-1}], blendTime [0.1,0.1]}
   n5 = "Blend Poses (EFortCustomGender)"(BlendPose_0=@n4, BlendPose_1=@n3, BlendTime_0=0.1, BlendTime_1=0.1, ActiveEnumValue=Get Gender() -> (Gender=Unknown)) {blendType=Linear, blendPose [{linkId=-1, sourceLinkId=-1}, {linkId=-1, sourceLinkId=-1}], blendTime [0.1,0.1]}
   n6 = Get StartTime()
   n7 = Get PlayRate()
   n8 = Blend Poses by int(
    BlendPose_0=Sequence Player 'Male_commando_Idle_2'!(PlayRate:float=@n7, StartPosition:float=@n6) -> (Pose) {sequence=/Game/Animation/Game/MainPlayer/Menu/BR/Male_commando_Idle_2}
    BlendPose_1=Sequence Player 'Male_commando_Idle_1'!(PlayRate:float=@n7, StartPosition:float=@n6) -> (Pose) {sequence=/Game/Animation/Game/MainPlayer/Menu/BR/Male_commando_Idle_1}
    BlendTime_0=0.1
    BlendTime_1=0.1
    ActiveChildIndex=Get PoseInt() -> (PoseInt=0)
   ) {blendType=Linear, blendPose [{linkId=-1, sourceLinkId=-1}, {linkId=-1, sourceLinkId=-1}], blendTime [0.1,0.1]}
   n9 = Blend Poses by int!(
    BlendPose_0=Sequence Player 'Female_Commando_Idle_02'!(PlayRate:float=@n7, StartPosition:float=@n6) -> (Pose) {sequence=/Game/Animation/Game/MainPlayer/Menu/BR/Female_Commando_Idle_02}
    BlendPose_1=Sequence Player 'Female_Commando_Idle_01'!(PlayRate:float=@n7, StartPosition:float=@n6) -> (Pose) {sequence=/Game/Animation/Game/MainPlayer/Menu/BR/Female_Commando_Idle_01}
    BlendPose_2:Pose Link Structure=Sequence Player 'Female_Commando_Idle_02_Rebirth'!(PlayRate:float=@n7, StartPosition:float=@n6) -> (Pose) {sequence=/Game/Animation/Game/MainPlayer/Menu/BR/Female_Commando_Idle_02_Rebirth}
    BlendTime_0=0.1
    BlendTime_1=0.1
    BlendTime_2:float=0.1
    ActiveChildIndex=Get PoseInt() -> (PoseInt=0)
   ) -> (Pose) {
    blendType=Linear
    blendPose [{linkId=-1, sourceLinkId=-1}, {linkId=-1, sourceLinkId=-1}, {linkId=-1, sourceLinkId=-1}]
    blendTime [0.1,0.1,0.1]
   }
   n10 = "Blend Poses (EFortCustomGender)"(BlendPose_0=@n8, BlendPose_1=@n9, BlendTime_0=0.1, BlendTime_1=0.1, ActiveEnumValue=Get Gender() -> (Gender=Unknown)) {blendType=Linear, blendPose [{linkId=-1, sourceLinkId=-1}, {linkId=-1, sourceLinkId=-1}], blendTime [0.1,0.1]}
   n11 = "Blend Poses (EFortPlayerAnimBodyType)"!(BlendPose_0=@n10, BlendPose_1=@n5, BlendTime_0=0.1, BlendTime_1=0.1, ActiveEnumValue=Get AnimBodyType() -> (AnimBodyType=Small)) -> (Pose) {blendType=Linear, blendPose [{linkId=-1, sourceLinkId=-1}, {linkId=-1, sourceLinkId=-1}], blendTime [0.1,0.1]}
   Output Animation Pose(Result=@n11)
  ]
 }
 Glider_or_Unmbrella {
  entry=ParaGlide_Into_Deploy
  path=AnimGraph.AnimGraphNode_StateMachine_3.Glider_or_Unmbrella
  ids=id
  nodes [
   AnimStateNode_1 = ParaGlide_Into_Deploy()
   AnimStateNode_3 = ParaGlide() {is=title AnimStateNode}
   AnimStateTransitionNode_2 = AnimStateTransitionNode() {bAutomaticRuleBasedOnSequencePlayerInState=true, crossfadeDuration=0}
  ]
  links [{
    dst=ParaGlide
    rule=AnimGraph.AnimGraphNode_StateMachine_3.Glider_or_Unmbrella.AnimStateTransitionNode_2.AnimationTransitionGraph_1
    src=ParaGlide_Into_Deploy
    via=AnimStateTransitionNode_2
   }]
 }
 ParaGlide_Into_Deploy {
  path=AnimGraph.AnimGraphNode_StateMachine_3.Glider_or_Unmbrella.AnimStateNode_1.ParaGlide_Into_Deploy
  nodes [
   n1 = Blendspace Player(
    X=Get Skydive Aim Yaw() -> (Skydive Aim Yaw:double)
    Y=Get Skydive Aim Pitch() -> (Skydive Aim Pitch:double)
    BlendSpace=Break Fort Anim Input Player Glider Anim Asset!(FortAnimInput_PlayerGliderAnimAsset=Get CurrentGliderAnimSet()) -> (
     Default_Base_BS:Blend Space Object Reference
     Default_BodyAdditive_MaleMedium_BS:Blend Space Object Reference
     Default_BodyAdditive_MaleLarge_BS:Blend Space Object Reference
     Default_BodyAdditive_FemaleSmall_BS:Blend Space Object Reference
     Default_BodyAdditive_FemaleMedium_BS
     Default_BodyAdditive_FemaleLarge_BS:Blend Space Object Reference
     Default_TurnAdditive_MaleMedium_BS:Blend Space Object Reference
     Default_TurnAdditive_MaleLarge_BS:Blend Space Object Reference
     Default_TurnAdditive_FemaleSmall_BS:Blend Space Object Reference
     Default_TurnAdditive_FemaleMedium_BS:Blend Space Object Reference
     Default_TurnAdditive_FemaleLarge_BS:Blend Space Object Reference
     Into_Base_BS:Blend Space Object Reference
     Into_Base_MaleMedium_BS
     Into_BodyAdditive_MaleMedium_BS:Blend Space Object Reference
     Into_BodyAdditive_MaleLarge_BS:Blend Space Object Reference
     Into_BodyAdditive_FemaleSmall_BS:Blend Space Object Reference
     Into_BodyAdditive_FemaleMedium_BS:Blend Space Object Reference
     Into_BodyAdditive_FemaleLarge_BS:Blend Space Object Reference
     Into_TurnAdditive_MaleMedium_BS:Blend Space Object Reference
     Into_TurnAdditive_MaleLarge_BS:Blend Space Object Reference
     Into_TurnAdditive_FemaleSmall_BS:Blend Space Object Reference
     Into_TurnAdditive_FemaleMedium_BS:Blend Space Object Reference
     Into_TurnAdditive_FemaleLarge_BS:Blend Space Object Reference
     Lean_MaleMedium_BS:Blend Space Object Reference
     Lean_MaleLarge_BS:Blend Space Object Reference
     Lean_FemaleSmall_BS:Blend Space Object Reference
     Lean_FemaleMedium_BS:Blend Space Object Reference
     Lean_FemaleLarge_BS:Blend Space Object Reference
     LeanAdditive_Center_MaleMedium_Pose:Anim Sequence Object Reference
     LeanAdditive_Center_MaleLarge_Pose:Anim Sequence Object Reference
     LeanAdditive_Center_FemaleSmall_Pose:Anim Sequence Object Reference
     LeanAdditive_Center_FemaleMedium_Pose:Anim Sequence Object Reference
     LeanAdditive_Center_FemaleLarge_Pose:Anim Sequence Object Reference
     LeanAdditive_Into_BS:Blend Space Object Reference
     LeanAdditive_ForwardInto_Anim:Anim Sequence Object Reference
     LeanAdditive_ForwardInto_FromDeploy_Anim:Anim Sequence Object Reference
     LeanAdditive_BackInto_Anim:Anim Sequence Object Reference
     LeanAdditive_LeftInto_Anim:Anim Sequence Object Reference
     LeanAdditive_RightInto_Anim:Anim Sequence Object Reference
     LeanAdditive_ForwardOut_Anim:Anim Sequence Object Reference
     LeanAdditive_BackOut_Anim:Anim Sequence Object Reference
     LeanAdditive_LeftOut_Anim:Anim Sequence Object Reference
     LeanAdditive_RightOut_Anim:Anim Sequence Object Reference
     ToGlide_BS:Blend Space Object Reference
     ToGlide_Lean_BS:Blend Space Object Reference
     ToDive_BS:Blend Space Object Reference
     ToDive_Lean_BS:Blend Space Object Reference
     Dive_WeaponR_Additive_BS:Blend Space Object Reference
     Glide_WeaponR_Additive_Anim:Anim Sequence Object Reference
     GenericAdditive_Male_BS:Blend Space Object Reference
     GenericAdditive_Female_BS:Blend Space Object Reference
     RootModPitchMin:float=~
     RootModPitchMax:float=~
     RootModYOffsetMin:float=~
     RootModYOffsetMax:float=~
     PelvisModVertical
     PelvisModLateral
     PelvisModBankAngle
     ThighModLegAngle
     ThighModLegPitch
     ThighModLegBank
     ThighModLeftLegDrag
     PlayerGliderType
     bEnableSpringMods=~
     bAllowPlayerDeployRootMod:bool=~
     bUseSurfStyle:bool=~
     bUseLeanPlayRateAdjust=~
     bUseForwardAs_Y_BS=~
     bUseDeployCurveForIntoPlayRateAdjust=~
     LeanForwardValue=~
     IdleValue=~
     LeanBackValue=~
     IdleInterpToSpeed=~
     LeanForwardInterpToSpeed=~
     LeanBackInterpToSpeed=~
    ).Into_Base_BS
   ) {bLoop=false}
   AnimGraphNode_StateMachine_1 = BodyAdditiveState() {is=title AnimGraphNode_StateMachine}
   AnimGraphNode_StateMachine_0 = IntoTurnPoseState() {is=title AnimGraphNode_StateMachine}
   Output Animation Pose(Result=Apply Additive!(
     Base=Apply Additive!(Base=@n1, Additive=@AnimGraphNode_StateMachine_1, AlphaCurveName:Name=ParaGlideLeanAlpha) -> (Pose) {alphaInputType=Curve, alphaBoolBlend {blendOption=Linear}}
     Additive=@AnimGraphNode_StateMachine_0
     Alpha:float
    ) -> (Pose) {alphaBoolBlend {blendOption=Linear}})
  ]
 }
 IntoTurnPoseState {
  entry=Turn
  path=AnimGraph.AnimGraphNode_StateMachine_3.Glider_or_Unmbrella.AnimStateNode_1.ParaGlide_Into_Deploy.AnimGraphNode_StateMachine_0.IntoTurnPoseState
  ids=id
  nodes [AnimStateNode_0 = Turn() {is=title AnimStateNode}]
 }
 Turn {
  path=AnimGraph.AnimGraphNode_StateMachine_3.Glider_or_Unmbrella.AnimStateNode_1.ParaGlide_Into_Deploy.AnimGraphNode_StateMachine_0.IntoTurnPoseState.AnimStateNode_0.Turn
  nodes [
   n1 = Break Fort Anim Input Player Glider Anim Asset(FortAnimInput_PlayerGliderAnimAsset=Get CurrentGliderAnimSet()) -> (
    RootModPitchMin=~
    RootModPitchMax=~
    RootModYOffsetMin=~
    RootModYOffsetMax=~
    bEnableSpringMods=~
    bAllowPlayerDeployRootMod=~
    bUseSurfStyle=~
    bUseLeanPlayRateAdjust=~
    bUseForwardAs_Y_BS=~
    bUseDeployCurveForIntoPlayRateAdjust=~
    LeanForwardValue=~
    IdleValue=~
    LeanBackValue=~
    IdleInterpToSpeed=~
    LeanForwardInterpToSpeed=~
    LeanBackInterpToSpeed=~
   )
   n2 = Break Fort Anim Input Skydiving(FortAnimInput_Skydiving=Get Skydiving()) -> (
    ParaGilderRootOffset=~
    ParaGliderRootRotationOverride=~
    bUseParaGlideRootModifier=~
    bIsSkydivingFromLaunchPad=~
    bIsSkydivingFromBus=~
    bIsInVortex=~
    bIsDiving=~
    bIsParachuteOpen=~
    bIsSkydiving=~
    bIsParachuteLeaning=~
    bIsSkydiveLeaning=~
    bIsLeaning=~
    bIsSkydiveDiveMode=~
    bPlaySkydiveDrift=~
    bSkydiveDriftDelayActive=~
    bSkydiveDriftAnimAllowed=~
    bIsGliderRight=~
    bIsGliderCenter=~
    bIsGliderLeft=~
    bIsGliderForward=~
    bIsGliderBack=~
    bHasDeployedAndIsStandingUpright=~
    LocalAccelForward=~
    LocalAccelRight=~
    LocalVelocityRight=~
    SkydiveAimPitch=~
    SkydiveAimYaw=~
    DeployChuteAnimRate=~
    SkydiveDriftAnimRate=~
    SkydiveDriftAnimRateCurrent=~
    SkydiveFidgetAnimRate=~
    SkydiveFidgetAnimRateCurrent=~
    SkydiveAdditiveAlpha=~
    SkydiveDriftDelay=~
   )
   n3 = "Blend Poses (EFortPlayerAnimBodyType)"(
    BlendPose_0=Blendspace Player(X=@n2.SkydiveAimYaw, Y=0, BlendSpace=@n1.Default_TurnAdditive_MaleMedium_BS) {bIgnoreForRelevancyTest=true}
    BlendPose_1=Blendspace Player(X=@n2.SkydiveAimYaw, Y=0, BlendSpace=@n1.Default_TurnAdditive_MaleLarge_BS) {bIgnoreForRelevancyTest=true}
    BlendTime_0=0.1
    BlendTime_1=0.1
    ActiveEnumValue=Get AnimBodyType() -> (AnimBodyType=Small)
   ) {blendType=Linear, blendPose [{linkId=-1, sourceLinkId=-1}, {linkId=-1, sourceLinkId=-1}], blendTime [0.1,0.1]}
   n4 = "Blend Poses (EFortPlayerAnimBodyType)"!(
    BlendPose_0=Blendspace Player(X=@n2.SkydiveAimYaw, Y=0, BlendSpace=@n1.Default_TurnAdditive_FemaleMedium_BS) {bIgnoreForRelevancyTest=true}
    BlendPose_1=Blendspace Player(X=@n2.SkydiveAimYaw, Y=0, BlendSpace=@n1.Default_TurnAdditive_FemaleSmall_BS) {bIgnoreForRelevancyTest=true}
    BlendPose_2:Pose Link Structure=Blendspace Player(X=@n2.SkydiveAimYaw, Y=0, BlendSpace=@n1.Default_TurnAdditive_FemaleLarge_BS) {bIgnoreForRelevancyTest=true}
    BlendTime_0=0.1
    BlendTime_1=0.1
    BlendTime_2:float=0.1
    ActiveEnumValue=Get AnimBodyType() -> (AnimBodyType=Small)
   ) -> (Pose) {
    blendType=Linear
    blendPose [{linkId=-1, sourceLinkId=-1}, {linkId=-1, sourceLinkId=-1}, {linkId=-1, sourceLinkId=-1}]
    blendTime [0.1,0.1,0.1]
   }
   n5 = "Blend Poses (EFortCustomGender)"(BlendPose_0=@n3, BlendPose_1=@n4, BlendTime_0=0, BlendTime_1=0, ActiveEnumValue=Get Gender() -> (Gender=Unknown)) {blendType=Linear, blendPose [{linkId=-1, sourceLinkId=-1}, {linkId=-1, sourceLinkId=-1}], blendTime [0.1,0.1]}
   n6 = Break Fort Anim Input Player Glider Anim Asset(FortAnimInput_PlayerGliderAnimAsset=Get CurrentGliderAnimSet()) -> (
    RootModPitchMin=~
    RootModPitchMax=~
    RootModYOffsetMin=~
    RootModYOffsetMax=~
    bEnableSpringMods=~
    bAllowPlayerDeployRootMod=~
    bUseSurfStyle=~
    bUseLeanPlayRateAdjust=~
    bUseForwardAs_Y_BS=~
    bUseDeployCurveForIntoPlayRateAdjust=~
    LeanForwardValue=~
    IdleValue=~
    LeanBackValue=~
    IdleInterpToSpeed=~
    LeanForwardInterpToSpeed=~
    LeanBackInterpToSpeed=~
   )
   n7 = Break Fort Anim Input Skydiving(FortAnimInput_Skydiving=Get Skydiving()) -> (
    ParaGilderRootOffset=~
    ParaGliderRootRotationOverride=~
    bUseParaGlideRootModifier=~
    bIsSkydivingFromLaunchPad=~
    bIsSkydivingFromBus=~
    bIsInVortex=~
    bIsDiving=~
    bIsParachuteOpen=~
    bIsSkydiving=~
    bIsParachuteLeaning=~
    bIsSkydiveLeaning=~
    bIsLeaning=~
    bIsSkydiveDiveMode=~
    bPlaySkydiveDrift=~
    bSkydiveDriftDelayActive=~
    bSkydiveDriftAnimAllowed=~
    bIsGliderRight=~
    bIsGliderCenter=~
    bIsGliderLeft=~
    bIsGliderForward=~
    bIsGliderBack=~
    bHasDeployedAndIsStandingUpright=~
    LocalAccelForward=~
    LocalAccelRight=~
    LocalVelocityRight=~
    SkydiveAimPitch=~
    SkydiveAimYaw=~
    DeployChuteAnimRate=~
    SkydiveDriftAnimRate=~
    SkydiveDriftAnimRateCurrent=~
    SkydiveFidgetAnimRate=~
    SkydiveFidgetAnimRateCurrent=~
    SkydiveAdditiveAlpha=~
    SkydiveDriftDelay=~
   )
   n8 = "Blend Poses (EFortPlayerAnimBodyType)"!(
    BlendPose_0=Blendspace Player(X=@n7.SkydiveAimYaw, Y=0, BlendSpace=@n6.Into_TurnAdditive_MaleMedium_BS) {bIgnoreForRelevancyTest=true}
    BlendPose_1=Blendspace Player(X=@n7.SkydiveAimYaw, Y=0, BlendSpace=@n6.Into_TurnAdditive_MaleLarge_BS) {bIgnoreForRelevancyTest=true}
    BlendTime_0=0.1
    BlendTime_1=0.1
    ActiveEnumValue=Get AnimBodyType() -> (AnimBodyType=Small)
   ) -> (Pose) {blendType=Linear, blendPose [{linkId=-1, sourceLinkId=-1}, {linkId=-1, sourceLinkId=-1}], blendTime [0.1,0.1]}
   n9 = "Blend Poses (EFortPlayerAnimBodyType)"!(
    BlendPose_0=Blendspace Player(X=@n7.SkydiveAimYaw, Y=0, BlendSpace=@n6.Into_TurnAdditive_FemaleMedium_BS) {bIgnoreForRelevancyTest=true}
    BlendPose_1=Blendspace Player(X=@n7.SkydiveAimYaw, Y=0, BlendSpace=@n6.Into_TurnAdditive_FemaleSmall_BS) {bIgnoreForRelevancyTest=true}
    BlendPose_2:Pose Link Structure=Blendspace Player(X=@n7.SkydiveAimYaw, Y=0, BlendSpace=@n6.Into_TurnAdditive_FemaleLarge_BS) {bIgnoreForRelevancyTest=true}
    BlendTime_0=0.1
    BlendTime_1=0.1
    BlendTime_2:float=0.1
    ActiveEnumValue=Get AnimBodyType() -> (AnimBodyType=Small)
   ) -> (Pose) {
    blendType=Linear
    blendPose [{linkId=-1, sourceLinkId=-1}, {linkId=-1, sourceLinkId=-1}, {linkId=-1, sourceLinkId=-1}]
    blendTime [0.1,0.1,0.1]
   }
   n10 = "Blend Poses (EFortCustomGender)"(BlendPose_0=@n8, BlendPose_1=@n9, BlendTime_0=0, BlendTime_1=0, ActiveEnumValue=Get Gender() -> (Gender=Unknown)) {blendType=Linear, blendPose [{linkId=-1, sourceLinkId=-1}, {linkId=-1, sourceLinkId=-1}], blendTime [0.1,0.1]}
   Output Animation Pose(Result=Two Way Blend!(A=@n10, B=@n5, AlphaCurveName:Name=BlendIn) -> (Pose) {blendNode {alphaInputType=Curve, alphaBoolBlend {blendOption=Linear}, alphaScaleBiasClamp {bInterpResult=true}}})
  ]
 }
 BodyAdditiveState {
  entry=BoddyAdditive
  path=AnimGraph.AnimGraphNode_StateMachine_3.Glider_or_Unmbrella.AnimStateNode_1.ParaGlide_Into_Deploy.AnimGraphNode_StateMachine_1.BodyAdditiveState
  ids=id
  nodes [AnimStateNode_0 = BoddyAdditive() {is=title AnimStateNode}]
 }
 BoddyAdditive {
  path=AnimGraph.AnimGraphNode_StateMachine_3.Glider_or_Unmbrella.AnimStateNode_1.ParaGlide_Into_Deploy.AnimGraphNode_StateMachine_1.BodyAdditiveState.AnimStateNode_0.BoddyAdditive
  nodes [
   n1 = Break Fort Anim Input Skydiving(FortAnimInput_Skydiving=Get Skydiving()) -> (
    ParaGilderRootOffset=~
    ParaGliderRootRotationOverride=~
    bUseParaGlideRootModifier=~
    bIsSkydivingFromLaunchPad=~
    bIsSkydivingFromBus=~
    bIsInVortex=~
    bIsDiving=~
    bIsParachuteOpen=~
    bIsSkydiving=~
    bIsParachuteLeaning=~
    bIsSkydiveLeaning=~
    bIsLeaning=~
    bIsSkydiveDiveMode=~
    bPlaySkydiveDrift=~
    bSkydiveDriftDelayActive=~
    bSkydiveDriftAnimAllowed=~
    bIsGliderRight=~
    bIsGliderCenter=~
    bIsGliderLeft=~
    bIsGliderForward=~
    bIsGliderBack=~
    bHasDeployedAndIsStandingUpright=~
    LocalAccelForward=~
    LocalAccelRight=~
    LocalVelocityRight=~
    SkydiveAimPitch=~
    SkydiveAimYaw=~
    DeployChuteAnimRate=~
    SkydiveDriftAnimRate=~
    SkydiveDriftAnimRateCurrent=~
    SkydiveFidgetAnimRate=~
    SkydiveFidgetAnimRateCurrent=~
    SkydiveAdditiveAlpha=~
    SkydiveDriftDelay=~
   )
   n2 = Break Fort Anim Input Player Glider Anim Asset(FortAnimInput_PlayerGliderAnimAsset=Get CurrentGliderAnimSet()) -> (
    RootModPitchMin=~
    RootModPitchMax=~
    RootModYOffsetMin=~
    RootModYOffsetMax=~
    bEnableSpringMods=~
    bAllowPlayerDeployRootMod=~
    bUseSurfStyle=~
    bUseLeanPlayRateAdjust=~
    bUseForwardAs_Y_BS=~
    bUseDeployCurveForIntoPlayRateAdjust=~
    LeanForwardValue=~
    IdleValue=~
    LeanBackValue=~
    IdleInterpToSpeed=~
    LeanForwardInterpToSpeed=~
    LeanBackInterpToSpeed=~
   )
   n3 = "Blend Poses (EFortPlayerAnimBodyType)"!(
    BlendPose_0=Blendspace Player(X=@n1.SkydiveAimYaw, Y=0, BlendSpace=@n2.Into_BodyAdditive_MaleMedium_BS) {bIgnoreForRelevancyTest=true}
    BlendPose_1=Blendspace Player(X=@n1.SkydiveAimYaw, Y=0, BlendSpace=@n2.Into_BodyAdditive_MaleLarge_BS) {bIgnoreForRelevancyTest=true}
    BlendTime_0=0.1
    BlendTime_1=0.1
    ActiveEnumValue=Get AnimBodyType() -> (AnimBodyType=Small)
   ) -> (Pose) {blendType=Linear, blendPose [{linkId=-1, sourceLinkId=-1}, {linkId=-1, sourceLinkId=-1}], blendTime [0.1,0.1]}
   n4 = "Blend Poses (EFortPlayerAnimBodyType)"!(
    BlendPose_0=Blendspace Player(X=@n1.SkydiveAimYaw, Y=0, BlendSpace=@n2.Into_BodyAdditive_FemaleMedium_BS) {bIgnoreForRelevancyTest=true}
    BlendPose_1=Blendspace Player(X=@n1.SkydiveAimYaw, Y=0, BlendSpace=@n2.Into_BodyAdditive_FemaleSmall_BS) {bIgnoreForRelevancyTest=true}
    BlendPose_2:Pose Link Structure=Blendspace Player(X=@n1.SkydiveAimYaw, Y=0, BlendSpace=@n2.Into_BodyAdditive_FemaleLarge_BS) {bIgnoreForRelevancyTest=true}
    BlendTime_0=0.1
    BlendTime_1=0.1
    BlendTime_2:float=0.1
    ActiveEnumValue=Get AnimBodyType() -> (AnimBodyType=Small)
   ) -> (Pose) {
    blendType=Linear
    blendPose [{linkId=-1, sourceLinkId=-1}, {linkId=-1, sourceLinkId=-1}, {linkId=-1, sourceLinkId=-1}]
    blendTime [0.1,0.1,0.1]
   }
   n5 = "Blend Poses (EFortCustomGender)"(BlendPose_0=@n3, BlendPose_1=@n4, BlendTime_0=0, BlendTime_1=0, ActiveEnumValue=Get Gender() -> (Gender=Unknown)) {blendType=Linear, blendPose [{linkId=-1, sourceLinkId=-1}, {linkId=-1, sourceLinkId=-1}], blendTime [0.1,0.1]}
   Output Animation Pose(Result=@n5)
  ]
 }
 ParaGlide {
  path=AnimGraph.AnimGraphNode_StateMachine_3.Glider_or_Unmbrella.AnimStateNode_3.ParaGlide
  nodes [
   n1 = Blendspace Player(
    X=Break Fort Anim Input Skydiving(FortAnimInput_Skydiving=Get Skydiving()) -> (
     ParaGilderRootOffset=~
     ParaGliderRootRotationOverride=~
     bUseParaGlideRootModifier=~
     bIsSkydivingFromLaunchPad=~
     bIsSkydivingFromBus=~
     bIsInVortex=~
     bIsDiving=~
     bIsParachuteOpen=~
     bIsSkydiving=~
     bIsParachuteLeaning=~
     bIsSkydiveLeaning=~
     bIsLeaning=~
     bIsSkydiveDiveMode=~
     bPlaySkydiveDrift=~
     bSkydiveDriftDelayActive=~
     bSkydiveDriftAnimAllowed=~
     bIsGliderRight=~
     bIsGliderCenter=~
     bIsGliderLeft=~
     bIsGliderForward=~
     bIsGliderBack=~
     bHasDeployedAndIsStandingUpright=~
     LocalAccelForward=~
     LocalAccelRight=~
     LocalVelocityRight=~
     SkydiveAimPitch=~
     SkydiveAimYaw=~
     DeployChuteAnimRate=~
     SkydiveDriftAnimRate=~
     SkydiveDriftAnimRateCurrent=~
     SkydiveFidgetAnimRate=~
     SkydiveFidgetAnimRateCurrent=~
     SkydiveAdditiveAlpha=~
     SkydiveDriftDelay=~
    ).SkydiveAimYaw
    Y=0
    BlendSpace=Break Fort Anim Input Player Glider Anim Asset(FortAnimInput_PlayerGliderAnimAsset=Get CurrentGliderAnimSet()) -> (
     RootModPitchMin=~
     RootModPitchMax=~
     RootModYOffsetMin=~
     RootModYOffsetMax=~
     bEnableSpringMods=~
     bAllowPlayerDeployRootMod=~
     bUseSurfStyle=~
     bUseLeanPlayRateAdjust=~
     bUseForwardAs_Y_BS=~
     bUseDeployCurveForIntoPlayRateAdjust=~
     LeanForwardValue=~
     IdleValue=~
     LeanBackValue=~
     IdleInterpToSpeed=~
     LeanForwardInterpToSpeed=~
     LeanBackInterpToSpeed=~
    )
   )
   AnimGraphNode_StateMachine_0 = BodyAdditiveState()
   AnimGraphNode_StateMachine_1 = TurnAdditiveState() {is=title AnimGraphNode_StateMachine}
   Output Animation Pose(Result=Apply Additive!(
     Base=Apply Additive(Base=@n1, Additive=@AnimGraphNode_StateMachine_0) {alphaBoolBlend {blendOption=Linear}}
     Additive=@AnimGraphNode_StateMachine_1
     Alpha:float
    ) -> (Pose) {alphaBoolBlend {blendOption=Linear}})
  ]
 }
 BodyAdditiveState {
  entry=BodyAdditive
  path=AnimGraph.AnimGraphNode_StateMachine_3.Glider_or_Unmbrella.AnimStateNode_3.ParaGlide.AnimGraphNode_StateMachine_0.BodyAdditiveState
  ids=id
  nodes [AnimStateNode_0 = BodyAdditive() {is=title AnimStateNode}]
 }
 BodyAdditive {
  path=AnimGraph.AnimGraphNode_StateMachine_3.Glider_or_Unmbrella.AnimStateNode_3.ParaGlide.AnimGraphNode_StateMachine_0.BodyAdditiveState.AnimStateNode_0.BodyAdditive
  nodes [
   n1 = Break Fort Anim Input Skydiving(FortAnimInput_Skydiving=Get Skydiving()) -> (
    ParaGilderRootOffset=~
    ParaGliderRootRotationOverride=~
    bUseParaGlideRootModifier=~
    bIsSkydivingFromLaunchPad=~
    bIsSkydivingFromBus=~
    bIsInVortex=~
    bIsDiving=~
    bIsParachuteOpen=~
    bIsSkydiving=~
    bIsParachuteLeaning=~
    bIsSkydiveLeaning=~
    bIsLeaning=~
    bIsSkydiveDiveMode=~
    bPlaySkydiveDrift=~
    bSkydiveDriftDelayActive=~
    bSkydiveDriftAnimAllowed=~
    bIsGliderRight=~
    bIsGliderCenter=~
    bIsGliderLeft=~
    bIsGliderForward=~
    bIsGliderBack=~
    bHasDeployedAndIsStandingUpright=~
    LocalAccelForward=~
    LocalAccelRight=~
    LocalVelocityRight=~
    SkydiveAimPitch=~
    SkydiveAimYaw=~
    DeployChuteAnimRate=~
    SkydiveDriftAnimRate=~
    SkydiveDriftAnimRateCurrent=~
    SkydiveFidgetAnimRate=~
    SkydiveFidgetAnimRateCurrent=~
    SkydiveAdditiveAlpha=~
    SkydiveDriftDelay=~
   )
   n2 = Break Fort Anim Input Player Glider Anim Asset(FortAnimInput_PlayerGliderAnimAsset=Get CurrentGliderAnimSet()) -> (
    RootModPitchMin=~
    RootModPitchMax=~
    RootModYOffsetMin=~
    RootModYOffsetMax=~
    bEnableSpringMods=~
    bAllowPlayerDeployRootMod=~
    bUseSurfStyle=~
    bUseLeanPlayRateAdjust=~
    bUseForwardAs_Y_BS=~
    bUseDeployCurveForIntoPlayRateAdjust=~
    LeanForwardValue=~
    IdleValue=~
    LeanBackValue=~
    IdleInterpToSpeed=~
    LeanForwardInterpToSpeed=~
    LeanBackInterpToSpeed=~
   )
   n3 = "Blend Poses (EFortPlayerAnimBodyType)"!(
    BlendPose_0=Blendspace Player(X=@n1.SkydiveAimYaw, Y=0, BlendSpace=@n2.Default_BodyAdditive_MaleMedium_BS)
    BlendPose_1=Blendspace Player(X=@n1.SkydiveAimYaw, Y=0, BlendSpace=@n2.Default_BodyAdditive_MaleLarge_BS)
    BlendTime_0=0.1
    BlendTime_1=0.1
    ActiveEnumValue=Get AnimBodyType() -> (AnimBodyType=Small)
   ) -> (Pose) {blendType=Linear, blendPose [{linkId=-1, sourceLinkId=-1}, {linkId=-1, sourceLinkId=-1}], blendTime [0.1,0.1]}
   n4 = "Blend Poses (EFortPlayerAnimBodyType)"!(
    BlendPose_0=Blendspace Player(X=@n1.SkydiveAimYaw, Y=0, BlendSpace=@n2.Default_BodyAdditive_FemaleMedium_BS)
    BlendPose_1=Blendspace Player(X=@n1.SkydiveAimYaw, Y=0, BlendSpace=@n2.Default_BodyAdditive_FemaleSmall_BS)
    BlendPose_2:Pose Link Structure=Blendspace Player(X=@n1.SkydiveAimYaw, Y=0, BlendSpace=@n2.Default_BodyAdditive_FemaleLarge_BS)
    BlendTime_0=0.1
    BlendTime_1=0.1
    BlendTime_2:float=0.1
    ActiveEnumValue=Get AnimBodyType() -> (AnimBodyType=Small)
   ) -> (Pose) {
    blendType=Linear
    blendPose [{linkId=-1, sourceLinkId=-1}, {linkId=-1, sourceLinkId=-1}, {linkId=-1, sourceLinkId=-1}]
    blendTime [0.1,0.1,0.1]
   }
   n5 = "Blend Poses (EFortCustomGender)"(BlendPose_0=@n3, BlendPose_1=@n4, BlendTime_0=0, BlendTime_1=0, ActiveEnumValue=Get Gender() -> (Gender=Unknown)) {blendType=Linear, blendPose [{linkId=-1, sourceLinkId=-1}, {linkId=-1, sourceLinkId=-1}], blendTime [0.1,0.1]}
   Output Animation Pose(Result=@n5)
  ]
 }
 TurnAdditiveState {
  entry=TurnAdditive
  path=AnimGraph.AnimGraphNode_StateMachine_3.Glider_or_Unmbrella.AnimStateNode_3.ParaGlide.AnimGraphNode_StateMachine_1.TurnAdditiveState
  ids=id
  nodes [AnimStateNode_0 = TurnAdditive() {is=title AnimStateNode}]
 }
 TurnAdditive {
  path=AnimGraph.AnimGraphNode_StateMachine_3.Glider_or_Unmbrella.AnimStateNode_3.ParaGlide.AnimGraphNode_StateMachine_1.TurnAdditiveState.AnimStateNode_0.TurnAdditive
  nodes [
   n1 = Break Fort Anim Input Skydiving(FortAnimInput_Skydiving=Get Skydiving()) -> (
    ParaGilderRootOffset=~
    ParaGliderRootRotationOverride=~
    bUseParaGlideRootModifier=~
    bIsSkydivingFromLaunchPad=~
    bIsSkydivingFromBus=~
    bIsInVortex=~
    bIsDiving=~
    bIsParachuteOpen=~
    bIsSkydiving=~
    bIsParachuteLeaning=~
    bIsSkydiveLeaning=~
    bIsLeaning=~
    bIsSkydiveDiveMode=~
    bPlaySkydiveDrift=~
    bSkydiveDriftDelayActive=~
    bSkydiveDriftAnimAllowed=~
    bIsGliderRight=~
    bIsGliderCenter=~
    bIsGliderLeft=~
    bIsGliderForward=~
    bIsGliderBack=~
    bHasDeployedAndIsStandingUpright=~
    LocalAccelForward=~
    LocalAccelRight=~
    LocalVelocityRight=~
    SkydiveAimPitch=~
    SkydiveAimYaw=~
    DeployChuteAnimRate=~
    SkydiveDriftAnimRate=~
    SkydiveDriftAnimRateCurrent=~
    SkydiveFidgetAnimRate=~
    SkydiveFidgetAnimRateCurrent=~
    SkydiveAdditiveAlpha=~
    SkydiveDriftDelay=~
   )
   n2 = Break Fort Anim Input Player Glider Anim Asset(FortAnimInput_PlayerGliderAnimAsset=Get CurrentGliderAnimSet()) -> (
    RootModPitchMin=~
    RootModPitchMax=~
    RootModYOffsetMin=~
    RootModYOffsetMax=~
    bEnableSpringMods=~
    bAllowPlayerDeployRootMod=~
    bUseSurfStyle=~
    bUseLeanPlayRateAdjust=~
    bUseForwardAs_Y_BS=~
    bUseDeployCurveForIntoPlayRateAdjust=~
    LeanForwardValue=~
    IdleValue=~
    LeanBackValue=~
    IdleInterpToSpeed=~
    LeanForwardInterpToSpeed=~
    LeanBackInterpToSpeed=~
   )
   n3 = "Blend Poses (EFortPlayerAnimBodyType)"!(
    BlendPose_0=Blendspace Player(X=@n1.SkydiveAimYaw, Y=0, BlendSpace=@n2.Default_TurnAdditive_MaleMedium_BS)
    BlendPose_1=Blendspace Player(X=@n1.SkydiveAimYaw, Y=0, BlendSpace=@n2.Default_TurnAdditive_MaleLarge_BS)
    BlendTime_0=0.1
    BlendTime_1=0.1
    ActiveEnumValue=Get AnimBodyType() -> (AnimBodyType=Small)
   ) -> (Pose) {blendType=Linear, blendPose [{linkId=-1, sourceLinkId=-1}, {linkId=-1, sourceLinkId=-1}], blendTime [0.1,0.1]}
   n4 = "Blend Poses (EFortPlayerAnimBodyType)"!(
    BlendPose_0=Blendspace Player(X=@n1.SkydiveAimYaw, Y=0, BlendSpace=@n2.Default_TurnAdditive_FemaleMedium_BS)
    BlendPose_1=Blendspace Player(X=@n1.SkydiveAimYaw, Y=0, BlendSpace=@n2.Default_TurnAdditive_FemaleSmall_BS)
    BlendPose_2:Pose Link Structure=Blendspace Player(X=@n1.SkydiveAimYaw, Y=0, BlendSpace=@n2.Default_TurnAdditive_FemaleLarge_BS)
    BlendTime_0=0.1
    BlendTime_1=0.1
    BlendTime_2:float=0.1
    ActiveEnumValue=Get AnimBodyType() -> (AnimBodyType=Small)
   ) -> (Pose) {
    blendType=Linear
    blendPose [{linkId=-1, sourceLinkId=-1}, {linkId=-1, sourceLinkId=-1}, {linkId=-1, sourceLinkId=-1}]
    blendTime [0.1,0.1,0.1]
   }
   n5 = "Blend Poses (EFortCustomGender)"(BlendPose_0=@n3, BlendPose_1=@n4, BlendTime_0=0, BlendTime_1=0, ActiveEnumValue=Get Gender() -> (Gender=Unknown)) {blendType=Linear, blendPose [{linkId=-1, sourceLinkId=-1}, {linkId=-1, sourceLinkId=-1}], blendTime [0.1,0.1]}
   Output Animation Pose(Result=@n5)
  ]
 }
 AnimationTransitionGraph_1 {path=AnimGraph.AnimGraphNode_StateMachine_3.Glider_or_Unmbrella.AnimStateTransitionNode_2.AnimationTransitionGraph_1, nodes [Result()]}
 NoGlider {
  entry=ParaGlide_Into_Deploy
  path=AnimGraph.AnimGraphNode_StateMachine_73.NoGlider
  ids=id
  nodes [
   AnimStateNode_1 = ParaGlide_Into_Deploy()
   AnimStateNode_3 = ParaGlide()
   AnimStateTransitionNode_2 = AnimStateTransitionNode() {bAutomaticRuleBasedOnSequencePlayerInState=true, crossfadeDuration=0}
  ]
  links [{
    dst=ParaGlide
    rule=AnimGraph.AnimGraphNode_StateMachine_73.NoGlider.AnimStateTransitionNode_2.AnimationTransitionGraph_1
    src=ParaGlide_Into_Deploy
    via=AnimStateTransitionNode_2
   }]
 }
 ParaGlide_Into_Deploy {
  path=AnimGraph.AnimGraphNode_StateMachine_73.NoGlider.AnimStateNode_1.ParaGlide_Into_Deploy
  nodes [
   n1 = Break Fort Anim Input Skydiving!(FortAnimInput_Skydiving=Get Skydiving()) -> (bHasDeployedAndIsStandingUpright=~, SkydiveAimPitch=~, SkydiveAimYaw=~, PoseAdditive)
   n2 = Blendspace Player(
    X=@n1.SkydiveAimYaw
    Y=@n1.SkydiveAimPitch
    BlendSpace=Break Fort Anim Input Player Glider Anim Asset!(FortAnimInput_PlayerGliderAnimAsset=Get CurrentGliderAnimSet()) -> (
     Into_Base_BS
     Into_Base_MaleMedium_BS
     bUseLeanPlayRateAdjust=~
     bUseForwardAs_Y_BS=~
     bUseDeployCurveForIntoPlayRateAdjust=~
     LeanForwardValue=~
     IdleValue=~
     LeanBackValue=~
     IdleInterpToSpeed=~
     LeanForwardInterpToSpeed=~
     LeanBackInterpToSpeed=~
    )
   ) {bLoop=false}
   AnimGraphNode_StateMachine_1 = BodyAdditiveState()
   AnimGraphNode_StateMachine_0 = TurnAdditiveState()
   Output Animation Pose(Result=Apply Additive(
     Base=Apply Additive(Base=@n2, Additive=@AnimGraphNode_StateMachine_1) {alphaBoolBlend {blendOption=Linear}}
     Additive=@AnimGraphNode_StateMachine_0
    ) {alphaBoolBlend {blendOption=Linear}})
  ]
 }
 TurnAdditiveState {
  entry=State
  path=AnimGraph.AnimGraphNode_StateMachine_73.NoGlider.AnimStateNode_1.ParaGlide_Into_Deploy.AnimGraphNode_StateMachine_0.TurnAdditiveState
  ids=id
  nodes [AnimStateNode_0 = State() {is=title AnimStateNode}]
 }
 State {
  path=AnimGraph.AnimGraphNode_StateMachine_73.NoGlider.AnimStateNode_1.ParaGlide_Into_Deploy.AnimGraphNode_StateMachine_0.TurnAdditiveState.AnimStateNode_0.State
  nodes [
   n1 = Break Fort Anim Input Player Glider Anim Asset!(FortAnimInput_PlayerGliderAnimAsset=Get CurrentGliderAnimSet()) -> (
    Default_Base_BS:Blend Space Object Reference
    Default_BodyAdditive_MaleMedium_BS:Blend Space Object Reference
    Default_BodyAdditive_MaleLarge_BS:Blend Space Object Reference
    Default_BodyAdditive_FemaleSmall_BS:Blend Space Object Reference
    Default_BodyAdditive_FemaleMedium_BS:Blend Space Object Reference
    Default_BodyAdditive_FemaleLarge_BS:Blend Space Object Reference
    Default_TurnAdditive_MaleMedium_BS:Blend Space Object Reference
    Default_TurnAdditive_MaleLarge_BS:Blend Space Object Reference
    Default_TurnAdditive_FemaleSmall_BS:Blend Space Object Reference
    Default_TurnAdditive_FemaleMedium_BS:Blend Space Object Reference
    Default_TurnAdditive_FemaleLarge_BS:Blend Space Object Reference
    Into_Base_BS
    Into_Base_MaleMedium_BS
    Into_BodyAdditive_MaleMedium_BS:Blend Space Object Reference
    Into_BodyAdditive_MaleLarge_BS:Blend Space Object Reference
    Into_BodyAdditive_FemaleSmall_BS:Blend Space Object Reference
    Into_BodyAdditive_FemaleMedium_BS:Blend Space Object Reference
    Into_BodyAdditive_FemaleLarge_BS:Blend Space Object Reference
    Into_TurnAdditive_MaleMedium_BS:Blend Space Object Reference
    Into_TurnAdditive_MaleLarge_BS:Blend Space Object Reference
    Into_TurnAdditive_FemaleSmall_BS:Blend Space Object Reference
    Into_TurnAdditive_FemaleMedium_BS:Blend Space Object Reference
    Into_TurnAdditive_FemaleLarge_BS:Blend Space Object Reference
    Lean_MaleMedium_BS:Blend Space Object Reference
    Lean_MaleLarge_BS:Blend Space Object Reference
    Lean_FemaleSmall_BS:Blend Space Object Reference
    Lean_FemaleMedium_BS:Blend Space Object Reference
    Lean_FemaleLarge_BS:Blend Space Object Reference
    LeanAdditive_Center_MaleMedium_Pose:Anim Sequence Object Reference
    LeanAdditive_Center_MaleLarge_Pose:Anim Sequence Object Reference
    LeanAdditive_Center_FemaleSmall_Pose:Anim Sequence Object Reference
    LeanAdditive_Center_FemaleMedium_Pose:Anim Sequence Object Reference
    LeanAdditive_Center_FemaleLarge_Pose:Anim Sequence Object Reference
    LeanAdditive_Into_BS:Blend Space Object Reference
    LeanAdditive_ForwardInto_Anim:Anim Sequence Object Reference
    LeanAdditive_ForwardInto_FromDeploy_Anim:Anim Sequence Object Reference
    LeanAdditive_BackInto_Anim:Anim Sequence Object Reference
    LeanAdditive_LeftInto_Anim:Anim Sequence Object Reference
    LeanAdditive_RightInto_Anim:Anim Sequence Object Reference
    LeanAdditive_ForwardOut_Anim:Anim Sequence Object Reference
    LeanAdditive_BackOut_Anim:Anim Sequence Object Reference
    LeanAdditive_LeftOut_Anim:Anim Sequence Object Reference
    LeanAdditive_RightOut_Anim:Anim Sequence Object Reference
    ToGlide_BS:Blend Space Object Reference
    ToGlide_Lean_BS:Blend Space Object Reference
    ToDive_BS:Blend Space Object Reference
    ToDive_Lean_BS:Blend Space Object Reference
    Dive_WeaponR_Additive_BS:Blend Space Object Reference
    Glide_WeaponR_Additive_Anim:Anim Sequence Object Reference
    GenericAdditive_Male_BS:Blend Space Object Reference
    GenericAdditive_Female_BS:Blend Space Object Reference
    RootModPitchMin:float=~
    RootModPitchMax:float=~
    RootModYOffsetMin:float=~
    RootModYOffsetMax:float=~
    PelvisModVertical:Pelvis Mod Vertical Input Spring Structure
    PelvisModLateral:Pelvis Mod Lateral Input Spring Structure
    PelvisModBankAngle:Pelvis Mod Bank Angle Input Spring Structure
    ThighModLegAngle:Thigh Mod Leg Angle Input Structure
    ThighModLegPitch:Thigh Mod Leg Pitch Spring Structure
    ThighModLegBank:Thigh Mod Leg Bank Spring Structure
    ThighModLeftLegDrag:Thigh Mod Left Leg Drag Structure
    PlayerGliderType:EGliderType Enum
    bEnableSpringMods:bool=~
    bAllowPlayerDeployRootMod:bool=~
    bUseSurfStyle:bool=~
    bUseLeanPlayRateAdjust=~
    bUseForwardAs_Y_BS=~
    bUseDeployCurveForIntoPlayRateAdjust=~
    LeanForwardValue=~
    IdleValue=~
    LeanBackValue=~
    IdleInterpToSpeed=~
    LeanForwardInterpToSpeed=~
    LeanBackInterpToSpeed=~
   )
   n2 = Break Fort Anim Input Skydiving!(FortAnimInput_Skydiving=Get Skydiving()) -> (
    ParaGilderRootOffset:Vector=~
    ParaGliderRootRotationOverride:Rotator=~
    bUseParaGlideRootModifier:bool=~
    bIsSkydivingFromLaunchPad:bool=~
    bIsSkydivingFromBus:bool=~
    bIsInVortex:bool=~
    bIsDiving:bool=~
    bIsParachuteOpen:bool=~
    bIsSkydiving:bool=~
    bIsParachuteLeaning:bool=~
    bIsSkydiveLeaning:bool=~
    bIsLeaning:bool=~
    bIsSkydiveDiveMode:bool=~
    bPlaySkydiveDrift:bool=~
    bSkydiveDriftDelayActive:bool=~
    bSkydiveDriftAnimAllowed:bool=~
    bIsGliderRight:bool=~
    bIsGliderCenter:bool=~
    bIsGliderLeft:bool=~
    bIsGliderForward:bool=~
    bIsGliderBack:bool=~
    bHasDeployedAndIsStandingUpright=~
    LocalAccelForward:float=~
    LocalAccelRight:float=~
    LocalVelocityRight:float=~
    SkydiveAimPitch=~
    SkydiveAimYaw=~
    DeployChuteAnimRate:float=~
    SkydiveDriftAnimRate:float=~
    SkydiveDriftAnimRateCurrent:float=~
    SkydiveFidgetAnimRate:float=~
    SkydiveFidgetAnimRateCurrent:float=~
    SkydiveAdditiveAlpha:float=~
    SkydiveDriftDelay:float=~
    SkydiveDriftAnim:int
    SkydiveDriftAnimMax:int
    LaunchpadAnim:int
    LocalAccelDir:ESkydivingDirection Enum
    DirectionLast:ESkydivingDirection Enum
    PoseAdditive
   )
   n3 = "Blend Poses (EFortPlayerAnimBodyType)"!(
    BlendPose_0=Blendspace Player(X=@n2.SkydiveAimYaw, Y=0, BlendSpace=@n1.Default_TurnAdditive_MaleMedium_BS) {bIgnoreForRelevancyTest=true}
    BlendPose_1=Blendspace Player(X=@n2.SkydiveAimYaw, Y=0, BlendSpace=@n1.Default_TurnAdditive_MaleLarge_BS) {bIgnoreForRelevancyTest=true}
    BlendTime_0=0.1
    BlendTime_1=0.1
    ActiveEnumValue=Get AnimBodyType() -> (AnimBodyType=Small)
   ) -> (Pose) {blendType=Linear, blendPose [{linkId=-1, sourceLinkId=-1}, {linkId=-1, sourceLinkId=-1}], blendTime [0.1,0.1]}
   n4 = "Blend Poses (EFortPlayerAnimBodyType)"!(
    BlendPose_0=Blendspace Player(X=@n2.SkydiveAimYaw, Y=0, BlendSpace=@n1.Default_TurnAdditive_FemaleMedium_BS) {bIgnoreForRelevancyTest=true}
    BlendPose_1=Blendspace Player(X=@n2.SkydiveAimYaw, Y=0, BlendSpace=@n1.Default_TurnAdditive_FemaleSmall_BS) {bIgnoreForRelevancyTest=true}
    BlendPose_2:Pose Link Structure=Blendspace Player(X=@n2.SkydiveAimYaw, Y=0, BlendSpace=@n1.Default_TurnAdditive_FemaleLarge_BS) {bIgnoreForRelevancyTest=true}
    BlendTime_0=0.1
    BlendTime_1=0.1
    BlendTime_2:float=0.1
    ActiveEnumValue=Get AnimBodyType() -> (AnimBodyType=Small)
   ) -> (Pose) {
    blendType=Linear
    blendPose [{linkId=-1, sourceLinkId=-1}, {linkId=-1, sourceLinkId=-1}, {linkId=-1, sourceLinkId=-1}]
    blendTime [0.1,0.1,0.1]
   }
   n5 = "Blend Poses (EFortCustomGender)"(BlendPose_0=@n3, BlendPose_1=@n4, BlendTime_0=0, BlendTime_1=0, ActiveEnumValue=Get Gender() -> (Gender=Unknown)) {blendType=Linear, blendPose [{linkId=-1, sourceLinkId=-1}, {linkId=-1, sourceLinkId=-1}], blendTime [0.1,0.1]}
   Output Animation Pose(Result=@n5)
  ]
 }
 BodyAdditiveState {
  entry=BodyAdditive
  path=AnimGraph.AnimGraphNode_StateMachine_73.NoGlider.AnimStateNode_1.ParaGlide_Into_Deploy.AnimGraphNode_StateMachine_1.BodyAdditiveState
  ids=id
  nodes [AnimStateNode_0 = BodyAdditive()]
 }
 BodyAdditive {
  path=AnimGraph.AnimGraphNode_StateMachine_73.NoGlider.AnimStateNode_1.ParaGlide_Into_Deploy.AnimGraphNode_StateMachine_1.BodyAdditiveState.AnimStateNode_0.BodyAdditive
  nodes [
   n1 = Break Fort Anim Input Player Glider Anim Asset(FortAnimInput_PlayerGliderAnimAsset=Get CurrentGliderAnimSet()) -> (
    RootModPitchMin=~
    RootModPitchMax=~
    RootModYOffsetMin=~
    RootModYOffsetMax=~
    bEnableSpringMods=~
    bAllowPlayerDeployRootMod=~
    bUseSurfStyle=~
    bUseLeanPlayRateAdjust=~
    bUseForwardAs_Y_BS=~
    bUseDeployCurveForIntoPlayRateAdjust=~
    LeanForwardValue=~
    IdleValue=~
    LeanBackValue=~
    IdleInterpToSpeed=~
    LeanForwardInterpToSpeed=~
    LeanBackInterpToSpeed=~
   )
   n2 = Break Fort Anim Input Skydiving(FortAnimInput_Skydiving=Get Skydiving()) -> (
    ParaGilderRootOffset=~
    ParaGliderRootRotationOverride=~
    bUseParaGlideRootModifier=~
    bIsSkydivingFromLaunchPad=~
    bIsSkydivingFromBus=~
    bIsInVortex=~
    bIsDiving=~
    bIsParachuteOpen=~
    bIsSkydiving=~
    bIsParachuteLeaning=~
    bIsSkydiveLeaning=~
    bIsLeaning=~
    bIsSkydiveDiveMode=~
    bPlaySkydiveDrift=~
    bSkydiveDriftDelayActive=~
    bSkydiveDriftAnimAllowed=~
    bIsGliderRight=~
    bIsGliderCenter=~
    bIsGliderLeft=~
    bIsGliderForward=~
    bIsGliderBack=~
    bHasDeployedAndIsStandingUpright=~
    LocalAccelForward=~
    LocalAccelRight=~
    LocalVelocityRight=~
    SkydiveAimPitch=~
    SkydiveAimYaw=~
    DeployChuteAnimRate=~
    SkydiveDriftAnimRate=~
    SkydiveDriftAnimRateCurrent=~
    SkydiveFidgetAnimRate=~
    SkydiveFidgetAnimRateCurrent=~
    SkydiveAdditiveAlpha=~
    SkydiveDriftDelay=~
   )
   n3 = "Blend Poses (EFortPlayerAnimBodyType)"!(
    BlendPose_0=Blendspace Player(X=@n2.SkydiveAimYaw, Y=0, BlendSpace=@n1.Default_BodyAdditive_MaleMedium_BS) {bIgnoreForRelevancyTest=true}
    BlendPose_1=Blendspace Player(X=@n2.SkydiveAimYaw, Y=0, BlendSpace=@n1.Default_BodyAdditive_MaleLarge_BS) {bIgnoreForRelevancyTest=true}
    BlendTime_0=0.1
    BlendTime_1=0.1
    ActiveEnumValue=Get AnimBodyType() -> (AnimBodyType=Small)
   ) -> (Pose) {blendType=Linear, blendPose [{linkId=-1, sourceLinkId=-1}, {linkId=-1, sourceLinkId=-1}], blendTime [0.1,0.1]}
   n4 = "Blend Poses (EFortPlayerAnimBodyType)"!(
    BlendPose_0=Blendspace Player(X=@n2.SkydiveAimYaw, Y=0, BlendSpace=@n1.Default_BodyAdditive_FemaleMedium_BS) {bIgnoreForRelevancyTest=true}
    BlendPose_1=Blendspace Player(X=@n2.SkydiveAimYaw, Y=0, BlendSpace=@n1.Default_BodyAdditive_FemaleSmall_BS) {bIgnoreForRelevancyTest=true}
    BlendPose_2:Pose Link Structure=Blendspace Player(X=@n2.SkydiveAimYaw, Y=0, BlendSpace=@n1.Default_BodyAdditive_FemaleLarge_BS) {bIgnoreForRelevancyTest=true}
    BlendTime_0=0.1
    BlendTime_1=0.1
    BlendTime_2:float=0.1
    ActiveEnumValue=Get AnimBodyType() -> (AnimBodyType=Small)
   ) -> (Pose) {
    blendType=Linear
    blendPose [{linkId=-1, sourceLinkId=-1}, {linkId=-1, sourceLinkId=-1}, {linkId=-1, sourceLinkId=-1}]
    blendTime [0.1,0.1,0.1]
   }
   n5 = "Blend Poses (EFortCustomGender)"(BlendPose_0=@n3, BlendPose_1=@n4, BlendTime_0=0, BlendTime_1=0, ActiveEnumValue=Get Gender() -> (Gender=Unknown)) {blendType=Linear, blendPose [{linkId=-1, sourceLinkId=-1}, {linkId=-1, sourceLinkId=-1}], blendTime [0.1,0.1]}
   Output Animation Pose(Result=@n5)
  ]
 }
 ParaGlide {
  path=AnimGraph.AnimGraphNode_StateMachine_73.NoGlider.AnimStateNode_3.ParaGlide
  nodes [
   n1 = Blendspace Player(
    X=Break Fort Anim Input Skydiving!(FortAnimInput_Skydiving=Get Skydiving()) -> (bHasDeployedAndIsStandingUpright=~, LocalAccelForward=~, LocalAccelRight=~, SkydiveAimYaw=~, PoseAdditive).SkydiveAimYaw
    Y=0
    BlendSpace=Break Fort Anim Input Player Glider Anim Asset(FortAnimInput_PlayerGliderAnimAsset=Get CurrentGliderAnimSet()) -> (
     RootModPitchMin=~
     RootModPitchMax=~
     RootModYOffsetMin=~
     RootModYOffsetMax=~
     bEnableSpringMods=~
     bAllowPlayerDeployRootMod=~
     bUseSurfStyle=~
     bUseLeanPlayRateAdjust=~
     bUseForwardAs_Y_BS=~
     bUseDeployCurveForIntoPlayRateAdjust=~
     LeanForwardValue=~
     IdleValue=~
     LeanBackValue=~
     IdleInterpToSpeed=~
     LeanForwardInterpToSpeed=~
     LeanBackInterpToSpeed=~
    )
   )
   AnimGraphNode_StateMachine_1 = BodyAdditiveState()
   AnimGraphNode_StateMachine_0 = TurnAdditiveState()
   Output Animation Pose(Result=Apply Additive(
     Base=Apply Additive(Base=@n1, Additive=@AnimGraphNode_StateMachine_1) {alphaBoolBlend {blendOption=Linear}}
     Additive=@AnimGraphNode_StateMachine_0
    ) {alphaBoolBlend {blendOption=Linear}})
  ]
 }
 TurnAdditiveState {
  entry=TurnAdditive
  path=AnimGraph.AnimGraphNode_StateMachine_73.NoGlider.AnimStateNode_3.ParaGlide.AnimGraphNode_StateMachine_0.TurnAdditiveState
  ids=id
  nodes [AnimStateNode_0 = TurnAdditive()]
 }
 TurnAdditive {
  path=AnimGraph.AnimGraphNode_StateMachine_73.NoGlider.AnimStateNode_3.ParaGlide.AnimGraphNode_StateMachine_0.TurnAdditiveState.AnimStateNode_0.TurnAdditive
  nodes [
   n1 = Break Fort Anim Input Skydiving(FortAnimInput_Skydiving=Get Skydiving()) -> (bHasDeployedAndIsStandingUpright=~, LocalAccelForward=~, LocalAccelRight=~, SkydiveAimYaw=~)
   n2 = Break Fort Anim Input Player Glider Anim Asset(FortAnimInput_PlayerGliderAnimAsset=Get CurrentGliderAnimSet()) -> (
    RootModPitchMin=~
    RootModPitchMax=~
    RootModYOffsetMin=~
    RootModYOffsetMax=~
    bEnableSpringMods=~
    bAllowPlayerDeployRootMod=~
    bUseSurfStyle=~
    bUseLeanPlayRateAdjust=~
    bUseForwardAs_Y_BS=~
    bUseDeployCurveForIntoPlayRateAdjust=~
    LeanForwardValue=~
    IdleValue=~
    LeanBackValue=~
    IdleInterpToSpeed=~
    LeanForwardInterpToSpeed=~
    LeanBackInterpToSpeed=~
   )
   n3 = "Blend Poses (EFortPlayerAnimBodyType)"!(
    BlendPose_0=Blendspace Player(X=@n1.SkydiveAimYaw, Y=@n1.LocalAccelForward, BlendSpace=@n2.Default_TurnAdditive_MaleMedium_BS)
    BlendPose_1=Blendspace Player(X=@n1.SkydiveAimYaw, Y=@n1.LocalAccelForward, BlendSpace=@n2.Default_TurnAdditive_MaleLarge_BS)
    BlendTime_0=0.1
    BlendTime_1=0.1
    ActiveEnumValue=Get AnimBodyType() -> (AnimBodyType=Small)
   ) -> (Pose) {blendType=Linear, blendPose [{linkId=-1, sourceLinkId=-1}, {linkId=-1, sourceLinkId=-1}], blendTime [0.1,0.1]}
   n4 = "Blend Poses (EFortPlayerAnimBodyType)"!(
    BlendPose_0=Blendspace Player(X=@n1.SkydiveAimYaw, Y=@n1.LocalAccelForward, BlendSpace=@n2.Default_TurnAdditive_FemaleMedium_BS)
    BlendPose_1=Blendspace Player(X=@n1.SkydiveAimYaw, Y=@n1.LocalAccelForward, BlendSpace=@n2.Default_TurnAdditive_FemaleSmall_BS)
    BlendPose_2:Pose Link Structure=Blendspace Player(X=@n1.SkydiveAimYaw, Y=@n1.LocalAccelForward, BlendSpace=@n2.Default_TurnAdditive_FemaleLarge_BS)
    BlendTime_0=0.1
    BlendTime_1=0.1
    BlendTime_2:float=0.1
    ActiveEnumValue=Get AnimBodyType() -> (AnimBodyType=Small)
   ) -> (Pose) {
    blendType=Linear
    blendPose [{linkId=-1, sourceLinkId=-1}, {linkId=-1, sourceLinkId=-1}, {linkId=-1, sourceLinkId=-1}]
    blendTime [0.1,0.1,0.1]
   }
   n5 = "Blend Poses (EFortCustomGender)"(BlendPose_0=@n3, BlendPose_1=@n4, BlendTime_0=0, BlendTime_1=0, ActiveEnumValue=Get Gender() -> (Gender=Unknown)) {blendType=Linear, blendPose [{linkId=-1, sourceLinkId=-1}, {linkId=-1, sourceLinkId=-1}], blendTime [0.1,0.1]}
   Output Animation Pose(Result=@n5)
  ]
 }
 BodyAdditiveState {
  entry=BodyAdditive
  path=AnimGraph.AnimGraphNode_StateMachine_73.NoGlider.AnimStateNode_3.ParaGlide.AnimGraphNode_StateMachine_1.BodyAdditiveState
  ids=id
  nodes [AnimStateNode_0 = BodyAdditive()]
 }
 BodyAdditive {
  path=AnimGraph.AnimGraphNode_StateMachine_73.NoGlider.AnimStateNode_3.ParaGlide.AnimGraphNode_StateMachine_1.BodyAdditiveState.AnimStateNode_0.BodyAdditive
  nodes [
   Blendspace Player(X=0, Y=0)
   n1 = Break Fort Anim Input Skydiving(FortAnimInput_Skydiving=Get Skydiving()) -> (bHasDeployedAndIsStandingUpright=~, LocalAccelForward=~, LocalAccelRight=~, SkydiveAimYaw=~)
   n2 = Break Fort Anim Input Player Glider Anim Asset(FortAnimInput_PlayerGliderAnimAsset=Get CurrentGliderAnimSet()) -> (
    RootModPitchMin=~
    RootModPitchMax=~
    RootModYOffsetMin=~
    RootModYOffsetMax=~
    bEnableSpringMods=~
    bAllowPlayerDeployRootMod=~
    bUseSurfStyle=~
    bUseLeanPlayRateAdjust=~
    bUseForwardAs_Y_BS=~
    bUseDeployCurveForIntoPlayRateAdjust=~
    LeanForwardValue=~
    IdleValue=~
    LeanBackValue=~
    IdleInterpToSpeed=~
    LeanForwardInterpToSpeed=~
    LeanBackInterpToSpeed=~
   )
   n3 = "Blend Poses (EFortPlayerAnimBodyType)"!(
    BlendPose_0=Blendspace Player(X=@n1.SkydiveAimYaw, Y=0, BlendSpace=@n2.Default_BodyAdditive_MaleMedium_BS)
    BlendPose_1=Blendspace Player(X=@n1.SkydiveAimYaw, Y=0, BlendSpace=@n2.Default_BodyAdditive_MaleLarge_BS)
    BlendTime_0=0.1
    BlendTime_1=0.1
    ActiveEnumValue=Get AnimBodyType() -> (AnimBodyType=Small)
   ) -> (Pose) {blendType=Linear, blendPose [{linkId=-1, sourceLinkId=-1}, {linkId=-1, sourceLinkId=-1}], blendTime [0.1,0.1]}
   n4 = "Blend Poses (EFortPlayerAnimBodyType)"!(
    BlendPose_0=Blendspace Player(X=@n1.SkydiveAimYaw, Y=0, BlendSpace=@n2.Default_BodyAdditive_FemaleMedium_BS)
    BlendPose_1=Blendspace Player(X=@n1.SkydiveAimYaw, Y=0, BlendSpace=@n2.Default_BodyAdditive_FemaleSmall_BS)
    BlendPose_2:Pose Link Structure=Blendspace Player(X=@n1.SkydiveAimYaw, Y=0, BlendSpace=@n2.Default_BodyAdditive_FemaleLarge_BS)
    BlendTime_0=0.1
    BlendTime_1=0.1
    BlendTime_2:float=0.1
    ActiveEnumValue=Get AnimBodyType() -> (AnimBodyType=Small)
   ) -> (Pose) {
    blendType=Linear
    blendPose [{linkId=-1, sourceLinkId=-1}, {linkId=-1, sourceLinkId=-1}, {linkId=-1, sourceLinkId=-1}]
    blendTime [0.1,0.1,0.1]
   }
   n5 = "Blend Poses (EFortCustomGender)"(BlendPose_0=@n3, BlendPose_1=@n4, BlendTime_0=0, BlendTime_1=0, ActiveEnumValue=Get Gender() -> (Gender=Unknown)) {blendType=Linear, blendPose [{linkId=-1, sourceLinkId=-1}, {linkId=-1, sourceLinkId=-1}], blendTime [0.1,0.1]}
   Output Animation Pose(Result=@n5)
  ]
 }
 AnimationTransitionGraph_1 {path=AnimGraph.AnimGraphNode_StateMachine_73.NoGlider.AnimStateTransitionNode_2.AnimationTransitionGraph_1, nodes [Result()]}
 CalculateMontagePlayTiming {nodes [
   CalculateMontagePlayTiming() {is=title K2Node_FunctionEntry}
   > Set StartTime(StartTime:double=RandomFloatInRange(Min=0, Max=3)) -> (Output_Get:double)
   > Set PlayRate(PlayRate:double=RandomFloatInRange(Min=0.98, Max=1.02)) -> (Output_Get:double)
  ]}
 CharacterSkeletalControlPostLegIK {nodes [Output Pose(Result=Input Pose() {name=InPosePostLegIK}) {layerGroup=PlayerOverride}]}
 EventGraph {nodes [
   n1 = OnMontageStarted_Event_HideProps!() -> (OutputDelegate, then, Montage:Anim Montage Object Reference) {is=title K2Node_CustomEvent}
   > HidePropsForBlend!(execute, self, NewMontage:Anim Montage Object Reference=@n1.Montage, BlendingIn:bool=true) -> (then)
   n2 = OnMontageBlendingOut_Event_HideProps!() -> (OutputDelegate, then, Montage:Anim Montage Object Reference, bInterrupted:bool=~) {is=title K2Node_CustomEvent}
   > HidePropsForBlend(NewMontage=@n2.Montage, BlendingIn=false)
   UserFocus() {is=title K2Node_CustomEvent}
   > Set HasBeenSelected(HasBeenSelected:bool=true) -> (Output_Get:bool)
   n3 = WhenBackpackPart!() -> (
    OutputDelegate
    then
    Pawn:Fort Player Pawn Object Reference
    Part:Custom Character Part Object Reference
    MeshPart:Skeletal Mesh Component Object Reference
    PartType:EFortCustomPartType Enum=Head
   ) {is=title K2Node_CustomEvent}
   > Is Valid(InputObject=@n3.MeshPart)
   n4 = GetSkeletalMeshAsset!(self:Skeletal Mesh Component Object Reference=@n3.MeshPart) -> (ReturnValue:Skeletal Mesh Object Reference)
   > Is Valid(InputObject=@n4)
   > Branch(Condition=Contains(SearchIn=GetObjectName(Object=@n4), Substring=BackPack_Pet))
   > Set PoseInt(PoseInt:int=1) -> (Output_Get:int=0)
   n5 = OnBodyPart!() -> (
    OutputDelegate
    then
    Pawn:Fort Player Pawn Object Reference
    Part:Custom Character Part Object Reference
    MeshPart:Skeletal Mesh Component Object Reference
    PartType:EFortCustomPartType Enum=Head
   ) {is=title K2Node_CustomEvent}
   > n6 = Is Valid(InputObject=@n5.MeshPart)
   > n7 = Sequence!(execute) -> (then_0, then_1, then_2:Exec, then_3:Exec, then_4:Exec)
   n8 = GetFrontendAnimMontageIdleOverride!(self:Fort Player Pawn Object Reference=Get PawnOwner() -> (PawnOwner:Player Pawn Athena Object Reference)) -> (ReturnValue:Anim Montage Object Reference)
   > n9 = Branch(Condition=OR Boolean!(
     A=Get bDontCrossArms() -> (bDontCrossArms:bool)
     B=Get bNoHandsOnHips() -> (bNoHandsOnHips:bool)
     C:bool=Get bUseSimplePose()
     D:bool=IsValid(Object=@n8)
    ) -> (ReturnValue))
   > Set PoseInt(PoseInt=1) -> (Output_Get=0)
   n10 = Branch(execute=@n9.else, Condition=Get bIsRebirth() -> (bIsRebirth:bool))
   > Set PoseInt(PoseInt=RandomIntegerInRange!(Min:int=1, Max:int=2) -> (ReturnValue:int=0)) -> (Output_Get=0)
   Set PoseInt(execute=@n10.else, PoseInt=RandomIntegerInRange(Min=0, Max=1) -> (ReturnValue=0)) -> (Output_Get=0)
   n11 = Get PawnOwner()
   WhenCosmeticSpecificPartApplied!(
    execute=@n7.then_1
    self:Fort Player Pawn Object Reference=@n11
    Delegate:Delegate=@n3
    SkipNullMesh:bool=false
    PartType:EFortCustomPartType Enum=Backpack
   ) -> (then, DelegateHandleController_Out:Delegate Handle Controller Structure)
   n12 = GetFrontendAnimMontageIdleOverride(self=Get PawnOwner())
   n13 = Branch(
    execute=@n7.then_3
    Condition=AND Boolean(
     A=Montage_IsActive!(self:Anim Instance Object Reference, Montage:Anim Montage Object Reference) -> (ReturnValue:bool)
     B=IsValid(Object=@n12)
    )
   )
   Do Once(execute=@n39.then, Reset=[@n14.Is Not Valid, @n6.Is Not Valid, @n13.then])
   > n14 = Is Valid(InputObject=@n11)
   > WhenCosmeticSpecificPartApplied(self=@n11, Delegate=@n5, SkipNullMesh=false)
   Branch(
    execute=@n7.then_2
    Condition="Not Equal (Object)"!(A:Object Reference=@n8, B:Object Reference=Get Current Idle Montage() -> (Current Idle Montage:Anim Montage Object Reference)) -> (ReturnValue:bool)
   )
   > CalculateMontagePlayTiming() {is=title K2Node_CallFunction}
   > Montage_Play!(
    execute
    self:Anim Instance Object Reference
    MontageToPlay:Anim Montage Object Reference=@n8
    InPlayRate:float=Get PlayRate()
    ReturnValueType:EMontagePlayReturnType Enum=MontageLength
    InTimeToStartMontageAt:float=Get StartTime()
    bStopAllMontages:bool=true
   ) -> (then, ReturnValue:float)
   > Set Current Idle Montage(Current Idle Montage:Anim Montage Object Reference=@n8) -> (Output_Get:Anim Montage Object Reference)
   AnimNotify_Melee_Swing_L_End() {is=title K2Node_Event}
   > n15 = Cast To FortPawn!(execute, Object=TryGetPawnOwner()) -> (then, CastFailed, AsFort Pawn:Fort Pawn Object Reference) {is=title K2Node_DynamicCast}
   > MeleeSwingLeft_End!(execute:Exec, self:Object Reference=@n15) -> (then:Exec) {is=title K2Node_Message}
   AnimNotify_LobbySkydive_Diving() {is=title K2Node_Event}
   > Set bIsSkydiveDiving(bIsSkydiveDiving:bool=true) -> (Output_Get:bool)
   n16 = Get CosmeticPetInstance!(self:Fort Player Pawn Object Reference=Get PawnOwner()) -> (CosmeticPetInstance:Fort Player Pet Object Reference)
   > Is Valid(InputObject=@n16)
   n17 = GetAnimInstance(self=Get PetMesh!(self:Fort Player Pet Object Reference=@n16) -> (PetMesh:Skeletal Mesh Component Budgeted Object Reference))
   > Is Valid(InputObject=@n17)
   > n18 = Branch(Condition=Contains(SearchIn=GetDisplayName(Object=@n17), Substring=HightowerRadish))
   > n19 = Cast To B_AnimBP_HightowerRadish!(execute, Object=@n17) -> (then, CastFailed, AsB Anim BP Hightower Radish:B Anim BP Hightower Radish Object Reference) {is=title K2Node_DynamicCast}
   > HighTowerRadish_Lobby SkyDive Diving(self:B Anim BP Hightower Radish Object Reference=@n19)
   n20 = Cast To B_MasterPetAnimBP!(execute=@n18.else, Object=@n17) -> (then, CastFailed, AsB Master Pet Anim BP:B Master Pet Anim BP Object Reference) {is=title K2Node_DynamicCast}
   > LobbySkyDive_IsDiving(self:B Master Pet Anim BP Object Reference=@n20)
   AnimNotify_Melee_Swing_L() {is=title K2Node_Event}
   > n21 = Cast To FortPawn(Object=TryGetPawnOwner())
   > MeleeSwingLeft!(execute:Exec, self:Object Reference=@n21, First Left:bool=true) -> (then:Exec) {is=title K2Node_Message}
   AnimNotify_Melee_Swing_R() {is=title K2Node_Event}
   > n22 = Cast To FortPawn(Object=TryGetPawnOwner())
   > MeleeSwingRight!(execute:Exec, self:Object Reference=@n22, First Right:bool=true) -> (then:Exec) {is=title K2Node_Message}
   AnimNotify_playFacialAnim() {is=title K2Node_Event}
   > Branch(Condition=Get bCanPlayCustomAnimations() -> (bCanPlayCustomAnimations:bool))
   > Call Menu Screen Dispatcher!(execute:Exec, self:Menu Screen Athena Object Reference) -> (then:Exec) {is=title K2Node_CallDelegate}
   AnimNotify_Melee_Swing_R_End() {is=title K2Node_Event}
   > n23 = Cast To FortPawn(Object=TryGetPawnOwner())
   > MeleeSwingRight_End!(execute:Exec, self:Object Reference=@n23) -> (then:Exec) {is=title K2Node_Message}
   AnimNotify_Melee_Swing_L_2() {is=title K2Node_Event}
   > n24 = Cast To FortPawn(Object=TryGetPawnOwner())
   > MeleeSwingLeft(self=@n24, First Left=false)
   AnimNotify_Melee_Swing_R_2() {is=title K2Node_Event}
   > n25 = Cast To FortPawn(Object=TryGetPawnOwner())
   > MeleeSwingRight(self=@n25, First Right=false)
   n26 = Event OnFocusedItemChanged!() -> (
    OutputDelegate
    then
    FocusedItem:Fort Item Definition Object Reference
    EquippedWeapon:Fort Weapon Object Reference
    InCurrentParachute:Fort Player Parachute Object Reference
   ) {is=title K2Node_Event}
   > n27 = Cast To AthenaPickaxeItemDefinition!(execute, Object=@n26.FocusedItem) -> (then, CastFailed, AsAthena Pickaxe Item Definition:Athena Pickaxe Item Definition Object Reference) {is=title K2Node_DynamicCast}
   Set HasWeaponEquipped(execute=[@n28.then, @n27.then], HasWeaponEquipped:bool=true) -> (Output_Get:bool)
   n28 = Cast To AthenaItemWrapDefinition!(execute=@n27.CastFailed, Object=@n26.FocusedItem) -> (then, CastFailed, AsAthena Item Wrap Definition:Athena Item Wrap Definition Object Reference) {is=title K2Node_DynamicCast}
   Set HasWeaponEquipped(execute=@n28.CastFailed, HasWeaponEquipped=false)
   > n29 = Cast To FortPlayerPawn(Object=TryGetPawnOwner())
   > n30 = Is Skydiving!(exec, Player:Object Reference=@n29) -> (true:Exec, false:Exec) {is=title K2Node_MacroInstance}
   n31 = GetFrontendAnimMontageIdleOverride(self=@n29)
   n32 = Branch(execute=@n30.false, Condition=Montage_IsPlaying(Montage=@n31))
   CalculateMontagePlayTiming(execute=@n32.else)
   > Montage_Play(MontageToPlay=@n31, InPlayRate=1, ReturnValueType=MontageLength, InTimeToStartMontageAt=0, bStopAllMontages=true)
   AnimNotify_LeftB() {is=title K2Node_Event}
   > Set HasBeenSelected(HasBeenSelected=false)
   n33 = Event BlueprintUpdateAnimation()
   > n34 = Set HidePropTimeRemaining(HidePropTimeRemaining:double=float - float(A=Get HidePropTimeRemaining() -> (HidePropTimeRemaining:double), B=@n33.DeltaTimeX) {is=title K2Node_CallFunction}) -> (Output_Get:double)
   > n35 = Branch(Condition="float <= float"!(A:double=@n34, B:double=0) -> (ReturnValue:bool) {is=title K2Node_CallFunction})
   > n36 = Set HidePropBones(HidePropBones:bool=false) -> (Output_Get:bool)
   n37 = GetCurveValue(CurveName=FrontendRotation)
   n38 = BreakRotator(InRot=Get OffsetRotate())
   n39 = Set OffsetRotate(execute=[@n36.then, @n35.else], OffsetRotate:Rotator=MakeRotator(Roll=@n38, Pitch=@n38.Pitch, Yaw=@n37)) -> (Output_Get:Rotator)
   AnimNotify_LobbySkydive_Gliding() {is=title K2Node_Event}
   > Set bIsSkydiveDiving(bIsSkydiveDiving=false)
   n40 = Get CosmeticPetInstance(self=Get PawnOwner())
   > Is Valid(InputObject=@n40)
   n41 = GetAnimInstance(self=Get PetMesh(self=@n40))
   > Is Valid(InputObject=@n41)
   > n42 = Branch(Condition=Contains(SearchIn=GetDisplayName(Object=@n41), Substring=HightowerRadish))
   > n43 = Cast To B_AnimBP_HightowerRadish(Object=@n41)
   > HighTowerRadish_Lobby SkyDive Gliding(self:B Anim BP Hightower Radish Object Reference=@n43)
   n44 = Cast To B_MasterPetAnimBP(execute=@n42.else, Object=@n41)
   > LobbySkyDive_IsGliding(self:B Master Pet Anim BP Object Reference=@n44)
   Event BlueprintInitializeAnimation()
   > Assign On Montage Started!(execute:Exec, self:Anim Instance Object Reference, Delegate:"Delegate (by ref)"=@n1) -> (then:Exec) {is=title K2Node_AssignDelegate}
   > Assign On Montage Blending Out!(execute:Exec, self:Anim Instance Object Reference, Delegate:"Delegate (by ref)"=@n2) -> (then:Exec) {is=title K2Node_AssignDelegate}
   n45 = TryGetPawnOwner()
   > Is Valid(InputObject=@n45)
   > n46 = Cast To PlayerPawn_Athena(Object=@n45)
   > n47 = Set PawnOwner(PawnOwner:Player Pawn Athena Object Reference=@n46) -> (Output_Get:Player Pawn Athena Object Reference)
   > CalculateMontagePlayTiming()
   n48 = GetFrontendAnimMontageIdleOverride(self=@n47)
   > Is Valid(InputObject=@n48)
   > Set Current Idle Montage()
   > n49 = Branch(Condition=OR Boolean!(A=Montage_IsActive(Montage=@n48), B=IsSkydiving!(self:Fort Player Pawn Object Reference=@n47) -> (ReturnValue:bool)) -> (ReturnValue))
   Montage_Play(
    execute=@n49.else
    MontageToPlay=@n48
    InPlayRate=Get PlayRate()
    ReturnValueType=MontageLength
    InTimeToStartMontageAt=Get StartTime()
    bStopAllMontages=true
   )
   float * float(A=@n37, B=180)
  ]}
 FullBodyOverride {nodes [Output Pose(Result=Input Pose() {name=PassThroughFullBodyPose}) {layerGroup=PlayerOverride}]}
 HidePropsForBlend {nodes [
   n1 = HidePropsForBlend!() -> (then, NewMontage:Anim Montage Object Reference, BlendingIn:bool=~) {is=title K2Node_FunctionEntry}
   > Branch(Condition=ShouldHidePropsDuringBlend!(self:Frontend Anim Instance Object Reference, Montage:Anim Montage Object Reference=@n1) -> (ReturnValue:bool))
   > n2 = Cast To FortPlayerPawn(Object=TryGetPawnOwner())
   n3 = GetFrontendAnimMontageIdleOverride(self=@n2)
   > Is Valid(InputObject=@n3)
   > Branch(Condition="Not Equal (Object)"(A=@n3, B=@n1))
   n4 = GetDefaultBlendOutTime!(self:Anim Montage Object Reference=@n1) -> (ReturnValue:float)
   n5 = GetDefaultBlendInTime!(self:Anim Montage Object Reference=@n1) -> (ReturnValue:float)
   > Set HidePropTimeRemaining(HidePropTimeRemaining=float + float(A=SelectFloat(A=@n5, B=@n4, bPickA=@n1.BlendingIn), B=0.05) {is=title K2Node_CommutativeAssociativeBinaryOperator})
   > Set HidePropBones(HidePropBones=true)
  ]}
 MenuScreenDispatcher {nodes [MenuScreenDispatcher() {is=title K2Node_FunctionEntry}]}
 PreLegIK {nodes [Output Pose(Result=Input Pose()) {layerGroup=PlayerOverride}]}
 UpperBody {nodes [
   Input Pose() {name=InPoseUpperBody}
   Output Pose() {layerGroup=PlayerOverride}
  ]}
]
variables [
 Current Idle Montage {type=AnimMontage, default=null}
 HasBeenSelected {type=boolean}
 HasWeaponEquipped {type=boolean}
 HidePropBones {type=boolean}
 HidePropTimeRemaining {type=number}
 IsSkydiveDiving_Potato {type=boolean}
 MenuScreenDispatcher {default=()}
 OffsetRotate {type=Rotator}
 OffsetTranslate {type=Vector}
 PawnOwner {type=PlayerPawn_Athena_C}
 PlayRate {type=number}
 PoseInt {type=integer}
 Skydive Aim Pitch {type=number}
 Skydive Aim Yaw {type=number}
 StartTime {type=number}
]
