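; PlayerPawn_Generic is a player pawn Blueprint that drives mostly cosmetic feedback for movement, damage, shields, water, death and speech bubbles.
; Tick and landing events spawn ground or water particles, gated by cosmetic-execution checks and delay throttles.
; Damage and gameplay-cue events drive hit particles, lens effects, camera shake and timeline-animated shield shatter on duplicate meshes.
; Cosmetic-part callbacks rebuild material-instance arrays, while other events handle first-person visibility and War Cry weapon materials.
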(event InterfacePlayerPawn|EventMeleeSwingLeft
  (params (First%20Left :type Boolean))
  ;[1 1 1] Dual-wield pickaxe event forwarding.
  (if First%20Left
    (Default|CallSwingLeft :!class /Game/Abilities/Player/Pawns/PlayerPawn_Generic.PlayerPawn_Generic_C)
    (else
      (Default|CallSwingLeft2))))

(event GameplayCue|GameplayCue.Abilities.Activation.Generic.HarvestBuff.Tier1
  (params (EventType :type EGameplayCueEvent%20Enum :default "OnActive") (Parameters :type Gameplay%20Cue%20Parameters%20Structure%20%28by%20ref%29))
  ;[1 1 1] Proto Team - Harvesting Improvement Test
  (switch Utilities|FlowControl|Switch|SwitchonEGameplayCueEvent EventType
    (:OnActive
      _)
    (:WhileActive
      _)
    (:Executed
      _)
    (:Removed
      _))
  (Ability|GameplayCue|ForwardGameplayCuetoParent))

(event InterfacePlayerPawn|EventMeleeSwingRight
  (params (First%20Right :type Boolean))
  ;[1 1 1] Dual-wield pickaxe event forwarding.
  (if First%20Right
    (Default|CallSwingRight)
    (else
      (Default|CallSwingRight2))))

(event EventPossessed
  (params (NewController :type Controller%20Object%20Reference))
  ;[1 1 1] Where we are going we need roads Ability
  (|CallParentFunction NewController)
  (FortPawn|StartPlayerDistanceTraveledTimer) ; Distance Traveled Stat Event, not specific to roads ability
  (Utilities|FlowControl|Delay 0.5)
  (if (GameplayTags|TagContainer|CompareTagAssetInterfacetoOtherTagContainers (Variables|Default|GetTC_WhereWereGoingWeNeedRoads) (Ability|GetAbilitySystemComponent self))
    (Variables|PawnStats|SetbUseBaseChanged true)
    (else
      (Variables|PawnStats|SetbUseBaseChanged false))))

(event GameplayCue|GameplayCue.Abilities.Activation.Generic.HarvestBuff.Tier2
  (params (EventType :type EGameplayCueEvent%20Enum :default "OnActive") (Parameters :type Gameplay%20Cue%20Parameters%20Structure%20%28by%20ref%29))
  ;[1 1 1] Proto Team - Harvesting Improvement Test
  (switch Utilities|FlowControl|Switch|SwitchonEGameplayCueEvent EventType
    (:OnActive
      _)
    (:WhileActive
      _)
    (:Executed
      _)
    (:Removed
      _))
  (Ability|GameplayCue|ForwardGameplayCuetoParent))

(event InterfacePlayerPawn|EventMeleeSwingLeft_End
  ;[1 1 1] Dual-wield pickaxe event forwarding.
  (Default|CallSwingLeftEnd :!class /Game/Abilities/Player/Pawns/PlayerPawn_Generic.PlayerPawn_Generic_C))

(event InterfacePlayerPawn|EventMeleeSwingRight_End
  ;[1 1 1] Dual-wield pickaxe event forwarding.
  (Default|CallSwingRightEnd :!class /Game/Abilities/Player/Pawns/PlayerPawn_Generic.PlayerPawn_Generic_C))

(event EventTick
  (params (DeltaSeconds :type Float%20%28single-precision%29))
  (bind _returnvalue (Transformation|GetVelocity))
  (bind _effect_player_walk_water (Variables|Default|GetEffect_Player_Walk_Water))
  (bind _effect_player_walk_water_2 (Variables|Default|GetEffect_Player_Walk_Water))
  ;[1 1 1] Walk / Run Dust FX
  (Networking|CanExecuteCosmeticEvents
    (:True
      (Variables|Default|SetPlayerSpeed (Math|Vector|VectorLength (Math|Vector|MakeVector (.x (Transformation|GetVelocity)) (.y (Transformation|GetVelocity)))))
      ;;[0.04099 0.1928 1] Ground/Water Interaction Effects
      ;;; Ground Interaction Effects
      (Utilities|UpdateCharacterWaterLevel
        (:out
          (if (Math|Boolean|ORBoolean (Variables|VFX|GetIsNinjaInShadowStance?) (not (AI|Components|NavMovement|IsMovingOnGround)) (not (Variables|VFX|GetEnableRunWalkWaterGroundFX)))
            (Components|Activation|Deactivate (Variables|Default|GetEffect_Player_Walk_Water))
            (elif (Variables|Water|GetIsInWater)
              (bind _outhit (Collision|LineTraceByChannel (+ (Transformation|GetWorldLocation (Variables|Transformation|GetRootComponent)) "0, 0,190.000000") (+ (Transformation|GetWorldLocation (Variables|Transformation|GetRootComponent)) "0, 0,-50.000000") "TraceTypeQuery10" true))
              (if (Collision|BreakHitResult _outhit)
                (if (Math|Vector|Equal(Vector) _returnvalue "0, 0, 0" 0.0001)
                  (Components|Activation|Deactivate (Variables|Default|GetEffect_Player_Walk_Water))
                  (else
                    (Variables|VFX|SetStableVelocityVector (Math|Vector|Normalize _returnvalue))))
                (bind _output_get (Variables|VFX|SetRunWalkParticleRotation (Math|Conversions|RotationFromXVector (Math|Vector|CrossProduct (.ImpactNormal (Collision|BreakHitResult _outhit)) (Variables|VFX|GetStableVelocityVector)))))
                (Transformation|SetWorldLocation _effect_player_walk_water (.Location (Collision|BreakHitResult _outhit)) false false)
                (Utilities|PlayerGroundInteractionParticleUpdates _effect_player_walk_water (Variables|Default|GetEffect_Player_Run_Water) 1 _output_get "1.000000,1.000000,1.000000"
                  (:Kick%20up%20-%20walk%20starts
                    (Components|Activation|Deactivate (Variables|Default|GetEffect_Player_Run_Water))
                    (Utilities|FlowControl|Delay 0.5) ; Time to throttle walk FX\nPrevents FX spamming
                    (Variables|VFX|SetCanSpawnWalkKickupFX? true))
                  (:Stop
                    (Components|Activation|SetActive _effect_player_walk_water_2 false true))
                  (:Run%20Kickup
                    (Components|Activation|Deactivate _effect_player_walk_water_2)))
                (else
                  (Components|Activation|Deactivate (Variables|Default|GetEffect_Player_Walk_Water))))
              (else
                (Components|Activation|Deactivate (Variables|Default|GetEffect_Player_Run_Water))
                (Variables|VFX|SetRunWalkParticleRotation (Math|Rotator|MakeRotfromX (Math|Vector|MakeVector (.x (Transformation|GetVelocity)) (.y (Transformation|GetVelocity)))))
                (Utilities|PlayerGroundInteractionParticleUpdates (Variables|Default|GetEffect_Player_Walk_Land) (Variables|Default|GetEffect_Player_Run_Land) 1.45 (Math|Conversions|RotationFromXVector (* (Variables|VFX|GetPreviousVelocityVector) -1)) "1.400000,1.400000,1.400000"
                  (:Kick%20up%20-%20walk%20starts
                    (Utilities|FlowControl|Delay 0.5) ; Time to throttle walk FX\nPrevents FX spamming
                    (Variables|VFX|SetCanSpawnWalkKickupFX? true))
                  (:Stop
                    (Utilities|FlowControl|Delay 0.25) ; Time to throttle walk FX\nPrevents FX spamming
                    (Variables|VFX|SetCanSpawnWalkKickupFX? true))
                  (:Run%20Kickup
                    (Utilities|FlowControl|Delay 0.5) ; Time to throttle walk FX\nPrevents FX spamming
                    (Variables|VFX|SetCanSpawnRunKickupFX? true)))))))))))

(event Vehicle|EventOnEnteredVehicle
  ;[1 1 1] Enter and exit vehicle events.
  (|CallParentFunction)
  (Ability|SendGameplayEventtoActor self "(TagName=\"Event.Vehicle.Entered\")"))

(event EventOnBaseChanged
  (params (NewBase :type Actor%20Object%20Reference))
  ;[1 1 1] Where we are going we need roads Ability
  (if (Utilities|IsValid NewBase :!class /Script/Engine.KismetSystemLibrary)
    (bind _asbuilding_static_mesh (Utilities|Casting|CastToBuildingSMActor NewBase)
      (:then
        (if (BuildingActor|IsPlayerBuilt _asbuilding_static_mesh)
          (if (not (Variables|Default|GetOn_Player_Built_Floor))
            (Variables|Default|SetOn_Player_Built_Floor true))
          (elif (Variables|Default|GetOn_Player_Built_Floor)
            _)))
      (:CastFailed
        _))
    (Variables|Default|SetOn_Player_Built_Floor false)))

(event Vehicle|EventOnExitedVehicle
  ;[1 1 1] Enter and exit vehicle events.
  (|CallParentFunction)
  (Ability|SendGameplayEventtoActor self "(TagName=\"Event.Vehicle.Exit\")"))

(event Custom|ToggleLandFX
  (params (NewHidden :type Boolean))
  ;[1 1 1] Event to turn off Land FX when crouch sliding
  (Development|SetHiddeninGame (Variables|Default|GetEffect_Player_Walk_Land) NewHidden false))

(event EventOnLanded
  (params (Hit :type Hit%20Result%20Structure%20%28by%20ref%29))
  (bind _returnvalue (Math|Random|RandomFloatinRange -45 45))
  (bind _returnvalue_2 (Math|Rotator|MakeRotator 0.0 0.0 _returnvalue))
  (bind _returnvalue_3 (Transformation|GetActorTransform))
  (bind _rotation (.rotation _returnvalue_3))
  (bind _returnvalue_4 (Math|Rotator|CombineRotators _rotation _returnvalue_2))
  (bind _location (.location _returnvalue_3))
  (bind _returnvalue_5 (+ _location "0.840000,-6.210000,-71.750000"))
  ;[0.05415 0 1] Player Landed Effects
  (Networking|CanExecuteCosmeticEvents
    (:True
      (if (not (or (Variables|VFX|GetIsNinjaInShadowStance?) (Variables|Water|GetIsInWater)))
        (CallFunction|OnLandCE)
        (if (Variables|VFX|GetCanSpawnDustLandFX?)
          (Effects|SpawnEmitteratLocation (Variables|VFX|GetEffect_Player_LandedDust) _returnvalue_5 _returnvalue_4 "1.000000,1.000000,1.000000" true "AutoRelease")
          (Utilities|AddWindParticleatLocation (Variables|VFX|GetEffect_Player_Landed_WindVector_P) (.location (Math|Transform|MakeTransform _returnvalue_5 _returnvalue_4)) (.rotation (Math|Transform|MakeTransform _returnvalue_5 _returnvalue_4)))
          (Variables|VFX|SetCanSpawnDustLandFX? false)
          (Utilities|FlowControl|Delay 0.4)
          (Variables|VFX|SetCanSpawnDustLandFX? true))))))

(event GameplayCue|GameplayCue.Damage.Shielded
  (params (EventType :type EGameplayCueEvent%20Enum :default "OnActive") (Parameters :type Gameplay%20Cue%20Parameters%20Structure%20%28by%20ref%29))
  ;[1 1 1] Shields
  (FortPawn|GameplayCueDamage EventType Parameters))

(event GameplayCue|GameplayCue.Damage
  (params (EventType :type EGameplayCueEvent%20Enum :default "OnActive") (Parameters :type Gameplay%20Cue%20Parameters%20Structure%20%28by%20ref%29))
  ;[1 1 1] Shields
  (FortPawn|GameplayCueDamage EventType Parameters))

(event EventOnDamagePlayEffects
  (params (Damage :type Float%20%28single-precision%29) (DamageTags :type Gameplay%20Tag%20Container%20Structure%20%28by%20ref%29) (Momentum :type Vector :default "0, 0, 0") (HitInfo :type Hit%20Result%20Structure%20%28by%20ref%29) (InstigatedBy :type Fort%20Pawn%20Object%20Reference) (DamageCauser :type Actor%20Object%20Reference) (EffectContext :type Gameplay%20Effect%20Context%20Handle%20Structure))
  (bind _mesh (Variables|Character|GetMesh))
  (bind _returnvalue (Transformation|GetSocketLocation _mesh "spine_04"))
  (bind (_bblockinghit _binitialoverlap _time _distance _location _impactpoint _normal _impactnormal _physmat _hitactor _hitcomponent _hitbonename _bonename _hititem _elementindex _faceindex _tracestart _traceend) (Collision|BreakHitResult HitInfo))
  (bind _returnvalue_2 (- _tracestart _traceend))
  (bind _latestdamageisshielddamage (Variables|Shield|GetLatestDamageIsShieldDamage))
  (bind _returnvalue_3 (not _latestdamageisshielddamage))
  (bind _hitdamageparticles (Variables|Default|GetHitDamageParticles))
  ;[1 1 1] Damage Effects
  ;; Play hit effects and On Screen Damage Indicators
  (Variables|Default|SetDamage_Taken Damage)
  (bind _output_get (Variables|Shield|SetLatestDamageIsShieldDamage (GameplayTags|HasAnyTags DamageTags (Variables|Shield|GetTC_GameplayCueDamageShielded))))
  (if _output_get
    (Variables|Shield|SetLast_Shield_Damage_Time (Utilities|Time|GetGameTimeinSeconds)))
  (Utilities|FlowControl|Sequence
    (:then_0
      (Components|Activation|Activate _hitdamageparticles true)
      (Effects|Components|ParticleSystem|SetEmitterEnable _hitdamageparticles "Shield" _latestdamageisshielddamage)
      (Effects|Components|ParticleSystem|SetEmitterEnable _hitdamageparticles "Streaks" _returnvalue_3)
      (Effects|Components|ParticleSystem|SetEmitterEnable _hitdamageparticles "Round" _returnvalue_3)
      (if (not (> (Math|Vector|VectorLengthSquared _returnvalue_2) 0.0))
        (Utilities|IsValid DamageCauser
          (:Is%20Valid
            (Transformation|SetWorldLocationAndRotation _hitdamageparticles _returnvalue (Math|Conversions|RotationFromXVector (- (Transformation|GetActorLocation DamageCauser) (Transformation|GetActorLocation))) false false))
          (:Is%20Not%20Valid
            (Transformation|SetWorldLocationAndRotation _hitdamageparticles _returnvalue (Math|Rotator|MakeRotfromX _returnvalue_2) false false :!label _label)))
        (Feedback|SpawnGenericCameraLensEffectDirectional (LensEffect|GetInterfaceClass (select (Variables|Shield|GetLatestDamageIsShieldDamage) (Variables|Default|GetLensEffectInterfaceClass_Directional_PlayerShieldDamage) (Variables|Default|GetLensEffectInterfaceClass_Directional_PlayerHealthDamage))) InstigatedBy HitInfo)
        (bind _returnvalue_4 (Feedback|SpawnGenericCameraLensEffect (LensEffect|GetInterfaceClass (select (Variables|Shield|GetLatestDamageIsShieldDamage) (Variables|Default|GetLensEffectInterfaceClass_PlayerShieldDamage) (Variables|Default|GetLensEffectInterfaceClass_PlayerHealthDamage)))))
        (if (Variables|Shield|GetLatestDamageIsShieldDamage)
          (bind _asb_player_shield_damage_camera_lens_effect (Utilities|Casting|CastToB_PlayerShieldDamage_CameraLensEffect (Utilities|ToObject(Interface) _returnvalue_4))
            (:then
              (bind _returnvalue_5 (CallFunction|GetShieldPercent2))
              (Class|BPlayerShieldDamageCameraLensEffect|PassParticleParameterShield _asb_player_shield_damage_camera_lens_effect _returnvalue_5))
            (:CastFailed
              _))
          (else
            (bind _asb_player_health_damage_camera_lens_effect (Utilities|Casting|CastToB_PlayerHealthDamage_CameraLensEffect (Utilities|ToObject(Interface) _returnvalue_4))
              (:then
                (bind _returnvalue_6 (CallFunction|GetHealthPercent2))
                (Class|BPlayerHealthDamageCameraLensEffect|PassParticleParameter _asb_player_health_damage_camera_lens_effect _returnvalue_6 (and (not (Variables|Shield|GetLatestDamageIsShieldDamage)) (< (- (Utilities|Time|GetGameTimeinSeconds) (Variables|Shield|GetLast_Shield_Damage_Time)) 0.2))))
              (:CastFailed
                _))))
        (bind _asfort_player_controller (Utilities|Casting|CastToFortPlayerController (Pawn|GetController)))
        (Game|Feedback|ClientPlayForceFeedback _asfort_player_controller (Variables|Default|GetDamageForceFeedback) "None" false false false)
        (CallFunction|PlayHitSound InstigatedBy DamageCauser)
        (if (Pawn|IsLocalPlayerController _asfort_player_controller)
          (if (not (GameplayTags|HasMatchingGameplayTag self "(TagName=\"Gameplay.Damage.OnTakeDamage.IgnoreCameraShake\")"))
            (Game|Feedback|ClientStartCameraShake :self _asfort_player_controller :Shake (Variables|Default|GetDamageCameraShake) :Scale 1.0 :PlaySpace "CameraLocal"))))
      (goto _label))
    (:then_1
      (if (not (GameplayTags|HasMatchingGameplayTag self "(TagName=\"Gameplay.Damage.OnTakeDamage.IgnoreHitReact\")"))
        (Animation|PlayLocalAnimMontage "/Game/Animation/Game/MainPlayer/Montages/General/AdditiveHitReacts.AdditiveHitReacts" 1.0 (Utilities|Select :Index (Math|Random|RandomInteger 3) :Option%200 "None" :Option%201 "InFrontv2" :Option%202 "FrontLeft" :Option%203 "FrontRight"))))))

(event EventOnDeathServer
  (params (Damage :type Float%20%28single-precision%29) (DamageTags :type Gameplay%20Tag%20Container%20Structure%20%28by%20ref%29) (Momentum :type Vector :default "0, 0, 0") (HitInfo :type Hit%20Result%20Structure%20%28by%20ref%29) (InstigatedBy :type Controller%20Object%20Reference) (DamageCauser :type Actor%20Object%20Reference) (EffectContext :type Gameplay%20Effect%20Context%20Handle%20Structure))
  ;[1 1 1] DBNO "Spawn at start"
  (if (Utilities|Equal(Object) DamageCauser self)
    (bind _asfort_player_controller (Utilities|Casting|CastToFortPlayerController (Pawn|GetController)))
    (Game|Feedback|ClientPlayForceFeedback _asfort_player_controller "/Game/Weapons/ForceFeedback/FF_Generic_DBNORevive.FF_Generic_DBNORevive" "DBNOResurrect" false false false)))

(event GameplayCue|GameplayCue.Shield.FullyCharged
  (params (EventType :type EGameplayCueEvent%20Enum :default "OnActive") (Parameters :type Gameplay%20Cue%20Parameters%20Structure%20%28by%20ref%29))
  ;[1 1 1] Shields
  ;;[1 1 1] Left here to keep BR specific items in GC and still allow StW custom assets here
  (switch Utilities|FlowControl|Switch|SwitchonEGameplayCueEvent EventType
    (:Executed
      (Audio|SpawnSoundAttached (Variables|Sounds|GetSound_Shield_FullyCharged) (Variables|Transformation|GetRootComponent)))))

(event GameplayCue|GameplayCue.Shield.Destroyed
  (params (EventType :type EGameplayCueEvent%20Enum :default "OnActive") (Parameters :type Gameplay%20Cue%20Parameters%20Structure%20%28by%20ref%29))
  ;[1 1 1] Shields
  (switch Utilities|FlowControl|Switch|SwitchonEGameplayCueEvent EventType
    (:Executed
      (Audio|SpawnSoundAttached (Variables|Sounds|GetSound_Shield_Destroyed) (Variables|Transformation|GetRootComponent))
      (Networking|CanExecuteCosmeticEvents
        (:True
          ;; Shatter Shield
          (if (Variables|Shield|GetShield%20Active)
            (Variables|Shield|SetShieldVisible_Shatter true)
            (CallFunction|ToggleShieldVisibility)
            (Components|Timeline|SetPlayRate (Variables|PlayerPawn_Generic|GetShatterShield) 0.5)
            (CallFunction|SetShieldMids 0.0 false (Variables|Shield|GetShield_PushMinValue) true)
            (bind (_direction _highlight_cracks _opacity _push) (Timeline|ShatterShield PlayFromStart)
              (:Update
                (Variables|Shield|SetShield%20Shatter%20Opacity _opacity)
                (CallFunction|FindShieldOpacity)
                (CallFunction|SetShieldMids _highlight_cracks true (+ (Variables|Shield|GetShield_PushMinValue) _push) true))
              (:Finished
                (Variables|Shield|SetShield%20Shatter%20Opacity 0.0)
                (CallFunction|FindShieldOpacity)
                (Variables|Shield|SetShieldVisible_Shatter false)
                (CallFunction|ToggleShieldVisibility)))))))))

(event EventOnCharacterPartsReinitialized
  ;[1 0.9036 0.208] Character mid init
  (Networking|CanExecuteCosmeticEvents
    (:True
      ;; Clear  Char MID Arrays
      (for _array_element (Variables|CharacterMIDs|GetPawnPartMaterials)
        (bind (_array_element_parttype_5_ebdfff5740627902cc51dd966b8ee419 _array_element_mid_7_a0d19ac74319389a5df2019166f976f0) (Utilities|Struct|BreakCharacterPartMidArrayStruct _array_element))
        (Utilities|Array|Clear _array_element_mid_7_a0d19ac74319389a5df2019166f976f0))
      (Utilities|Array|Clear (Variables|CharacterMIDs|GetPawnPartMaterials))
      (Utilities|Array|Clear (Variables|CharacterMIDs|GetDefault%20Weapon%20Materials))
      ;; Store Default Character Materials and Initialize Z Depth Render
      (Utilities|Array|Clear (Variables|VFX|GetSkeletalMeshes))
      (Utilities|Array|Clear (Variables|CharacterMIDs|GetPawnMaterials_ALL))
      (Utilities|FlowControl|Sequence
        (:then_0
          (Cosmetics|WhenCosmeticPartApplied (delegate Custom|OnMeshPart3)))
        (:then_1
          (Cosmetics|WhenCosmeticSpecificPartApplied (delegate Custom|OnBackpackPart) false "Backpack"))
        (:then_2
          (bind _returnvalue (Rendering|Material|GetScalarParameterValue "/Game/Packages/Fortress_SharedMaterials/Stencil_ID_Definitions/Stencil_ID_Definitions.Stencil_ID_Definitions" "Decal Mask" :!class /Script/Engine.KismetMaterialLibrary))
          (bind _returnvalue_2 (Rendering|Material|GetScalarParameterValue "/Game/Packages/Fortress_SharedMaterials/Stencil_ID_Definitions/Stencil_ID_Definitions.Stencil_ID_Definitions" "NonLocalPlayer_PlayerPawns" :!class /Script/Engine.KismetMaterialLibrary))
          (Cosmetics|WhenCosmeticPartApplied (delegate Custom|OnMeshPart2) true)
          (Cosmetics|WhenCosmeticsApplied (delegate Custom|OnMeshPartComplete))))
      (if (Variables|Water|GetIs%20the%20Wind/Water%20Scene%20Capture%20Enabled)
        (Utilities|FlowControl|Sequence
          (:then_0
            ;[0 0 0] Initialize Render To Texture References
            (Utilities|FlowControl|DoOnce false
              (:Completed
                (Class|WindManager|RegisterPlayerForRenderToTexturePurposes (Utilities|Casting|CastToWindManager (Fortnite|GetWindManager)) (Variables|Default|GetEffect_WaterInteraction_FX) (Variables|Default|GetGamplayWindInteractionEffects))
                ;[0 0.5033 1] Initialize Water Targets - can make all characters use a single texture in the future
                (bind _returnvalue_3 (Rendering|CreateRenderTarget2D 1 10))
                (Variables|Water|SetWetnessDepthTexture _returnvalue_3)
                (bind _returnvalue_4 (Rendering|Material|CreateDynamicMaterialInstance "/Game/WaterAndWind/Materials/RenderToTexture/WetnessDepthRender.WetnessDepthRender" "None" "None" :!class /Script/Engine.KismetMaterialLibrary))
                (Variables|Water|SetWaterLevelMID _returnvalue_4)
                (bind _returnvalue_5 (Rendering|Material|CreateDynamicMaterialInstance "/Game/WaterAndWind/Materials/RenderToTexture/DecayWaterLevel.DecayWaterLevel" "None" "None" :!class /Script/Engine.KismetMaterialLibrary))
                (Variables|Water|SetWaterDecayMID _returnvalue_5))))
          (:then_2
            ;[0 0.2525 0.835] Initialize mids as many times as needed
            (for _array_element_2 (Variables|CharacterMIDs|GetPawnMaterials_ALL)
              (Rendering|Material|SetTextureParameterValue _array_element_2 "WaterDepthTexture" (Variables|Water|GetWetnessDepthTexture)))))
        (else
          (Components|Activation|SetActive (Variables|Default|GetGamplayWindInteractionEffects) false false))))))

(event EventReinitializeWeaponMaterials
  (Networking|CanExecuteCosmeticEvents
    (:True
      (Utilities|Array|Clear (Variables|CharacterMIDs|GetDefault%20Weapon%20Materials))
      (Cosmetics|WhenCosmeticSpecificPartApplied (delegate Custom|OnBackpack) false "Backpack"))))

(event Custom|OnBackpack
  (params (Pawn :type Fort%20Player%20Pawn%20Object%20Reference) (Part :type Custom%20Character%20Part%20Object%20Reference) (MeshPart :type Skeletal%20Mesh%20Component%20Object%20Reference) (PartType :type EFortCustomPartType%20Enum :default "Head"))
  (Utilities|CharacterMaterialsSaveDefaults :WeaponMaterialArray (Variables|CharacterMIDs|GetDefault%20Weapon%20Materials) :Get%20Weapon%20Instead? true :MeshPart MeshPart))

(event Custom|PlayGameplayWindEffect
  (params (Player%20Wind%20Particle%20Emitter%20To%20Fire :type PlayerWindParticleEmitters%20Enum :default "NewEnumerator0"))
  (if (Variables|Water|GetIs%20the%20Wind/Water%20Scene%20Capture%20Enabled)
    (CallFunction|TriggerGameplayWindEmitter Player%20Wind%20Particle%20Emitter%20To%20Fire)))

(event EventExited/EnteredWaterVolume
  (bind _isinwater (Variables|Water|GetIsInWater))
  ;[0.08894 0.4556 1.27] Water
  (|CallParentFunction)
  (if (Variables|Water|GetIs%20the%20Wind/Water%20Scene%20Capture%20Enabled)
    (Components|Activation|SetActive (Variables|Default|GetEffect_WaterInteraction_FX) _isinwater true)
    (if _isinwater
      (Utilities|Time|ClearTimerbyFunctionName 0 "DisableWaterLevelTick")
      (Variables|Water|SetTickWaterLevel true)
      (else
        (Utilities|Time|SetTimerbyFunctionName 0 "DisableWaterLevelTick" 20)))))

(event EventPlayerCreatesaSplash
  (params (NewTransform :type Transform%20%28by%20ref%29))
  (bind _splasheffectsforenteringorexitingwater (Variables|Default|GetSplashEffectsForEnteringOrExitingWater))
  ;[0.08894 0.4556 1.27] Water
  (Transformation|SetWorldTransform _splasheffectsforenteringorexitingwater NewTransform false false)
  (Components|Activation|Activate _splasheffectsforenteringorexitingwater true)
  (bind _returnvalue (Rendering|Material|CreateDynamicMaterialInstance _splasheffectsforenteringorexitingwater 1 0 "None" :!class /Script/Engine.PrimitiveComponent))
  (bind _returnvalue_2 (Rendering|Material|CreateDynamicMaterialInstance _splasheffectsforenteringorexitingwater 2 0 "None" :!class /Script/Engine.PrimitiveComponent))
  (Rendering|Material|SetScalarParameterValue _returnvalue_2 "Z SpawnHeight" (.z (.location NewTransform)) :!class /Script/Engine.MaterialInstanceDynamic))

(event EventSetFirstPersonCamera
  (params (bNewUseFirstPersonCamera :type Boolean))
  (bind _weaponmesh (Variables|/Script/FortniteGame.FortWeapon|GetWeaponMesh (Variables|Weapon|GetCurrentWeapon)))
  (bind _returnvalue (Utilities|IsValid _weaponmesh :!class /Script/Engine.KismetSystemLibrary))
  (bind _currentweapon (Variables|Weapon|GetCurrentWeapon))
  (bind _returnvalue_2 (Utilities|IsValid _currentweapon :!class /Script/Engine.KismetSystemLibrary))
  ;[1 1 1] First-Person - toggle pawn visibility when 1st person camera is toggled
  (if bNewUseFirstPersonCamera
    (Networking|SwitchHasAuthority
      (:Authority
        _)
      (:Remote
        (FortPawn|SetPawnVisibilityLocal true)))
    (if _returnvalue_2
      (if _returnvalue
        (Development|SetHiddeninGame _weaponmesh true false)
        (bind _asfort_weap_building_tool_base (Utilities|Casting|CastToFortWeap_BuildingToolBase _currentweapon))
        (Development|SetHiddeninGame (Class|FortWeapBuildingToolBase|GetPencilMeshComponent _asfort_weap_building_tool_base) true false)))
    (else
      (Networking|SwitchHasAuthority
        (:Authority
          _)
        (:Remote
          (FortPawn|SetPawnVisibilityLocal false)))
      (if _returnvalue_2
        (if _returnvalue
          (Development|SetHiddeninGame _weaponmesh false false)
          (FortWeapon|SetWeaponVisibility _currentweapon (not (Weapon|IsWeaponHiddenbyHolstering 0)) true)
          (bind _asfort_weap_building_tool_base_2 (Utilities|Casting|CastToFortWeap_BuildingToolBase _currentweapon))
          (Development|SetHiddeninGame (Class|FortWeapBuildingToolBase|GetPencilMeshComponent _asfort_weap_building_tool_base_2) false false))))))

(event Custom|InternalSetFirstPersonCamera
  (params (NewUseFirstPersonCamera :type Boolean))
  (bind _weaponmesh (Variables|/Script/FortniteGame.FortWeapon|GetWeaponMesh (Variables|Weapon|GetCurrentWeapon)))
  (bind _returnvalue (Utilities|IsValid _weaponmesh :!class /Script/Engine.KismetSystemLibrary))
  (bind _currentweapon (Variables|Weapon|GetCurrentWeapon))
  (bind _returnvalue_2 (Utilities|IsValid _currentweapon :!class /Script/Engine.KismetSystemLibrary))
  ;[1 1 1] First-Person - toggle pawn visibility when 1st person camera is toggled
  (if NewUseFirstPersonCamera
    (Networking|SwitchHasAuthority
      (:Authority
        _)
      (:Remote
        (FortPawn|SetPawnVisibilityLocal true)))
    (if _returnvalue_2
      (if _returnvalue
        (Development|SetHiddeninGame _weaponmesh true false)
        (bind _asfort_weap_building_tool_base (Utilities|Casting|CastToFortWeap_BuildingToolBase _currentweapon))
        (Development|SetHiddeninGame (Class|FortWeapBuildingToolBase|GetPencilMeshComponent _asfort_weap_building_tool_base) true false)))
    (else
      (Networking|SwitchHasAuthority
        (:Authority
          _)
        (:Remote
          (FortPawn|SetPawnVisibilityLocal false)))
      (if _returnvalue_2
        (if _returnvalue
          (Development|SetHiddeninGame _weaponmesh false false)
          (FortWeapon|SetWeaponVisibility _currentweapon (not (Weapon|IsWeaponHiddenbyHolstering 0)) true)
          (bind _asfort_weap_building_tool_base_2 (Utilities|Casting|CastToFortWeap_BuildingToolBase _currentweapon))
          (Development|SetHiddeninGame (Class|FortWeapBuildingToolBase|GetPencilMeshComponent _asfort_weap_building_tool_base_2) false false))))))

(event EventOnWeaponEquipped
  (params (NewWeapon :type Fort%20Weapon%20Object%20Reference) (PrevWeapon :type Fort%20Weapon%20Object%20Reference))
  ;[1 1 1] First-Person - toggle pawn visibility when 1st person camera is toggled
  ;;[1 1 1] Mask post outline effects on weapons
  (if (Pawn|IsPlayerControlled)
    (bind _returnvalue (Rendering|Material|GetScalarParameterValue "/Game/Packages/Fortress_SharedMaterials/Stencil_ID_Definitions/Stencil_ID_Definitions.Stencil_ID_Definitions" "Decal Mask" :!class /Script/Engine.KismetMaterialLibrary))
    (Utilities|CharacterSetCharacterStencilValueandRenderSettings true (Math|Float|Round _returnvalue) true (Variables|Default|GetNullPart%20Mesh)
      (:Out
        _)))
  (bind _asfort_player_controller (Utilities|Casting|CastToFortPlayerController (Pawn|GetController)))
  (CallFunction|InternalSetFirstPersonCamera (Camera|IsUsingFirstPersonCamera _asfort_player_controller)))

(event Custom|UnBindWeaponSwap
  ;[1 1 1] War Cry weapon material changes
  (Networking|SwitchHasAuthority
    (:Authority
      (CallFunction|MultiEndSwap (Variables|Weapon|GetCurrentWeapon))
      (FortPawn|UnbindallEventsfromOnWeaponEquippedDelegate))))

(event Custom|BindWeaponSwap
  (params (InDuration :type Float%20%28double-precision%29))
  ;[1 1 1] War Cry weapon material changes
  (Networking|SwitchHasAuthority
    (:Authority
      (FortPawn|AssignOnWeaponEquippedDelegate (delegate Custom|ClientBindWeaponSwap))
      (Utilities|Time|SetTimerbyEvent (delegate Custom|UnBindWeaponSwap) InDuration))))

(event Custom|MultiEndSwap
  (params (CurrentWeapon :type Fort%20Weapon%20Object%20Reference))
  ;[1 1 1] War Cry weapon material changes
  (Utilities|IsValid CurrentWeapon
    (:Is%20Valid
      (Rendering|Material|SetScalarParameterValueonMaterials (Variables|/Script/FortniteGame.FortWeapon|GetWeaponMesh CurrentWeapon) "WarcryOn" 0.0))))

(event Custom|ClientBindWeaponSwap
  (params (NewWeapon :type Fort%20Weapon%20Object%20Reference) (PrevWeapon :type Fort%20Weapon%20Object%20Reference))
  ;[1 1 1] War Cry weapon material changes
  (CallFunction|MultiSwapWeapon NewWeapon PrevWeapon))

(event EventOnDisplaySentence
  (params (SpeechText :type Text%20%28by%20ref%29))
  ;[1 1 1] Speech bubble max visible distance
  (Networking|CanExecuteCosmeticEvents
    (:True
      (Variables|Default|SetLastSpeechText SpeechText)
      (bind _array_element (Utilities|Array|ForEachLoopwithBreak (Fortnite|GetLocalFortPlayerControllers true) :!label _label)
        (:LoopBody
          (if (<= (Math|Vector|VectorLength (- (Transformation|GetActorLocation (Game|Player|GetPlayerPawn _array_element)) (Transformation|GetActorLocation))) 4150)
            (bind _returnvalue (UserInterface|CreateWidget (Variables|Default|GetSpeechBubbleWidgetClass)))
            (bind _asfort_initialize_from_object_interface (Utilities|Casting|CastToFortInitializeFromObjectInterface _returnvalue)
              (:then
                (Init|InitfromObject _asfort_initialize_from_object_interface self :!class /Script/FortniteGame.FortInitializeFromObjectInterface))
              (:CastFailed
                _))
            (UserInterface|SetWidget (Variables|Default|GetSpeechBubbleComponent) _returnvalue)
            (goto _label Break)))))))

(event Custom|MultiSwapWeapon
  (params (New :type Fort%20Weapon%20Object%20Reference) (Prev :type Fort%20Weapon%20Object%20Reference))
  ;[1 1 1] War Cry weapon material changes
  (Utilities|IsValid New
    (:Is%20Valid
      (Rendering|Material|SetScalarParameterValueonMaterials (Variables|/Script/FortniteGame.FortWeapon|GetWeaponMesh New) "WarcryOn" 1))
    (:Is%20Not%20Valid
      _))
  (Utilities|IsValid Prev
    (:Is%20Valid
      (Rendering|Material|SetScalarParameterValueonMaterials (Variables|/Script/FortniteGame.FortWeapon|GetWeaponMesh Prev) "WarcryOn" 0.0))))

(event EventOnClearSentence
  ;[1 1 1] Clear the speech bubble widget
  (UserInterface|SetWidget (Variables|Default|GetSpeechBubbleComponent) 0))

(event EventOnDeathPlayEffects
  (params (Damage :type Float%20%28single-precision%29) (DamageTags :type Gameplay%20Tag%20Container%20Structure%20%28by%20ref%29) (Momentum :type Vector :default "0, 0, 0") (HitInfo :type Hit%20Result%20Structure%20%28by%20ref%29) (InstigatedBy :type Fort%20Pawn%20Object%20Reference) (DamageCauser :type Actor%20Object%20Reference) (EffectContext :type Gameplay%20Effect%20Context%20Handle%20Structure))
  (bind _mesh (Variables|Character|GetMesh))
  ;[1 1 1] On player killed
  (|CallParentFunction Damage DamageTags Momentum HitInfo InstigatedBy DamageCauser EffectContext)
  (if (Death|ShouldPlayDeathAnimation)
    (Animation|PlayLocalAnimMontage (select (Parachute|IsSkydiving) (Variables|Death|GetPlayerDeathSkydive_Montage) (select (Health|IsDBNO) (Variables|Death|GetPlayerDeathDBNO_Montage) (Variables|Death|GetPlayerDeath_Montage))))
    (bind _asfort_game_state (Utilities|Casting|CastToFortGameState (Game|GetGameState)))
    (if (GameState|DBNOEnabledforGameMode _asfort_game_state)
      _
      (elif (Variables|Death|GetPlayerDeath_EnableDeathDissolve)
        ;[1 1 1] Dissolve
        (Utilities|IsValid _mesh
          (:Is%20Valid
            (Rendering|SetCastShadow _mesh false)
            (for _array_element (Variables|VFX|GetSkeletalMeshes)
              (Rendering|SetCastShadow _array_element false))
            (bind _returnvalue (Game|SpawnActorfromClass "/Game/Effects/Fort_Effects/Effects/Characters/Generic/ResIn/DuplicateResOutMesh.DuplicateResOutMesh_C" (Transformation|GetActorTransform) "Undefined" "OverrideRootScale" 0 false (Variables|VFX|GetSkeletalMeshes) 1.333333 true self (Variables|CharacterMIDs|GetPawnMaterials_ALL) (Variables|Character|GetMesh) 0))))))))

(event Custom|ManualDeathDissolve
  (bind _mesh (Variables|Character|GetMesh))
  ;[1 1 1] Dissolve
  (Utilities|IsValid _mesh
    (:Is%20Valid
      (Rendering|SetCastShadow _mesh false)
      (for _array_element (Variables|VFX|GetSkeletalMeshes)
        (Rendering|SetCastShadow _array_element false))
      (bind _returnvalue (Game|SpawnActorfromClass "/Game/Effects/Fort_Effects/Effects/Characters/Generic/ResIn/DuplicateResOutMesh.DuplicateResOutMesh_C" (Transformation|GetActorTransform) "Undefined" "OverrideRootScale" 0 false (Variables|VFX|GetSkeletalMeshes) 1.333333 true self (Variables|CharacterMIDs|GetPawnMaterials_ALL) (Variables|Character|GetMesh) 0)))))

(event Cosmetic|EventOnEnteredWaterVolume
  (params (WaterSurfaceLocation :type Vector%20%28by%20ref%29 :default "0, 0, 0"))
  ;[0.01415 0.3663 1] HANDLE WATER - enter/exit splash & timed function for ambient FX determination, aligned for water volumes at the same level as the visual water
  (Variables|Water|SetIsInWater true)
  (Networking|CanExecuteCosmeticEvents
    (:True
      (CallFunction|EnableWaterAudio true)
      (Utilities|FlowControl|DoOnce false :!label _label
        (:Completed
          (Effects|SpawnEmitteratLocation "/Game/WaterAndWind/Particles/WaterInteraction/P_WaterSplash_athena.P_WaterSplash_Athena" (Transformation|GetWorldLocation (Variables|Character|GetCapsuleComponent)))
          (Utilities|FlowControl|Delay 0.4)
          (goto _label Reset))))))

(event Cosmetic|EventOnExitedWaterVolume
  (params (WaterSurfaceLocation :type Vector%20%28by%20ref%29 :default "0, 0, 0"))
  ;[0.01415 0.3663 1] HANDLE WATER - enter/exit splash & timed function for ambient FX determination, aligned for water volumes at the same level as the visual water
  (Variables|Water|SetIsInWater false)
  (Networking|CanExecuteCosmeticEvents
    (:True
      (CallFunction|EnableWaterAudio false)
      (Utilities|FlowControl|DoOnce false :!label _label
        (:Completed
          (Effects|SpawnEmitteratLocation "/Game/WaterAndWind/Particles/WaterInteraction/P_WaterSplash_athena.P_WaterSplash_Athena" (Transformation|GetWorldLocation (Variables|Character|GetCapsuleComponent)))
          (Utilities|FlowControl|Delay 0.4)
          (goto _label Reset))))))

(event Custom|SetBlockedByPawns
  (params (Blocked :type Boolean))
  ;[1 1 1] ServerSetCapsuleCollisionResponseToChannel
  (Networking|SwitchHasAuthority
    (:Authority
      (Variables|Default|SetBlockedByPawns Blocked))))

(event Custom|OnPartApplied
  (params (Pawn :type Fort%20Player%20Pawn%20Object%20Reference) (Part :type Custom%20Character%20Part%20Object%20Reference) (MeshPart :type Skeletal%20Mesh%20Component%20Object%20Reference) (PartType :type EFortCustomPartType%20Enum :default "Head"))
  (bind _l_shieldmidindex (Variables|Default|GetL_ShieldMIDIndex))
  (bind _shieldmidarrays (Variables|Shield|GetShieldMIDArrays))
  (bind (_output_parttype_5_ebdfff5740627902cc51dd966b8ee419 _output_mid_7_a0d19ac74319389a5df2019166f976f0) (Utilities|Struct|BreakCharacterPartMidArrayStruct (Utilities|Array|Get(aref) _shieldmidarrays _l_shieldmidindex)))
  (bind _shieldskmeshes (Variables|VFX|GetShieldSkMeshes))
  ;[1 1 1] Toggle shield mesh visibility on. Create shield meshes if they don't yet exist.
  ;;[1 1 1] Are we in the list?
  (for _array_element (Variables|Shield|GetShieldPartList)
    (if (Utilities|Enum|Equal(Enum) PartType _array_element)
      (if (and (Utilities|IsValid (Utilities|Map|Find _shieldskmeshes _array_element) :!class /Script/Engine.KismetSystemLibrary) (.ReturnValue (Utilities|Map|Find _shieldskmeshes _array_element)))
        (Rendering|SetVisibility (Utilities|Map|Find _shieldskmeshes _array_element) true false)
        (else
          (bind _returnvalue (AddComponent|NODE_AddSkeletalMeshComponent-21))
          (Utilities|Map|Add _shieldskmeshes _array_element _returnvalue)
          (Utilities|FlowControl|Sequence
            (:then_0
              (bind (_found _index) (CallFunction|FindShieldMIDArray _array_element))
              (if _found
                (Variables|Default|SetL_ShieldMIDIndex _index)
                (else
                  (bind _returnvalue_2 (Utilities|Array|Add (Variables|Shield|GetShieldMIDArrays) (Utilities|Struct|MakeCharacterPartMidArrayStruct _array_element)))
                  (Variables|Default|SetL_ShieldMIDIndex _returnvalue_2))))
            (:then_1
              (Utilities|Array|Clear _output_mid_7_a0d19ac74319389a5df2019166f976f0)
              (bind _objectsvalid (CallFunction|SetupDuplicateFXMesh (Variables|Shield|GetShield%20Material) _array_element _output_mid_7_a0d19ac74319389a5df2019166f976f0 _returnvalue true 2 MeshPart))
              (if _objectsvalid
                (Rendering|SetVisibility _returnvalue true false)
                (else
                  (Components|DestroyComponent _returnvalue))))))))))

(event GameplayCue|GameplayCue.Athena.Equipping
  (params (EventType :type EGameplayCueEvent%20Enum :default "OnActive") (Parameters :type Gameplay%20Cue%20Parameters%20Structure%20%28by%20ref%29))
  (EventDispatchers|CallOnUIGameplayCue "GameplayCue.Athena.Equipping" EventType Parameters))

(event Custom|ToggleShieldVisibility
  (bind _returnvalue (CallFunction|UseNewShieldHitMeshBehaviour))
  (bind _shieldskmeshes (Variables|VFX|GetShieldSkMeshes))
  (bind _legacyparttypetoclearshieldmesh (Variables|Default|GetLegacyPartTypeToClearShieldMesh))
  (if (Math|Boolean|ORBoolean (Variables|Shield|GetShieldVisible_Hit) (Variables|Shield|GetShieldVisible_Recover) (Variables|Shield|GetShieldVisible_Shatter) (Variables|Shield|GetShieldVisible_FullHealth))
    ;[1 1 1] Toggle shield mesh visibility on. Create shield meshes if they don't yet exist.
    (Cosmetics|WhenCosmeticPartApplied (delegate Custom|OnPartApplied) true)
    (else
      ;[1 1 1] Toggle shield mesh visibility off. Destroy meshes if Destroy Mesh parameter is true.
      (for _array_element (Variables|Shield|GetShieldPartList)
        (if (and (Utilities|IsValid (Utilities|Map|Find _shieldskmeshes (select _returnvalue _array_element _legacyparttypetoclearshieldmesh)) :!class /Script/Engine.KismetSystemLibrary) (.ReturnValue (Utilities|Map|Find _shieldskmeshes (select _returnvalue _array_element _legacyparttypetoclearshieldmesh))))
          (Rendering|SetVisibility (Utilities|Map|Find _shieldskmeshes (select _returnvalue _array_element _legacyparttypetoclearshieldmesh)) false false))))))

(event Custom|Set%20Scalar%20Parameter%20on%20Character%20MIDs
  (params (Parameter%20Name :type Name) (Parameter%20Value :type Float%20%28double-precision%29))
  (bind _mesh (Variables|Character|GetMesh)
    (:then
      (Rendering|Material|SetScalarParameterValueonMaterials _mesh Parameter%20Name Parameter%20Value))
    (:else
      _))
  (Cosmetics|WhenCosmeticPartApplied (delegate Custom|OnMeshPart) true))

(event Custom|OnMeshPart
  (params (Pawn :type Fort%20Player%20Pawn%20Object%20Reference) (Part :type Custom%20Character%20Part%20Object%20Reference) (MeshPart :type Skeletal%20Mesh%20Component%20Object%20Reference) (PartType :type EFortCustomPartType%20Enum :default "Head"))
  (Rendering|Material|SetScalarParameterValueonMaterials MeshPart (stale Custom|Set%20Scalar%20Parameter%20on%20Character%20MIDs Parameter%20Name) (stale Custom|Set%20Scalar%20Parameter%20on%20Character%20MIDs Parameter%20Value)))

(event MeleeSwingAndAnimTrailEventsPlayerInterface|EventBindToAnimTrailEvents
  (params (AnimNotifyBegin :type Delegate%20%28by%20ref%29) (AnimNotifyEnd :type Delegate%20%28by%20ref%29))
  (Default|BindEventtoAnimNotifyBegin AnimNotifyBegin :!class /Game/Abilities/Player/Pawns/PlayerPawn_Generic.PlayerPawn_Generic_C)
  (Default|BindEventtoAnimNotifyEnd AnimNotifyEnd :!class /Game/Abilities/Player/Pawns/PlayerPawn_Generic.PlayerPawn_Generic_C))

(event MeleeSwingAndAnimTrailEventsPlayerInterface|EventBindToSwingFX
  (params (SwingLeft :type Delegate%20%28by%20ref%29) (SwingLeftEnd :type Delegate%20%28by%20ref%29) (SwingRight :type Delegate%20%28by%20ref%29) (SwingRightEnd :type Delegate%20%28by%20ref%29))
  (Default|BindEventtoSwingLeft SwingLeft :!class /Game/Abilities/Player/Pawns/PlayerPawn_Generic.PlayerPawn_Generic_C)
  (Default|BindEventtoSwingLeftEnd SwingLeftEnd :!class /Game/Abilities/Player/Pawns/PlayerPawn_Generic.PlayerPawn_Generic_C)
  (Default|BindEventtoSwingRight SwingRight :!class /Game/Abilities/Player/Pawns/PlayerPawn_Generic.PlayerPawn_Generic_C)
  (Default|BindEventtoSwingRightEnd SwingRightEnd :!class /Game/Abilities/Player/Pawns/PlayerPawn_Generic.PlayerPawn_Generic_C))

(event MeleeSwingAndAnimTrailEventsPlayerInterface|EventBindToSwingFX2
  (params (SwingLeft2 :type Delegate%20%28by%20ref%29) (SwingRight2 :type Delegate%20%28by%20ref%29))
  (Default|BindEventtoSwingLeft2 SwingLeft2 :!class /Game/Abilities/Player/Pawns/PlayerPawn_Generic.PlayerPawn_Generic_C)
  (Default|BindEventtoSwingRight2 SwingRight2 :!class /Game/Abilities/Player/Pawns/PlayerPawn_Generic.PlayerPawn_Generic_C))

(event MeleeSwingAndAnimTrailEventsPlayerInterface|EventUnbindFromSwingFX2
  (params (SwingLeft2 :type Delegate%20%28by%20ref%29) (SwingRight2 :type Delegate%20%28by%20ref%29))
  (Default|UnbindEventfromSwingLeft2 SwingLeft2 :!class /Game/Abilities/Player/Pawns/PlayerPawn_Generic.PlayerPawn_Generic_C)
  (Default|UnbindEventfromSwingRight2 SwingRight2 :!class /Game/Abilities/Player/Pawns/PlayerPawn_Generic.PlayerPawn_Generic_C))

(event MeleeSwingAndAnimTrailEventsPlayerInterface|EventUnbindFromSwingFX
  (params (SwingLeft :type Delegate%20%28by%20ref%29) (SwingLeftEnd :type Delegate%20%28by%20ref%29) (SwingRight :type Delegate%20%28by%20ref%29) (SwingRightEnd :type Delegate%20%28by%20ref%29))
  (Default|UnbindEventfromSwingLeft SwingLeft :!class /Game/Abilities/Player/Pawns/PlayerPawn_Generic.PlayerPawn_Generic_C)
  (Default|UnbindEventfromSwingLeftEnd SwingLeftEnd :!class /Game/Abilities/Player/Pawns/PlayerPawn_Generic.PlayerPawn_Generic_C)
  (Default|UnbindEventfromSwingRight SwingRight :!class /Game/Abilities/Player/Pawns/PlayerPawn_Generic.PlayerPawn_Generic_C)
  (Default|UnbindEventfromSwingRightEnd SwingRightEnd :!class /Game/Abilities/Player/Pawns/PlayerPawn_Generic.PlayerPawn_Generic_C))

(event MeleeSwingAndAnimTrailEventsPlayerInterface|EventUnbindFromAnimTrailEvents
  (params (AnimNotifyBegin :type Delegate%20%28by%20ref%29) (AnimNotifyEnd :type Delegate%20%28by%20ref%29))
  (Default|UnbindEventfromAnimNotifyBegin AnimNotifyBegin :!class /Game/Abilities/Player/Pawns/PlayerPawn_Generic.PlayerPawn_Generic_C)
  (Default|UnbindEventfromAnimNotifyEnd AnimNotifyEnd :!class /Game/Abilities/Player/Pawns/PlayerPawn_Generic.PlayerPawn_Generic_C))

(fn ArethewindandwaterRTTpassesenabled
  (if (not (Variables|Water|GetWas%20RTT%20Enabled%20Var%20Set))
    (bind _returnvalue (Rendering|Material|GetScalarParameterValue "/Game/Packages/Fortress_SharedMaterials/GlobalMaterialParameters/FortniteMaterialParameters.FortniteMaterialParameters" "Water Interaction Enabled" :!class /Script/Engine.KismetMaterialLibrary))
    (Variables|Water|SetIs%20the%20Wind/Water%20Scene%20Capture%20Enabled (> _returnvalue 0.5))
    (Variables|Water|SetWas%20RTT%20Enabled%20Var%20Set true))
  (return (Variables|Water|GetIs%20the%20Wind/Water%20Scene%20Capture%20Enabled)))

(fn CreateandDuplicateEffectPoseableSkeletalMesh
  (params (Material%20to%20Apply :type Material%20Interface%20Object%20Reference) (Empty%20MID%20Array :type Array%20of%20Material%20Instance%20Dynamic%20Object%20References) (TranslucentSortPriority :type Integer) (PoseableMesh :type Poseable%20Mesh%20Component%20Object%20Reference%20%28by%20ref%29) (CustomBodyType :type EFortCustomPartType%20Enum :default "Head") (PartMesh :type Skeletal%20Mesh%20Component%20Object%20Reference))
  (Utilities|CharacterDuplicateandSlavePoseableMeshes Material%20to%20Apply Empty%20MID%20Array PoseableMesh CustomBodyType PartMesh :!label _label
    (:Duplicate%20Skeletal%20Mesh%20Is%20Not%20Valid
      (bind _returnvalue (AddComponent|NODE_AddPoseableMeshComponent-0))
      (Variables|SetBy-RefVar PoseableMesh _returnvalue)
      (Rendering|SetTranslucentSortPriority _returnvalue TranslucentSortPriority)
      (goto _label))))

(fn CreateandDuplicateEffectSkeletalMeshesParent
  (params (DuplicatedSkeletalMeshComponent :type Skeletal%20Mesh%20Component%20Object%20Reference%20%28by%20ref%29) (Material%20to%20Apply :type Material%20Interface%20Object%20Reference) (Empty%20MID%20Array :type Array%20of%20Material%20Instance%20Dynamic%20Object%20References) (TranslucentSortPriority :type Integer) (MeshPart :type Skeletal%20Mesh%20Component%20Object%20Reference) (CustomBodyType :type EFortCustomPartType%20Enum :default "Head"))
  (Utilities|FlowControl|Sequence
    (:then_0
      (Utilities|FlowControl|DoOnce false :!label _label
        (:Completed
          (bind _returnvalue (AddComponent|NODE_AddSkeletalMeshComponent-0))
          (Rendering|SetTranslucentSortPriority _returnvalue TranslucentSortPriority)
          (Variables|SetBy-RefVar DuplicatedSkeletalMeshComponent _returnvalue))) ; Safety to prevent infinite loop
    )
    (:then_1
      _))
  (Utilities|CharacterDuplicateandSlaveBodyMeshes :Skeletal%20Mesh%20Component%20Variable DuplicatedSkeletalMeshComponent :Mat%20Interface%20To%20Apply Material%20to%20Apply :Empty%20MID%20Array Empty%20MID%20Array :TranslucentSortPriority 0 :Ignore%20Base%20Material%20Input true :MeshPart MeshPart :CustomPartType CustomBodyType
    (:Duplicate%20Skeletal%20Mesh%20Is%20Not%20Valid
      (goto _label))))

(fn DisableWaterLevelTick
  (Variables|Water|SetTickWaterLevel false))

(fn EnableWaterAudio
  (params (IsEnteringWater :type Boolean))
  (if IsEnteringWater
    (Audio|PlaySoundatLocation (select (Pawn|IsLocallyControlled) "/Game/Athena/Sounds/Character/Footsteps/1P/Water/Athena_Character_Land_Water_Cue.Athena_Character_Land_Water_Cue" "/Game/Athena/Sounds/Character/Footsteps/3P/Water/Athena_Character_Land_Water_3P_Cue.Athena_Character_Land_Water_3P_Cue") (+ (Transformation|GetActorLocation) "0, 0,50.000000"))
    (else
      (Audio|PlaySoundatLocation (select (Pawn|IsLocallyControlled) "/Game/Athena/Sounds/Character/Footsteps/1P/Water/Athena_Character_Jump_Water_Cue.Athena_Character_Jump_Water_Cue" "/Game/Athena/Sounds/Character/Footsteps/3P/Water/Athena_Character_Jump_Water_3P_Cue.Athena_Character_Jump_Water_3P_Cue") (+ (Transformation|GetActorLocation) "0, 0,50.000000")))))

(fn FindShieldMIDArray
  (params (PartType :type EFortCustomPartType%20Enum :default "Head"))
  (bind (_output_parttype_5_ebdfff5740627902cc51dd966b8ee419 _output_mid_7_a0d19ac74319389a5df2019166f976f0) (Utilities|Struct|BreakCharacterPartMidArrayStruct (Utilities|Array|Get(aref) (Variables|Shield|GetShieldMIDArrays) (LocalVariables|GetL_Index))))
  (LocalVariables|SetL_Found false)
  (bind (_array_element _array_index) (Utilities|Array|ForEachLoopwithBreak (Variables|Shield|GetShieldMIDArrays) :!label _label)
    (:LoopBody
      (bind _array_element_parttype_5_ebdfff5740627902cc51dd966b8ee419 (Utilities|Struct|BreakCharacterPartMidArrayStruct _array_element))
      (if (Utilities|Enum|Equal(Enum) _array_element_parttype_5_ebdfff5740627902cc51dd966b8ee419 PartType)
        (LocalVariables|SetL_Found true)
        (LocalVariables|SetL_Index _array_index)
        (goto _label Break)))
    (:Completed
      (if (LocalVariables|GetL_Found)
        (return (LocalVariables|GetL_Found) (LocalVariables|GetL_Index) _output_mid_7_a0d19ac74319389a5df2019166f976f0)
        (else
          (return (LocalVariables|GetL_Found) -1 0))))))

(fn FindShieldOpacity
  (LocalVariables|SetCurrentShieldOpacity (* (Math|Float|Max(Float) (Variables|Shield|GetShield%20Shatter%20Opacity) (Variables|Shield|GetRegen%20Opacity) (Variables|Shield|GetDamage%20Opacity) (Variables|Shield|GetFully%20Regened%20Shield%20health%20Opacity)) (Variables|Shield|GetDefault%20Shield%20opacity)))
  ;[1 1 1] Loop to set Opacity on all Shield MIDs
  (for _array_element (Variables|Shield|GetShieldMIDArrays)
    (bind (_array_element_parttype_5_ebdfff5740627902cc51dd966b8ee419 _array_element_mid_7_a0d19ac74319389a5df2019166f976f0) (Utilities|Struct|BreakCharacterPartMidArrayStruct _array_element))
    (for _array_element_2 _array_element_mid_7_a0d19ac74319389a5df2019166f976f0
      (Rendering|Material|SetScalarParameterValue _array_element_2 "Opacity" (LocalVariables|GetCurrentShieldOpacity) :!class /Script/Engine.MaterialInstanceDynamic))))

(fn GetHealthPercent2
  (bind _self self)
  (bind _returnvalue (Ability|GetAbilitySystemComponent _self))
  (bind (_bsuccessfullyfoundattribute _returnvalue_2) (Ability|Attribute|GetFloatAttributefromAbilitySystemComponent _returnvalue (Variables|Default|GetAttribute%20CurrentHealth%20Const)))
  (bind _attribute_max_health_const (Variables|Default|GetAttribute%20Max%20Health%20Const))
  (bind (_bsuccessfullyfoundattribute_2 _returnvalue_3) (Ability|Attribute|GetFloatAttributefromAbilitySystemComponent _returnvalue _attribute_max_health_const))
  (if (> _returnvalue_3 0.0)
    (return (* (/ _returnvalue_2 _returnvalue_3) 100))
    (else
      (return 0.0))))

(fn GetInWater
  (return (Variables|Water|GetIsInWater)))

(fn GetShieldPercent2
  (bind _attribute_shield_const (Variables|Default|GetAttribute%20Shield%20Const))
  (bind _self self)
  (bind _returnvalue (Ability|GetAbilitySystemComponent _self))
  (bind (_bsuccessfullyfoundattribute _returnvalue_2) (Ability|Attribute|GetFloatAttributefromAbilitySystemComponent _returnvalue _attribute_shield_const))
  (bind (_bsuccessfullyfoundattribute_2 _returnvalue_3) (Ability|Attribute|GetFloatAttributefromAbilitySystemComponent _returnvalue (Variables|Default|GetAttribute%20CurrentShield%20Const)))
  (if (> _returnvalue_2 0.0)
    (return (* (/ _returnvalue_3 _returnvalue_2) 100))
    (else
      (return 0.0))))

(fn Melee_Effect_Color
  (return "0, 0, 0"))

(fn OnSetMaterialPart
  (params (Pawn :type Fort%20Player%20Pawn%20Object%20Reference) (Part :type Custom%20Character%20Part%20Object%20Reference) (MeshPart :type Skeletal%20Mesh%20Component%20Object%20Reference) (PartType :type EFortCustomPartType%20Enum :default "Head"))
  (switch Utilities|FlowControl|Switch|SwitchonEFortCustomPartType PartType
    (:Head
      (CallFunction|SetPartMaterial (Variables|GameplayCueNeeded|GetPrevious%20Head%20MID) MeshPart))
    (:Body
      (CallFunction|SetPartMaterial (Variables|GameplayCueNeeded|GetPrevious%20Body%20MID) MeshPart))
    (:Hat
      (CallFunction|SetPartMaterial (Variables|GameplayCueNeeded|GetPrevious%20Hat%20MID) MeshPart))
    (:Backpack
      (CallFunction|SetPartMaterial (Variables|GameplayCueNeeded|GetPrevious%20Backpack%20MID) MeshPart))
    (:MiscOrTail
      (CallFunction|SetPartMaterial (Variables|GameplayCueNeeded|GetPrevious%20Charm%20MID) MeshPart))
    (:Face
      (CallFunction|SetPartMaterial (Variables|GameplayCueNeeded|GetPrevious%20Face%20MID) MeshPart))))

(fn OnRep_BlockedByPawns
  (bind _blockedbypawns (Variables|Default|GetBlockedByPawns))
  (bind _returnvalue (select _blockedbypawns "ECR_Block" "ECR_Ignore"))
  (Collision|SetCollisionResponsetoChannel (Variables|Character|GetCapsuleComponent) "ECC_Pawn" _returnvalue)
  (Collision|SetCollisionResponsetoChannel (Variables|Character|GetCapsuleComponent) "ECC_GameTraceChannel4" _returnvalue))

(fn OnRep_On_Player_Built_Floor
  (bind _self self)
  (bind _gameplayeventdata (Utilities|Struct|MakeGameplayEventData :Instigator _self :Target _self :ContextHandle "()" :InstigatorTags "(GameplayTags=)" :TargetTags "(GameplayTags=)" :EventMagnitude 0.0 :TargetData "()"))
  (if (Variables|Default|GetOn_Player_Built_Floor)
    (Ability|SendGameplayEventtoActor _self (Variables|Default|GetEvent_NeedRoadsActive) _gameplayeventdata)
    (else
      (Ability|SendGameplayEventtoActor _self (Variables|Default|GetEvent_NeedRoadsDeactive) _gameplayeventdata))))

(fn OverrideMaterialsandCopyParametersonWeaponsMesh
  (params (Material%20to%20Apply :type Material%20Interface%20Object%20Reference))
  (bind _previous_weapon_mid (Variables|GameplayCueNeeded|GetPrevious%20Weapon%20MID))
  (bind _currentweapon (Variables|Weapon|GetCurrentWeapon))
  (bind _weaponmesh (Variables|/Script/FortniteGame.FortWeapon|GetWeaponMesh _currentweapon))
  (Networking|CanExecuteCosmeticEvents
    (:True
      (Utilities|IsValid Material%20to%20Apply
        (:Is%20Valid
          (for (_array_element _array_index) (Rendering|Material|GetMaterials _weaponmesh)
            (bind _asmaterial_instance_dynamic (Utilities|Casting|CastToMaterialInstanceDynamic _array_element)
              (:then
                (Utilities|Array|SetArrayElem _previous_weapon_mid _array_index _asmaterial_instance_dynamic true)
                (bind _returnvalue (Rendering|Material|CreateDynamicMaterialInstance Material%20to%20Apply "None" "None" :!class /Script/Engine.KismetMaterialLibrary))
                (Rendering|Material|CopyMaterialInstanceParameters _returnvalue _asmaterial_instance_dynamic true)
                (Rendering|Material|SetMaterial _weaponmesh _array_index _returnvalue))
              (:CastFailed
                (bind _returnvalue_2 (Rendering|Material|CreateDynamicMaterialInstance _array_element "None" "None" :!class /Script/Engine.KismetMaterialLibrary))
                (Utilities|Array|SetArrayElem _previous_weapon_mid _array_index _returnvalue_2 true)
                (bind _returnvalue_3 (Rendering|Material|CreateDynamicMaterialInstance Material%20to%20Apply "None" "None" :!class /Script/Engine.KismetMaterialLibrary))
                (Rendering|Material|CopyMaterialInstanceParameters _returnvalue_3 _returnvalue_2 true)
                (Rendering|Material|SetMaterial _weaponmesh _array_index _returnvalue_3)))))))))

(fn PlayHitSound
  (params (Instigator :type Fort%20Pawn%20Object%20Reference) (Damage%20Causer :type Actor%20Object%20Reference))
  (Utilities|FlowControl|Sequence
    (:then_0
      (bind _returnvalue (GameplayTags|HasAnyMatchingGameplayTags(Message) Damage%20Causer (Variables|Parent_Pawn_Variables|GetGameplayTagsForHitByHuskMeleeSound)))
      (if _returnvalue
        (LocalVariables|SetHitSource Damage%20Causer)
        (LocalVariables|SetHitSound (select (Variables|Shield|GetLatestDamageIsShieldDamage) (Variables|Sounds|GetSound_Shield_Impact) (Variables|Parent_Pawn_Variables|GetHitByHuskMeleeSound)))))
    (:then_1
      (Utilities|IsValid (LocalVariables|GetHitSource)
        (:Is%20Not%20Valid
          (bind _returnvalue_2 (GameplayTags|HasAnyMatchingGameplayTags(Message) Instigator (Variables|Parent_Pawn_Variables|GetGameplayTagsForHitByHuskMeleeSound)))
          (if _returnvalue_2
            (LocalVariables|SetHitSource Instigator)
            (LocalVariables|SetHitSound (select (Variables|Shield|GetLatestDamageIsShieldDamage) (Variables|Sounds|GetSound_Shield_Impact) (Variables|Parent_Pawn_Variables|GetHitByHuskMeleeSound)))))))
    (:then_2
      (bind _asfort_player_controller (Utilities|Casting|CastToFortPlayerController (Pawn|GetController)))
      (Utilities|IsValid (LocalVariables|GetHitSource)
        (:Is%20Valid
          _)
        (:Is%20Not%20Valid
          (LocalVariables|SetHitSource _asfort_player_controller) ; Unknown.  Either an ability or something stranger still.  Use ranged sound.
          (LocalVariables|SetHitSound (select (Variables|Shield|GetLatestDamageIsShieldDamage) (Variables|Sounds|GetSound_Shield_Impact) (Variables|Parent_Pawn_Variables|GetHitByHuskRangedSound)))))
      (Controller|PlaySpatialSound _asfort_player_controller (LocalVariables|GetHitSound) (Transformation|GetActorLocation)))))

(fn RemoveOverlayOnMesh
  (params (Skeletal%20Mesh%20Component :type Skeletal%20Mesh%20Component%20Object%20Reference))
  (Utilities|FlowControl|Sequence
    (:then_0
      ; FORT-1065067 - Overlay Material displaying after shield fx triggered
      (if (CallFunction|UseNewShieldHitMeshBehaviour)
        (for (_array_element _array_index) (Variables|/Script/Engine.SkeletalMesh|GetMaterials (Variables|/Script/Engine.SkinnedMeshComponent|GetSkeletalMesh Skeletal%20Mesh%20Component))
          (if (or (Utilities|IsValid (Rendering|Material|GetOverlayMaterial Skeletal%20Mesh%20Component true _array_index :!class /Script/Engine.MeshComponent) :!class /Script/Engine.KismetSystemLibrary) (Utilities|IsValid (.OverlayMaterialInterface (Utilities|Struct|BreakSkeletalMaterial _array_element)) :!class /Script/Engine.KismetSystemLibrary))
            (Rendering|Material|SetOverlayMaterial Skeletal%20Mesh%20Component (LocalVariables|GetNon-visible%20Overlay%20Material) true _array_index :!class /Script/Engine.MeshComponent)))
        (Utilities|IsValid (Rendering|Material|GetOverlayMaterial Skeletal%20Mesh%20Component false 0 :!class /Script/Engine.MeshComponent)
          (:Is%20Valid
            (Rendering|Material|SetOverlayMaterial Skeletal%20Mesh%20Component (LocalVariables|GetNon-visible%20Overlay%20Material) false 0 :!class /Script/Engine.MeshComponent)))))
    (:then_1
      (return))))

(fn RestorePreviousMaterialsonCharacterMesh
  (Networking|CanExecuteCosmeticEvents
    (:True
      (CallFunction|SetPartMaterial (Variables|GameplayCueNeeded|GetPrevious%20MID) (Variables|Character|GetMesh))
      (Cosmetics|WhenCosmeticPartApplied (EventDispatchers|CreateEvent :!event On%20Set%20Material%20Part)))))

(fn RestorePreviousMaterialsonWeaponsMesh
  (params (Weapon%20to%20Restore :type Fort%20Weapon%20Object%20Reference))
  (bind _weaponmesh (Variables|/Script/FortniteGame.FortWeapon|GetWeaponMesh Weapon%20to%20Restore))
  (Networking|CanExecuteCosmeticEvents
    (:True
      (for (_array_element _array_index) (Variables|GameplayCueNeeded|GetPrevious%20Weapon%20MID)
        (Utilities|IsValid _array_element
          (:Is%20Valid
            (bind _returnvalue (Fortnite|GetDynamicMaterialInstance _weaponmesh _array_index))
            (if (not (Utilities|Equal(Object) _returnvalue _array_element))
              (Rendering|Material|SetMaterial _weaponmesh _array_index _array_element))))))))

(fn SetBodyTypeSounds
  (switch Utilities|FlowControl|Switch|SwitchonEFortCustomBodyType (Customization|GetCharacterBodyType)
    (:Small
      (Utilities|FlowControl|Sequence
        (:then_0
          (Variables|Sound|SetDefaultFootstepSound "/Game/Sounds/Fortress_Foley/Player_Movement/Player_Step_Walk_Cue.Player_Step_Walk_Cue"))
        (:then_1
          (Variables|Sound|SetDefaultFastFootstepSound "/Game/Sounds/Fortress_Foley/Player_Movement/Player_Step_Run_Cue.Player_Step_Run_Cue"))
        (:then_2
          (Variables|Sound|SetDefaultHardLandingSound "/Game/Sounds/Fort_Step_Sounds/Human/Human_Land_Heavy_Dirt_Cue.Human_Land_Heavy_Dirt_Cue"))
        (:then_3
          (Variables|Sound|SetDefaultJumpSound "/Game/Sounds/Fort_Foley_Sounds/Moves/Player_Foley_Jump_Cue.Player_Foley_Jump_Cue"))
        (:then_4
          (Variables|Sound|SetDefaultLandingSound "/Game/Sounds/Fortress_Foley/Player_Movement/Player_Land_Cue.Player_Land_Cue"))))
    (:Medium
      (Utilities|FlowControl|Sequence
        (:then_0
          (Variables|Sound|SetDefaultFootstepSound "/Game/Sounds/Fortress_Foley/Player_Movement/Commando_Step_Walk_Cue.Commando_Step_Walk_Cue"))
        (:then_1
          (Variables|Sound|SetDefaultFastFootstepSound "/Game/Sounds/Fortress_Foley/Player_Movement/Commando_Step_Run_Cue.Commando_Step_Run_Cue"))
        (:then_2
          (Variables|Sound|SetDefaultHardLandingSound "/Game/Sounds/Fortress_Foley/Player_Movement/Commando_LandHard_Cue.Commando_LandHard_Cue"))
        (:then_3
          (Variables|Sound|SetDefaultJumpSound "/Game/Sounds/Fortress_Foley/Player_Movement/Commando_Jump_Cue.Commando_Jump_Cue"))
        (:then_4
          (Variables|Sound|SetDefaultLandingSound "/Game/Sounds/Fortress_Foley/Player_Movement/Commando_Land_Cue.Commando_Land_Cue"))))
    (:Large
      (Utilities|FlowControl|Sequence
        (:then_0
          (Variables|Sound|SetDefaultFootstepSound "/Game/Sounds/Fortress_Foley/Player_Movement/Constructor_Step_Walk_Cue.Constructor_Step_Walk_Cue"))
        (:then_1
          (Variables|Sound|SetDefaultFastFootstepSound "/Game/Sounds/Fortress_Foley/Player_Movement/Constructor_Step_Run_Cue.Constructor_Step_Run_Cue"))
        (:then_2
          (Variables|Sound|SetDefaultHardLandingSound "/Game/Sounds/Fortress_Foley/Player_Movement/Constructor_LandHard_Cue.Constructor_LandHard_Cue"))
        (:then_3
          (Variables|Sound|SetDefaultJumpSound "/Game/Sounds/Fortress_Foley/Player_Movement/Constructor_Jump_Cue.Constructor_Jump_Cue"))
        (:then_4
          (Variables|Sound|SetDefaultLandingSound "/Game/Sounds/Fortress_Foley/Player_Movement/Constructor_Land_Cue.Constructor_Land_Cue"))))))

(fn SetPartMaterial
  (params (Array :type Array%20of%20Material%20Instance%20Dynamic%20Object%20References) (MeshPart :type Skeletal%20Mesh%20Component%20Object%20Reference))
  (Utilities|IsValid MeshPart
    (:Is%20Valid
      (for (_array_element _array_index) Array
        (Utilities|IsValid _array_element
          (:Is%20Valid
            (bind _returnvalue (Fortnite|GetDynamicMaterialInstance MeshPart _array_index))
            (if (not (Utilities|Equal(Object) _returnvalue _array_element))
              (Rendering|Material|SetMaterial (Variables|Character|GetMesh) _array_index _array_element))))))))

(fn SetScalarParameteronDuplicateMeshMIDs
  (params (Parameter%20Name :type Name) (Parameter%20Value :type Float%20%28double-precision%29))
  (bind _duplicatecharactermesh (Variables|GameplayCueNeeded|GetDuplicateCharacterMesh)
    (:then
      (Rendering|Material|SetScalarParameterValueonMaterials _duplicatecharactermesh Parameter%20Name Parameter%20Value))
    (:else
      _))
  (return))

(fn SetShieldMids
  (params (Highlight%20Cracks :type Float%20%28double-precision%29) (Set%20Highlight%20Cracks :type Boolean) (Push :type Float%20%28double-precision%29) (Set%20Push :type Boolean))
  (for _array_element (Variables|Shield|GetShieldMIDArrays)
    (bind (_array_element_parttype_5_ebdfff5740627902cc51dd966b8ee419 _array_element_mid_7_a0d19ac74319389a5df2019166f976f0) (Utilities|Struct|BreakCharacterPartMidArrayStruct _array_element))
    (CallFunction|SetShieldMidsInternalLoop Highlight%20Cracks Set%20Highlight%20Cracks Push Set%20Push _array_element_mid_7_a0d19ac74319389a5df2019166f976f0))
  (return))

(fn SetShieldMids_InternalLoop
  (params (Highlight%20Cracks :type Float%20%28double-precision%29) (Set%20Highlight%20Cracks :type Boolean) (Push :type Float%20%28double-precision%29) (Set%20Push :type Boolean) (MID%20Array :type Array%20of%20Material%20Instance%20Dynamic%20Object%20References))
  (for _array_element MID%20Array
    (if Set%20Highlight%20Cracks
      (Rendering|Material|SetScalarParameterValue _array_element "HighlightCracks" Highlight%20Cracks :!class /Script/Engine.MaterialInstanceDynamic))
    (if Set%20Push
      (Rendering|Material|SetScalarParameterValue _array_element "VertPushAmount" Push :!class /Script/Engine.MaterialInstanceDynamic)))
  (return))

(fn SetupDuplicateFXMesh
  (params (Material%20to%20Apply :type Material%20Interface%20Object%20Reference%20%28by%20ref%29) (PartType :type EFortCustomPartType%20Enum :default "Head") (Empty%20MID%20Array :type Array%20of%20Material%20Instance%20Dynamic%20Object%20References) (Duplicated%20Mesh :type Skeletal%20Mesh%20Component%20Object%20Reference%20%28by%20ref%29) (Transfer%20Material%20Parameters :type Boolean :default true) (Translucent%20Sort%20Order :type Integer) (MeshPart :type Skeletal%20Mesh%20Component%20Object%20Reference))
  (bind (_found _index _mids) (CallFunction|FindStoredCharPartMaterials PartType))
  (if _found
    (Utilities|TransferMaterialAttributesandSlaveDuplicatedMeshesUsingDefaultMats Material%20to%20Apply _mids Empty%20MID%20Array Duplicated%20Mesh false Transfer%20Material%20Parameters Translucent%20Sort%20Order MeshPart
      (:Success
        (return true))
      (:Objects%20Not%20Valid
        _)))
  (Development|PrintString (Utilities|String|Append "Unable to duplicate character part: " (Utilities|String|EnumtoString PartType)))
  (return false))

(fn SlaveAMeshToTheBody
  (params (Mesh :type Skeletal%20Mesh%20Component%20Object%20Reference) (Master :type Skeletal%20Mesh%20Component%20Object%20Reference))
  (bind _returnvalue (Transformation|GetAttachSocketName Master))
  (if (Utilities|Name|Equal(Name) _returnvalue "None")
    (LocalVariables|SetMasterToUse Master)
    ;[1 1 1] Use the master-most component
    (while (Utilities|IsValid (Variables|/Script/Engine.SkinnedMeshComponent|GetLeaderPoseComponent (LocalVariables|GetMasterToUse)) :!class /Script/Engine.KismetSystemLibrary)
      (LocalVariables|SetMasterToUse (Variables|/Script/Engine.SkinnedMeshComponent|GetLeaderPoseComponent (LocalVariables|GetMasterToUse))))
    (Transformation|SetRelativeTransform Mesh (Transformation|GetRelativeTransform (Variables|Character|GetMesh)) false false)
    (Components|SkinnedMesh|SetLeaderPoseComponent Mesh (LocalVariables|GetMasterToUse) false false)
    (Clothing|BindClothtoLeaderPoseComponent Mesh)
    (Rendering|SetVisibility Mesh false false)
    (else
      (Transformation|AttachComponentToComponent Mesh (Transformation|GetAttachParent Master) _returnvalue "SnapToTarget" "SnapToTarget" "SnapToTarget" true)
      (Rendering|SetVisibility Mesh false false))))

(fn TriggerGameplayWindEmitter
  (params (Player%20Wind%20Particle%20Emitter%20To%20Fire :type PlayerWindParticleEmitters%20Enum :default "NewEnumerator0"))
  (bind _gamplaywindinteractioneffects (Variables|Default|GetGamplayWindInteractionEffects))
  ;[0 0 0] Due to the way particle systems work the system must be reset to emit particles again after all of it's emitters were disabled
  (Components|Activation|Activate (Variables|Default|GetGamplayWindInteractionEffects) true)
  ;[0 0 0] Disable all emitters but the one that is supposed to fire
  (bind _enumvalue (Utilities|Enum|ForEachPlayerWindParticleEmitters false)
    (:LoopBody
      (if (Utilities|Enum|Equal(Enum) _enumvalue Player%20Wind%20Particle%20Emitter%20To%20Fire)
        (Effects|Components|ParticleSystem|SetEmitterEnable _gamplaywindinteractioneffects (Utilities|String|StringToName (Utilities|String|EnumtoString _enumvalue)) true)
        (else
          (Effects|Components|ParticleSystem|SetEmitterEnable _gamplaywindinteractioneffects (Utilities|String|StringToName (Utilities|String|EnumtoString _enumvalue)) false))))))

(fn ConstructionScript
  ;[0 0 0] Attach run/walk dust FX emitters to feet socket
  (Transformation|AttachComponentToComponent (Variables|Default|GetEffect_Player_Run_Land) (Variables|Character|GetMesh) "FX_Feet" "KeepRelative" "KeepRelative" "KeepRelative" true)
  (Transformation|AttachComponentToComponent (Variables|Default|GetEffect_Player_Walk_Land) (Variables|Character|GetMesh) "FX_Feet" "KeepRelative" "KeepRelative" "KeepRelative" true)
  ;[0 0 0] Temporary hack because the outlander pawn is locked right now. It should be using RootComponent for the "component to attach" when playing sounds
  (Variables|Default|SetSound%20Player%20Hit (Variables|Transformation|GetRootComponent))
  ;[0 0 0] Sets the "Is the Wind/Water Scene Capture Enabled" var
  (CallFunction|AreTheWindAndWaterRTTPassesEnabled))
