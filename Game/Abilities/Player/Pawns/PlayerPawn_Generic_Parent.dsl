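; Generic parent Blueprint for player pawns, providing shared cosmetic setup and hooks for child Blueprints.
; BeginPlay adds and stores a post-process component only where cosmetic events can run and the pawn is locally controlled.
; FindStoredCharPartMaterials scans stored part-material structs by part type, returning found flag, index and MID; the loop never breaks, so the last match wins.
; Splash and weapon-material events are empty; the water event keeps inherited variables referenced to prevent deletion.
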(event Custom|Exited/Entered%20Water%20Volume
  ;[0 0 0] These variables are placed here to prevent them from being deleted. They're implemented in this bp so that a character specific macro can utilize them and so they may be inherited by other bps
  (Default|CallOnOverlapWaterVolume (Variables|Water|GetIsInWater) self (Variables|Water|GetCurrentWaterMeshActor)))

(event Custom|Player%20Creates%20a%20Splash
  (params (NewTransform :type Transform%20%28by%20ref%29)))

(event EventBeginPlay
  (Networking|CanExecuteCosmeticEvents
    (:True
      (if (Pawn|IsLocallyControlled)
        (bind _returnvalue (AddComponent|NODE_AddPostProcessComponent-0))
        (Variables|Default|SetPlayerPostProcessFX _returnvalue)))))

(event Custom|ReinitializeWeaponMaterials)

(fn FindStoredCharPartMaterials
  (params (PartType :type EFortCustomPartType%20Enum :default "Body"))
  (bind (_output_parttype_5_ebdfff5740627902cc51dd966b8ee419 _output_mid_7_a0d19ac74319389a5df2019166f976f0) (Utilities|Struct|BreakCharacterPartMidArrayStruct (Utilities|Array|Get(aref) (Variables|CharacterMIDs|GetPawnPartMaterials) (LocalVariables|GetL_Index))))
  (LocalVariables|SetL_Found false)
  (bind (_array_element _array_index) (Utilities|Array|ForEachLoopwithBreak (Variables|CharacterMIDs|GetPawnPartMaterials))
    (:LoopBody
      (bind _array_element_parttype_5_ebdfff5740627902cc51dd966b8ee419 (Utilities|Struct|BreakCharacterPartMidArrayStruct _array_element))
      (if (Utilities|Enum|Equal(Enum) _array_element_parttype_5_ebdfff5740627902cc51dd966b8ee419 PartType)
        (LocalVariables|SetL_Index _array_index)
        (LocalVariables|SetL_Found true)))
    (:Completed
      _))
  (if (LocalVariables|GetL_Found)
    (return (LocalVariables|GetL_Found) (LocalVariables|GetL_Index) _output_mid_7_a0d19ac74319389a5df2019166f976f0)
    (else
      (return (LocalVariables|GetL_Found) -1 0))))

(fn ConstructionScript)
